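# Control Rig for the F_MED_CordSyrup female medium cosmetic body, applying corrective deformation to skirt, left shoulder armour and right collar.
# It runs the shared DeformRig function, then the Shoulder Armour and Skirt setups.
# Armour and collar offsets follow deformRig animation attributes, with Emote Float and DBNO altering the armour.
# Skirt bones blend pose offsets and parent constraints weighted by spherical pose readers on the thighs, plus crouch corrections.
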ControlRigBlueprint /BRCosmetics/Characters/Player/Female/Medium/Bodies/F_MED_CordSyrup/Meshes/F_MED_CordSyrup_CtrlRig
graphs [
 CollapseNode_ContainedGraph {nodes [
   RigVMFunctionEntryNode()
   > RigUnit_SetTransform = &RigUnit_SetTransform(
    Item=(Type=Bone,Name="thigh_r")
    Weight=1
    bInitial=False
    Value=&RigUnit_GetTransform(Item=(Type=Control,Name="thigh_r_ctrl"), bInitial=False).Transform
   )
   > &RigUnit_SetTransform(
    Item=(Type=Bone,Name="thigh_l")
    Weight=1
    bInitial=False
    Value=&RigUnit_GetTransform(Item=(Type=Control,Name="thigh_l_ctrl"), bInitial=False).Transform
   )
   > &RigUnit_SetTransform(
    Item=(Type=Bone,Name="calf_r")
    Weight=1
    bInitial=False
    Value=&RigUnit_GetTransform(Item=(Type=Control,Name="calf_r_ctrl"), bInitial=False).Transform
   )
   > &RigUnit_SetTransform(
    Item=(Type=Bone,Name="calf_l")
    Weight=1
    bInitial=False
    Value=&RigUnit_GetTransform(Item=(Type=Control,Name="calf_l_ctrl"), bInitial=False).Transform
   )
   RigUnit_GetTransform = &RigUnit_GetTransform(Item=(Type=Control,Name="pelvis_ctrl"), bInitial=False)
   > &RigUnit_SetTransform(Item=(Type=Bone,Name="pelvis"), Weight=1, bInitial=False, Value=@RigUnit_GetTransform.Transform)
   > &RigUnit_SetTransform(
    Item=(Type=Bone,Name="spine_01")
    Weight=1
    bInitial=False
    Value=&RigUnit_GetTransform(Item=(Type=Control,Name="spine_01_ctrl"), bInitial=False).Transform
   )
   > &RigUnit_SetTransform(
    Item=(Type=Bone,Name="spine_02")
    Weight=1
    bInitial=False
    Value=&RigUnit_GetTransform(Item=(Type=Control,Name="spine_02_ctrl"), bInitial=False).Transform
   )
   > &RigUnit_SetTransform(
    Item=(Type=Bone,Name="spine_03")
    Weight=1
    bInitial=False
    Value=&RigUnit_GetTransform(Item=(Type=Control,Name="spine_03_ctrl"), bInitial=False).Transform
   )
   > &RigUnit_SetTransform(
    Item=(Type=Bone,Name="spine_04")
    Weight=1
    bInitial=False
    Value=&RigUnit_GetTransform(Item=(Type=Control,Name="spine_04_ctrl"), bInitial=False).Transform
   )
   > &RigUnit_SetTransform(
    Item=(Type=Bone,Name="clavicle_l")
    Weight=1
    bInitial=False
    Value=&RigUnit_GetTransform(Item=(Type=Control,Name="clavicle_l_ctrl"), bInitial=False).Transform
   )
   > &RigUnit_SetTransform(
    Item=(Type=Bone,Name="upperarm_l")
    Weight=1
    bInitial=False
    Value=&RigUnit_GetTransform(Item=(Type=Control,Name="upperarm_l_ctrl"), bInitial=False).Transform
   )
   > RigVMFunctionReturnNode()
   RigVMRerouteNode(Value=())
  ]}
 Multiply_1_1_ContainedGraph {nodes [
   n1 = RigVMFunctionEntryNode()
   RigVMFunctionReturnNode(Result=&RigVMFunction_MathDoubleMul(B=@n1.C, A=&RigVMFunction_MathDoubleMul(A=@n1.A, B=@n1.B).Result).Result)
  ]}
 Multiply_2_ContainedGraph {nodes [
   n1 = RigVMFunctionEntryNode()
   Multiply = &RigVMFunction_MathDoubleMul(A=@n1.A, B=@n1.B)
   RigVMFunctionReturnNode(Result=&RigVMFunction_MathDoubleMul(B=@n1.C, A=@Multiply.Result).Result)
  ]}
 New Function_ContainedGraph {nodes [
   RigVMFunctionEntryNode()
   > RigVMFunction_Sequence = RigVMAggregateNode()
   n1 = &RigUnit_SphericalPoseReader(
    ActiveRegionScaleFactors=(PositiveWidth=4,NegativeWidth=1,PositiveHeight=4,NegativeHeight=4)
    ActiveRegionSize=0.1
    Debug=(bDrawDebug=false,DebugScale=25,DebugSegments=20,DebugThickness=0.25)
    DriverAxis=(X=-1,Y=0,Z=0)
    DriverItem=(Type=Bone,Name="thigh_l")
    FalloffRegionScaleFactors=(PositiveWidth=1,NegativeWidth=1,PositiveHeight=1,NegativeHeight=1)
    FalloffSize=0.48
    OptionalParentItem=(Type=Bone,Name="None")
    RotationOffset=(X=90,Y=0,Z=0)
    ExecutePin=@RigVMFunction_Sequence.A
   )
   > RigVMVariableNode(Value=@n1.OutputParam) {variable=Front Leg L}
   > n2 = &RigUnit_SphericalPoseReader(
    ActiveRegionScaleFactors=(PositiveWidth=3,NegativeWidth=0.1,PositiveHeight=1,NegativeHeight=1)
    ActiveRegionSize=0.1
    Debug=(bDrawDebug=false,DebugScale=25,DebugSegments=20,DebugThickness=0.25)
    DriverAxis=(X=-1,Y=0,Z=0)
    DriverItem=(Type=Bone,Name="thigh_l")
    FalloffRegionScaleFactors=(PositiveWidth=1,NegativeWidth=1,PositiveHeight=1,NegativeHeight=1.5)
    FalloffSize=0.45
    FlipHeightScaling=false
    FlipWidthScaling=false
    OptionalParentItem=(Type=Bone,Name="None")
    RotationOffset=(X=200,Y=20,Z=80)
   )
   > RigVMVariableNode(Value=@n2.OutputParam) {variable=Side Leg L}
   > n3 = &RigUnit_SphericalPoseReader(
    ActiveRegionScaleFactors=(PositiveWidth=3.5,NegativeWidth=0.5,PositiveHeight=3,NegativeHeight=3)
    ActiveRegionSize=0.1
    Debug=(bDrawDebug=false,DebugScale=25,DebugSegments=20,DebugThickness=0.25)
    DriverAxis=(X=-1,Y=0,Z=0)
    DriverItem=(Type=Bone,Name="thigh_twist_01_l")
    FalloffRegionScaleFactors=(PositiveWidth=1,NegativeWidth=1,PositiveHeight=1,NegativeHeight=1)
    FalloffSize=0.48
    OptionalParentItem=(Type=Bone,Name="None")
    RotationOffset=(X=-90,Y=0,Z=0)
   )
   > RigVMVariableNode(Value=@n3.OutputParam) {variable=Back Leg L}
   > n4 = &RigUnit_SphericalPoseReader(
    ActiveRegionScaleFactors=(PositiveWidth=3.5,NegativeWidth=0.25,PositiveHeight=3,NegativeHeight=2)
    ActiveRegionSize=0.1
    Debug=(bDrawDebug=false,DebugScale=25,DebugSegments=20,DebugThickness=0.25)
    DriverAxis=(X=1,Y=0,Z=0)
    DriverItem=(Type=Bone,Name="thigh_l")
    FalloffRegionScaleFactors=(PositiveWidth=1,NegativeWidth=1,PositiveHeight=1,NegativeHeight=1.075)
    FalloffSize=0.48
    OptionalParentItem=(Type=Bone,Name="None")
    RotationOffset=(X=180,Y=0,Z=0)
   )
   > RigVMVariableNode(Value=@n4.OutputParam) {variable=Set Inward Side Leg L}
   > n5 = &RigUnit_SphericalPoseReader(
    ActiveRegionScaleFactors=(PositiveWidth=3,NegativeWidth=3,PositiveHeight=3,NegativeHeight=3)
    ActiveRegionSize=0.1
    Debug=(bDrawDebug=false,DebugScale=25,DebugSegments=20,DebugThickness=0.25)
    DriverAxis=(X=1,Y=0,Z=0)
    DriverItem=(Type=Bone,Name="thigh_r")
    FalloffRegionScaleFactors=(PositiveWidth=1.2,NegativeWidth=1.2,PositiveHeight=1.2,NegativeHeight=1.2)
    FalloffSize=0.55
    FlipHeightScaling=false
    FlipWidthScaling=true
    OptionalParentItem=(Type=Bone,Name="pelvis")
    RotationOffset=(X=-90,Y=0,Z=-20)
   )
   > RigVMVariableNode(Value=@n5.OutputParam) {variable=Front Leg R}
   > n6 = &RigUnit_SphericalPoseReader(
    ActiveRegionScaleFactors=(PositiveWidth=3,NegativeWidth=1,PositiveHeight=1,NegativeHeight=1)
    ActiveRegionSize=0.1
    Debug=(bDrawDebug=false,bDraw2D=false,DebugScale=25,DebugSegments=20,DebugThickness=0.25)
    DriverAxis=(X=1,Y=0,Z=0)
    DriverItem=(Type=Bone,Name="thigh_r")
    FalloffRegionScaleFactors=(PositiveWidth=1,NegativeWidth=1,PositiveHeight=1,NegativeHeight=1.5)
    FalloffSize=0.45
    FlipHeightScaling=false
    FlipWidthScaling=true
    OptionalParentItem=(Type=Bone,Name="None")
    RotationOffset=(X=20,Y=20,Z=80)
   )
   > RigVMVariableNode(Value=@n6.OutputParam) {variable=Side Leg R}
   > n7 = &RigUnit_SphericalPoseReader(
    ActiveRegionScaleFactors=(PositiveWidth=3.5,NegativeWidth=0.5,PositiveHeight=3,NegativeHeight=3)
    ActiveRegionSize=0.1
    Debug=(bDrawDebug=false,DebugScale=25,DebugSegments=20,DebugThickness=0.25)
    DriverAxis=(X=-1,Y=0,Z=0)
    DriverItem=(Type=Bone,Name="thigh_twist_01_r")
    FalloffRegionScaleFactors=(PositiveWidth=1,NegativeWidth=1,PositiveHeight=1,NegativeHeight=1)
    FalloffSize=0.48
    OptionalParentItem=(Type=Bone,Name="pelvis")
    RotationOffset=(X=-90,Y=0,Z=0)
   )
   > RigVMVariableNode(Value=@n7.OutputParam) {variable=Back Leg R}
   > n8 = &RigUnit_SphericalPoseReader(
    ActiveRegionScaleFactors=(PositiveWidth=3.5,NegativeWidth=0.25,PositiveHeight=3,NegativeHeight=3)
    ActiveRegionSize=0.1
    Debug=(bDrawDebug=false,DebugScale=25,DebugSegments=20,DebugThickness=0.25)
    DriverAxis=(X=-1,Y=0,Z=0)
    DriverItem=(Type=Bone,Name="thigh_r")
    FalloffRegionScaleFactors=(PositiveWidth=1,NegativeWidth=1,PositiveHeight=1,NegativeHeight=1.075)
    FalloffSize=0.48
    FlipHeightScaling=false
    FlipWidthScaling=false
    OptionalParentItem=(Type=Bone,Name="None")
    RotationOffset=(X=0,Y=0,Z=0)
   )
   > RigVMVariableNode(Value=@n8.OutputParam) {variable=Set Inward Side Leg R}
   > &RigUnit_ParentConstraint(
    AdvancedSettings=(RotationOrderForFilter=XZY)
    Child=(Type=Bone,Name="base_skirt_l_C")
    Parents=((Item=(Type=Bone,Name="spine_01"),Weight=0.0109),(Item=(Type=Bone,Name="spine_02"),Weight=0.0217),(Item=(Type=Bone,Name="deform_glute_l"),Weight=0.067),(Item=(Type=Bone,Name="pelvis"),Weight=0.9004))
    Weight=1
    bMaintainOffset=true
   )
   > &RigUnit_ParentConstraint(
    AdvancedSettings=(RotationOrderForFilter=XZY)
    Child=(Type=Bone,Name="base_skirt_r_C")
    Parents=((Item=(Type=Bone,Name="thigh_r"),Weight=0.0063),(Item=(Type=Bone,Name="spine_01"),Weight=0.0108),(Item=(Type=Bone,Name="spine_02"),Weight=0.0216),(Item=(Type=Bone,Name="deform_glute_r"),Weight=0.0666),(Item=(Type=Bone,Name="pelvis"),Weight=0.8947))
    Weight=1
   )
   &RigUnit_ModifyTransforms(
    ItemToModify="((Item=(Type=Bone,Name=\"base_skirt_l_A\"),Transform=(Rotation=(X=-0.0653,Y=-0.113,Z=-0.4957,W=0.8586),Translation=(X=0,Y=1,Z=0),Scale3D=(X=0.75,Y=1,Z=1))),(Item=(Type=Bone,Name=\"base_skirt_l_B\"),Transform=(Rotation=(X=0.05,Y=-0.0714,Z=0.5714,W=0.816),Transl ...(truncated)"
    WeightMinimum=0
    ExecutePin=@RigVMFunction_Sequence.B
    Weight=RigVMVariableNode() {variable=Front Leg L}.Value
   )
   > &RigUnit_ParentConstraint(
    AdvancedSettings=(InterpolationType=Average,RotationOrderForFilter=XZY)
    Child=(Type=Bone,Name="base_skirt_l_B")
    Filter=(TranslationFilter=(bX=false,bY=false,bZ=false),RotationFilter=(bX=True,bY=True,bZ=True),ScaleFilter=(bX=True,bY=True,bZ=True))
    Parents=((Item=(Type=Bone,Name="thigh_l"),Weight=1))
    Weight=0.3
   )
   > &RigUnit_ParentConstraint(
    AdvancedSettings=(InterpolationType=Average,RotationOrderForFilter=XZY)
    Child=(Type=Bone,Name="base_skirt_l_A")
    Filter=(TranslationFilter=(bX=false,bY=false,bZ=false),RotationFilter=(bX=True,bY=True,bZ=True),ScaleFilter=(bX=True,bY=True,bZ=True))
    Parents=((Item=(Type=Bone,Name="thigh_l"),Weight=1))
    Weight=0.2
   )
   > &RigUnit_ModifyTransforms(
    ItemToModify="((Item=(Type=Bone,Name=\"base_skirt_l_A\"),Transform=(Rotation=(X=-0.0216,Y=0.555,Z=-0.2588,W=0.7903),Translation=(X=0,Y=0,Z=0),Scale3D=(X=1,Y=1,Z=1))),(Item=(Type=Bone,Name=\"base_skirt_l_B\"),Transform=(Rotation=(X=0.0755,Y=-0.4981,Z=0.0436,W=0.8627),Transla ...(truncated)"
    WeightMinimum=0
    Weight=RigVMVariableNode() {variable=Side Leg L}.Value
   )
   > &RigUnit_ModifyTransforms(
    ItemToModify="((Item=(Type=Bone,Name=\"base_skirt_l_A\"),Transform=(Rotation=(X=-0.0114,Y=-0.13,Z=-0.0864,W=0.9877),Translation=(X=0,Y=0,Z=0),Scale3D=(X=1,Y=1,Z=1))),(Item=(Type=Bone,Name=\"base_skirt_r_A\"),Transform=(Rotation=(X=-0.0057,Y=-0.1304,Z=-0.0432,W=0.9905),Trans ...(truncated)"
    WeightMinimum=0
    Weight=RigVMVariableNode() {variable=Set Inward Side Leg L}.Value
   )
   Multiply = RigVMAggregateNode(
    C=&RigVMFunction_MathBoolToFloat(Value=RigVMVariableNode() {variable=Get Crouching}.Value).Result
    A=RigVMVariableNode() {variable=Front Leg L}.Value
    B=RigVMVariableNode() {variable=Set Inward Side Leg L}.Value
   )
   > &RigUnit_ModifyTransforms(
    ItemToModify="((Item=(Type=Bone,Name=\"base_skirt_l_A\"),Transform=(Rotation=(X=-0.0283,Y=-0.2146,Z=-0.1274,W=0.9679),Translation=(X=0,Y=0,Z=0),Scale3D=(X=1,Y=1,Z=1))),(Item=(Type=Bone,Name=\"base_skirt_r_A\"),Transform=(Rotation=(X=-0.0057,Y=-0.1304,Z=-0.0432,W=0.9905),Tra ...(truncated)"
    WeightMinimum=0
    Weight=@Multiply.Result
   )
   > &RigUnit_ModifyTransforms(
    ItemToModify="((Item=(Type=Bone,Name=\"base_skirt_l_C\"),Transform=(Rotation=(X=0,Y=0,Z=0.766,W=0.6428),Translation=(X=-1,Y=0,Z=0),Scale3D=(X=1,Y=1,Z=1))),(Item=(Type=Bone,Name=\"base_skirt_l_B\"),Transform=(Rotation=(X=-0.0603,Y=-0.1158,Z=-0.4578,W=0.8794),Translation=(X=0 ...(truncated)"
    WeightMinimum=0
    Weight=RigVMVariableNode() {variable=Back Leg L}.Value
   )
   > n9 = &RigUnit_SphericalPoseReader(
    ActiveRegionScaleFactors=(PositiveWidth=1,NegativeWidth=1,PositiveHeight=1,NegativeHeight=1)
    ActiveRegionSize=0.075
    Debug=(bDrawDebug=false,bDraw2D=False,bDrawLocalAxes=False,DebugScale=25,DebugSegments=20,DebugThickness=0.25)
    DriverAxis=(X=-1,Y=0,Z=0)
    DriverItem=(Type=Bone,Name="thigh_l")
    FalloffRegionScaleFactors=(PositiveWidth=1,NegativeWidth=1,PositiveHeight=1,NegativeHeight=1)
    FalloffSize=0.175
    FlipHeightScaling=False
    FlipWidthScaling=False
    OptionalParentItem=(Type=Bone,Name="None")
    RotationOffset=(X=140,Y=10,Z=0)
   )
   > &RigUnit_ModifyTransforms(
    ItemToModify="((Item=(Type=Bone,Name=\"base_skirt_l_B\"),Transform=(Rotation=(X=0,Y=0.3338,Z=0,W=0.9426),Translation=(X=0,Y=0,Z=0),Scale3D=(X=1,Y=1,Z=1))),(Item=(Type=Bone,Name=\"base_skirt_l_A\"),Transform=(Rotation=(X=0,Y=-0.1736,Z=0,W=0.9848),Translation=(X=0,Y=0,Z=0),Sc ...(truncated)"
    WeightMinimum=0
    Weight=@n9.OutputParam
   )
   &RigUnit_ModifyTransforms(
    ItemToModify="((Item=(Type=Bone,Name=\"base_skirt_r_A\"),Transform=(Rotation=(X=0.0549,Y=0.1176,Z=-0.419,W=0.8986),Translation=(X=-2.5328,Y=1.1802,Z=2),Scale3D=(X=1,Y=1,Z=1))),(Item=(Type=Bone,Name=\"base_skirt_r_B\"),Transform=(Rotation=(X=0,Y=0,Z=-0.5,W=0.866),Translation ...(truncated)"
    WeightMinimum=0
    ExecutePin=@RigVMFunction_Sequence.C
    Weight=RigVMVariableNode() {variable=Front Leg R}.Value
   )
   > &RigUnit_ParentConstraint(
    AdvancedSettings=(InterpolationType=Average,RotationOrderForFilter=XZY)
    Child=(Type=Bone,Name="base_skirt_r_B")
    Filter=(TranslationFilter=(bX=false,bY=false,bZ=false),RotationFilter=(bX=True,bY=True,bZ=True),ScaleFilter=(bX=True,bY=True,bZ=True))
    Parents=((Item=(Type=Bone,Name="thigh_r"),Weight=1))
    Weight=0.3
   )
   > &RigUnit_ParentConstraint(
    AdvancedSettings=(InterpolationType=Average,RotationOrderForFilter=XZY)
    Child=(Type=Bone,Name="base_skirt_r_A")
    Filter=(TranslationFilter=(bX=false,bY=false,bZ=false),RotationFilter=(bX=True,bY=True,bZ=True),ScaleFilter=(bX=True,bY=True,bZ=True))
    Parents=((Item=(Type=Bone,Name="thigh_r"),Weight=1))
    Weight=0.2
   )
   > &RigUnit_ModifyTransforms(
    ItemToModify="((Item=(Type=Bone,Name=\"base_skirt_r_A\"),Transform=(Rotation=(X=-0.25,Y=-0.433,Z=-0.433,W=0.75),Translation=(X=0,Y=0,Z=2),Scale3D=(X=1,Y=1,Z=1))),(Item=(Type=Bone,Name=\"base_skirt_r_B\"),Transform=(Rotation=(X=-0.0946,Y=-0.6368,Z=-0.0893,W=0.76),Translation ...(truncated)"
    WeightMinimum=0
    Weight=RigVMVariableNode() {variable=Side Leg R}.Value
   )
   > &RigUnit_ModifyTransforms(
    ItemToModify="((Item=(Type=Bone,Name=\"base_skirt_r_A\"),Transform=(Rotation=(X=0.0114,Y=0.13,Z=-0.0864,W=0.9877),Translation=(X=0,Y=0,Z=0),Scale3D=(X=1,Y=1,Z=1))),(Item=(Type=Bone,Name=\"base_skirt_l_A\"),Transform=(Rotation=(X=0.0057,Y=0.1304,Z=-0.0432,W=0.9905),Translati ...(truncated)"
    WeightMinimum=0
    Weight=RigVMVariableNode() {variable=Set Inward Side Leg R}.Value
   )
   Multiply_1 = RigVMAggregateNode(
    C=&RigVMFunction_MathBoolToFloat(Value=RigVMVariableNode() {variable=Get Crouching}.Value).Result
    B=RigVMVariableNode() {variable=Set Inward Side Leg R}.Value
    A=RigVMVariableNode() {variable=Front Leg R}.Value
   )
   > &RigUnit_ModifyTransforms(
    ItemToModify="((Item=(Type=Bone,Name=\"base_skirt_r_A\"),Transform=(Rotation=(X=0.0158,Y=0.2583,Z=-0.059,W=0.9641),Translation=(X=0,Y=0,Z=-2.5),Scale3D=(X=1,Y=1,Z=1))),(Item=(Type=Bone,Name=\"base_skirt_l_A\"),Transform=(Rotation=(X=0.003,Y=0.0697,Z=-0.0435,W=0.9966),Transl ...(truncated)"
    WeightMinimum=0
    Weight=@Multiply_1.Result
   )
   > &RigUnit_ModifyTransforms(
    ItemToModify="((Item=(Type=Bone,Name=\"base_skirt_r_C\"),Transform=(Rotation=(X=0,Y=0,Z=0.766,W=0.6428),Translation=(X=-1,Y=0,Z=0),Scale3D=(X=1,Y=1,Z=1))),(Item=(Type=Bone,Name=\"base_skirt_r_B\"),Transform=(Rotation=(X=0.0802,Y=-0.154,Z=0.4547,W=0.8735),Translation=(X=0,Y= ...(truncated)"
    WeightMinimum=0
    Weight=RigVMVariableNode() {variable=Back Leg R}.Value
   )
   > n10 = &RigUnit_SphericalPoseReader(
    ActiveRegionScaleFactors=(PositiveWidth=1,NegativeWidth=1,PositiveHeight=1,NegativeHeight=1)
    ActiveRegionSize=0.075
    Debug=(bDrawDebug=false,bDraw2D=False,bDrawLocalAxes=False,DebugScale=25,DebugSegments=20,DebugThickness=0.25)
    DriverAxis=(X=1,Y=0,Z=0)
    DriverItem=(Type=Bone,Name="thigh_r")
    FalloffRegionScaleFactors=(PositiveWidth=1,NegativeWidth=1,PositiveHeight=1,NegativeHeight=1)
    FalloffSize=0.175
    FlipHeightScaling=False
    FlipWidthScaling=False
    OptionalParentItem=(Type=Bone,Name="None")
    RotationOffset=(X=-40,Y=10,Z=0)
   )
   > &RigUnit_ModifyTransforms(
    ItemToModify="((Item=(Type=Bone,Name=\"base_skirt_r_B\"),Transform=(Rotation=(X=0,Y=0.342,Z=0,W=0.9397),Translation=(X=0,Y=0,Z=0),Scale3D=(X=1,Y=1,Z=1))),(Item=(Type=Bone,Name=\"base_skirt_r_A\"),Transform=(Rotation=(X=0,Y=0.1736,Z=0,W=0.9848),Translation=(X=0,Y=0,Z=0),Scal ...(truncated)"
    WeightMinimum=0
    Weight=@n10.OutputParam
   )
   > &RigUnit_ModifyTransforms(
    ItemToModify=((Item=(Type=Bone,Name="ALT_fix"),Transform=(Rotation=(X=0,Y=0,Z=0,W=1),Translation=(X=0,Y=0,Z=0),Scale3D=(X=0.2,Y=0.2,Z=0.2))))
    Weight=1
    WeightMinimum=0
   )
   RigVMFunctionReturnNode(ExecuteContext=@RigVMFunction_Sequence.F)
   n11 = RigVMVariableNode() {variable=Front Leg L}
   n12 = &RigVMFunction_MathDoubleRemap(SourceMaximum=1, SourceMinimum=0.45, TargetMaximum=1, TargetMinimum=0, Value=@n11.Value)
   &RigUnit_ParentConstraint(
    AdvancedSettings=(InterpolationType=Average,RotationOrderForFilter=XZY)
    Child=(Type=Bone,Name="base_skirt_l_A")
    Parents=((Item=(Type=Bone,Name="thigh_l"),Weight=1))
    ExecutePin=@RigVMFunction_Sequence.E
    Weight=@n12.Result
   )
   > &RigUnit_ModifyTransforms(
    ItemToModify="((Item=(Type=Bone,Name=\"base_skirt_l_A\"),Transform=(Rotation=(X=-0.0927,Y=-0.0136,Z=-0.0118,W=0.9955),Translation=(X=9.9983,Y=0.003,Z=0.012),Scale3D=(X=1,Y=1,Z=1))),(Item=(Type=Bone,Name=\"base_skirt_r_A\"),Transform=(Rotation=(X=0,Y=0,Z=-0.0436,W=0.999),Tra ...(truncated)"
    WeightMinimum=0
    Weight=@n12.Result
   )
   > &RigUnit_ParentConstraint(
    AdvancedSettings=(InterpolationType=Average,RotationOrderForFilter=XZY)
    Child=(Type=Bone,Name="base_skirt_l_B")
    Parents=((Item=(Type=Bone,Name="thigh_l"),Weight=1))
    Weight=@n12.Result
   )
   > &RigUnit_ModifyTransforms(
    ItemToModify="((Item=(Type=Bone,Name=\"base_skirt_l_B\"),Transform=(Rotation=(X=-0.174,Y=0.0132,Z=-0.174,W=0.9692),Translation=(X=1,Y=3,Z=1),Scale3D=(X=1,Y=1,Z=1))),(Item=(Type=Bone,Name=\"base_skirt_l_C\"),Transform=(Rotation=(X=0,Y=0,Z=0.0872,W=0.9962),Translation=(X=-2,Y ...(truncated)"
    WeightMinimum=0
    Weight=@n12.Result
   )
   n13 = RigVMVariableNode() {variable=Front Leg R}
   n14 = &RigVMFunction_MathDoubleRemap(SourceMaximum=1, SourceMinimum=0.45, TargetMaximum=1, TargetMinimum=0, Value=@n13.Value)
   > &RigUnit_ParentConstraint(
    AdvancedSettings=(InterpolationType=Average,RotationOrderForFilter=XZY)
    Child=(Type=Bone,Name="base_skirt_r_A")
    Parents=((Item=(Type=Bone,Name="thigh_twist_01_r"),Weight=1))
    Weight=@n14.Result
   )
   > &RigUnit_ModifyTransforms(
    ItemToModify="((Item=(Type=Bone,Name=\"base_skirt_r_A\"),Transform=(Rotation=(X=-0.0041,Y=0.0142,Z=0.0173,W=0.9997),Translation=(X=10.5157,Y=0.2805,Z=0.2326),Scale3D=(X=1,Y=1,Z=1.2))),(Item=(Type=Bone,Name=\"base_skirt_l_A\"),Transform=(Rotation=(X=0,Y=0,Z=-0.0436,W=0.999), ...(truncated)"
    WeightMinimum=0
    Weight=@n14.Result
   )
   > &RigUnit_ParentConstraint(
    AdvancedSettings=(InterpolationType=Average,RotationOrderForFilter=XZY)
    Child=(Type=Bone,Name="base_skirt_r_B")
    Parents=((Item=(Type=Bone,Name="thigh_twist_01_r"),Weight=0.8),(Item=(Type=Bone,Name="pelvis"),Weight=0.2))
    Weight=@n14.Result
   )
   > &RigUnit_ModifyTransforms(
    ItemToModify="((Item=(Type=Bone,Name=\"base_skirt_r_B\"),Transform=(Rotation=(X=0.0872,Y=0,Z=0,W=0.9962),Translation=(X=2,Y=-2,Z=2),Scale3D=(X=1,Y=1,Z=1))),(Item=(Type=Bone,Name=\"base_skirt_r_C\"),Transform=(Rotation=(X=0,Y=0,Z=0.0872,W=0.9962),Translation=(X=-2,Y=0,Z=0),S ...(truncated)"
    WeightMinimum=0
    Weight=@n14.Result
   )
   n15 = RigVMVariableNode() {variable=Front Leg L}
   n16 = RigVMVariableNode() {variable=Front Leg R}
   &RigUnit_ModifyTransforms(
    ItemToModify=((Item=(Type=Bone,Name="base_skirt_r_A"),Transform=(Rotation=(X=-0.0858,Y=0.0151,Z=-0.173,W=0.9811),Translation=(X=0,Y=0,Z=0),Scale3D=(X=1,Y=1,Z=1))))
    WeightMinimum=0
    Weight=&RigVMFunction_MathDoubleSub(A=@n15.Value, B=@n16.Value).Result
    ExecutePin=@RigVMFunction_Sequence.D
   )
   > &RigUnit_ModifyTransforms(
    ItemToModify=((Item=(Type=Bone,Name="base_skirt_l_A"),Transform=(Rotation=(X=0.0858,Y=-0.0151,Z=-0.173,W=0.9811),Translation=(X=0,Y=0,Z=0),Scale3D=(X=1,Y=1,Z=1))))
    WeightMinimum=0
    Weight=&RigVMFunction_MathDoubleSub(B=@n15.Value, A=@n16.Value).Result
   )
   RigVMFunction_MathBoolToFloat = &RigVMFunction_MathBoolToFloat(Value=RigVMVariableNode() {variable=Get Crouching}.Value)
   > &RigUnit_ModifyTransforms(
    ItemToModify=((Item=(Type=Bone,Name="base_skirt_l_A"),Transform=(Rotation=(X=0,Y=0,Z=-0.0087,W=1),Translation=(X=0,Y=0,Z=0),Scale3D=(X=1,Y=1,Z=1))))
    WeightMinimum=0
    Weight=@RigVMFunction_MathBoolToFloat.Result
   )
   &RigVMFunction_MathDoubleRemap(SourceMaximum=1, SourceMinimum=0.7, TargetMaximum=1, TargetMinimum=0, Value=@n13.Value)
   &RigVMFunction_MathDoubleRemap(SourceMaximum=1, SourceMinimum=0.7, TargetMaximum=1, TargetMinimum=0, Value=@n11.Value)
  ]}
 New Function_ContainedGraph {nodes [
   RigVMFunctionEntryNode()
   > RigVMFunction_Sequence = RigVMAggregateNode()
   Remap = &RigVMFunction_MathDoubleRemap(SourceMaximum=1, SourceMinimum=0, TargetMaximum=0.4, TargetMinimum=0, bClamp=False, Value=RigVMVariableNode() {variable=Emote Float}.Value)
   &RigUnit_AimConstraintLocalSpaceOffset(
    Child=(Type=Bone,Name="base_upperarm_twist_01_extra_l")
    Filter=(bX=True,bY=false,bZ=false)
    Parents=((Item=(Type=Bone,Name="base_upperarm_twist_01_aim_l"),Weight=1))
    WorldUp=(Target=(X=0,Y=0,Z=1),Kind=Direction,Space=(Type=None,Name="None"))
    Weight=@Remap.Result
    ExecutePin=@RigVMFunction_Sequence.A
   )
   > &RigUnit_ModifyTransforms(
    ItemToModify="((Item=(Type=Bone,Name=\"upperarm_twist_01_extra_l\"),Transform=(Rotation=(X=0,Y=0,Z=0,W=1),Translation=(X=7.5579,Y=-1.5,Z=0),Scale3D=(X=1,Y=1,Z=1))),(Item=(Type=Bone,Name=\"base_armArmour_01\"),Transform=(Rotation=(X=0,Y=0,Z=0,W=1),Translation=(X=0,Y=0,Z=0),S ...(truncated)"
    WeightMinimum=0
    Weight=&RigVMFunction_MathDoubleMul(
     A=RigVMDispatchNode(BoneName=root, Default=0, Name=deformRig_l_upperarm_up_alpha) {template="DISPATCH_RigDispatch_GetAnimAttribute(in Name,in BoneName,in Default,out Value,out Found)"}.Value
     B=RigVMVariableNode() {variable=Emote Float}.Value
    ).Result
   )
   > &RigUnit_ModifyTransforms(
    ItemToModify="((Item=(Type=Bone,Name=\"upperarm_twist_01_extra_l\"),Transform=(Rotation=(X=0,Y=0,Z=0,W=1),Translation=(X=9.5833,Y=-2.8496,Z=0.8652),Scale3D=(X=1,Y=1,Z=1))),(Item=(Type=Bone,Name=\"base_armArmour_01\"),Transform=(Rotation=(X=0,Y=0,Z=0,W=1),Translation=(X=0,Y= ...(truncated)"
    WeightMinimum=0
    Weight=&RigVMFunction_MathDoubleMul(
     A=RigVMDispatchNode(BoneName=root, Default=0, Name=deformRig_l_upperarm_fr_alpha) {template="DISPATCH_RigDispatch_GetAnimAttribute(in Name,in BoneName,in Default,out Value,out Found)"}.Value
     B=RigVMVariableNode() {variable=Emote Float}.Value
    ).Result
   )
   Get_Animation_Attribute_1 = RigVMDispatchNode(BoneName=None, Default=0, Name=deformRig_l_upperarm_bk_alpha) {template="DISPATCH_RigDispatch_GetAnimAttribute(in Name,in BoneName,in Default,out Value,out Found)"}
   > &RigUnit_ModifyTransforms(
    ItemToModify="((Item=(Type=Bone,Name=\"upperarm_twist_01_extra_l\"),Transform=(Rotation=(X=0,Y=0,Z=0,W=1),Translation=(X=7.9852,Y=-2.2962,Z=0),Scale3D=(X=1,Y=1,Z=1))),(Item=(Type=Bone,Name=\"base_armArmour_01\"),Transform=(Rotation=(X=0,Y=0,Z=0,W=1),Translation=(X=0,Y=0,Z=0 ...(truncated)"
    WeightMinimum=0
    Weight=&RigVMFunction_MathDoubleMul(A=@Get_Animation_Attribute_1.Value, B=RigVMVariableNode() {variable=Emote Float}.Value).Result
   )
   n1 = &RigVMFunction_MathDoubleRemap(SourceMaximum=1, SourceMinimum=0, TargetMaximum=0, TargetMinimum=1, bClamp=False, Value=RigVMVariableNode() {variable=Emote Float}.Value)
   n2 = &RigUnit_ParentConstraint(
    AdvancedSettings=(InterpolationType=Average,RotationOrderForFilter=XZY)
    Child=(Type=Bone,Name="upperarm_twist_01_extra_l")
    Filter=(TranslationFilter=(bX=True,bY=True,bZ=True),RotationFilter=(bX=True,bY=True,bZ=True),ScaleFilter=(bX=false,bY=false,bZ=false))
    Parents=((Item=(Type=Bone,Name="deform_upperarm_up_l"),Weight=1),(Item=(Type=Bone,Name="upperarm_twist_01_l"),Weight=1),(Item=(Type=Bone,Name="clavicle_l"),Weight=2))
    Weight=@n1.Result
    ExecutePin=@RigVMFunction_Sequence.B
   )
   RigVMFunction_ControlFlowBranch = &RigVMFunction_ControlFlowBranch(ExecuteContext=@n2.ExecutePin, Condition=RigVMVariableNode() {variable=Get DBNO}.Value)
   n3 = RigVMDispatchNode(BoneName=root, Default=0, Name=deformRig_l_upperarm_fr_alpha) {template="DISPATCH_RigDispatch_GetAnimAttribute(in Name,in BoneName,in Default,out Value,out Found)"}
   Get_Animation_Attribute = RigVMDispatchNode(BoneName=None, Default=0, Name=deformRig_l_upperarm_bk_alpha) {template="DISPATCH_RigDispatch_GetAnimAttribute(in Name,in BoneName,in Default,out Value,out Found)"}
   Add = &RigVMFunction_MathDoubleAdd(A=@n3.Value, B=@Get_Animation_Attribute.Value)
   Multiply = &RigVMFunction_MathDoubleMul(B=-20, A=@n3.Value)
   RigVMFunction_MathQuaternionFromEuler = &RigVMFunction_MathQuaternionFromEuler(
    Euler=(X=0,Y=-25,Z=20)
    Euler.X=&RigVMFunction_MathDoubleAdd(A=@Multiply.Result, B=&RigVMFunction_MathDoubleMul(B=20, A=@Get_Animation_Attribute.Value).Result).Result
   )
   &RigUnit_ModifyTransforms(
    ItemToModify=((Item=(Type=Bone,Name="upperarm_twist_01_extra_l"),Transform=(Rotation=(X=0,Y=0,Z=0,W=1),Translation=(X=11,Y=6,Z=3),Scale3D=(X=0.85,Y=1.2,Z=0.9))))
    WeightMinimum=0
    Weight=&RigVMFunction_MathDoubleMul(A=@n1.Result, B=@Add.Result).Result
    ExecutePin=@RigVMFunction_ControlFlowBranch.False
    ItemToModify.0.Transform.Rotation=@RigVMFunction_MathQuaternionFromEuler.Result
   )
   Subtract = &RigVMFunction_MathDoubleSub(
    B=@n3.Value
    A=RigVMDispatchNode(BoneName=root, Default=0, Name=deformRig_l_upperarm_up_alpha) {template="DISPATCH_RigDispatch_GetAnimAttribute(in Name,in BoneName,in Default,out Value,out Found)"}.Value
   )
   > &RigUnit_ModifyTransforms(
    ItemToModify=((Item=(Type=Bone,Name="upperarm_twist_01_extra_l"),Transform=(Rotation=(X=0,Y=-0.1736,Z=0,W=0.9848),Translation=(X=5,Y=0,Z=2),Scale3D=(X=0.9,Y=0.9,Z=0.9))))
    WeightMinimum=0
    Weight=&RigVMFunction_MathDoubleMul(A=@n1.Result, B=@Subtract.Result).Result
   )
   RigVMFunctionReturnNode(ExecuteContext=@RigVMFunction_Sequence.C)
  ]}
 RigVMFunctionLibrary {nodes [
   Shoulder_Armour_Setup = RigVMCollapseNode(ExecuteContext=())
   Skirt_Setup = RigVMCollapseNode(ExecuteContext=())
  ]}
 RigVMFunction_Sequence_1_ContainedGraph {nodes [
   n1 = RigVMFunctionEntryNode()
   RigVMFunction_Sequence = &RigVMFunction_Sequence(ExecuteContext=@n1.ExecuteContext)
   n2 = &RigVMFunction_Sequence(ExecuteContext=@RigVMFunction_Sequence.B)
   RigVMFunctionReturnNode(A=@RigVMFunction_Sequence.A, B=@n2.A, C=@n2.B)
  ]}
 RigVMFunction_Sequence_1_ContainedGraph {nodes [
   n1 = RigVMFunctionEntryNode()
   RigVMFunction_Sequence = &RigVMFunction_Sequence(ExecuteContext=@n1.ExecuteContext)
   n2 = &RigVMFunction_Sequence(ExecuteContext=@RigVMFunction_Sequence.B)
   n3 = &RigVMFunction_Sequence(ExecuteContext=@n2.B)
   n4 = &RigVMFunction_Sequence(ExecuteContext=@n3.B)
   n5 = &RigVMFunction_Sequence(ExecuteContext=@n4.B)
   RigVMFunctionReturnNode(A=@RigVMFunction_Sequence.A, B=@n2.A, C=@n3.A, D=@n4.A, E=@n5.A, F=@n5.B)
  ]}
 RigVMFunction_Sequence_1_ContainedGraph {nodes [
   n1 = RigVMFunctionEntryNode()
   RigVMFunction_Sequence = &RigVMFunction_Sequence(ExecuteContext=@n1.ExecuteContext)
   n2 = &RigVMFunction_Sequence(ExecuteContext=@RigVMFunction_Sequence.B)
   RigVMFunctionReturnNode(A=@RigVMFunction_Sequence.A, B=@n2.A, C=@n2.B)
  ]}
 RigVMModel {nodes [
   RigUnit_BeginExecution = &RigUnit_BeginExecution()
   RigVMFunction_Sequence = RigVMAggregateNode(ExecuteContext=@RigUnit_BeginExecution.ExecutePin)
   FN_DeformRig_FNC = RigVMFunctionReferenceNode(
    DeformRig_ Curve="(deformRig_l_upperarm_up_alpha_72_#1=true,deformRig_r_upperarm_up_alpha_73_#2=False,deformRig_l_upperarm_fr_alpha_74_#3=true,deformRig_r_upperarm_fr_alpha_75_CB51B80C ...(truncated)"
    DeformRig_Bool="(upperarm_1_#4=True,elbow_3_#5=True,wrist_7_#6=True,groin_8_#7=True,glute_17_#8=True,knee_9_40C408324CA261 ...(truncated)"
    DeformRig_Debug=(upperarm_debug_18_#4=False,wrist_debug_20_#6=False,lat_debug_26_#9=False,groinGlute_debug_27_#7=False)
    DeformRig_Value="(upperarm_up_18_#10=(Rotation=(X=0.0,Y=-0.0,Z=0.0,W=1.0),Translation=(X=0.0,Y=0.0,Z=8.0),Scale3D=(X=0.70,Y=1.0,Z=1.0)),upperarm_fwd_fr_19_#11=(Rotation=(X=0.0,Y=-0.0,Z=0.0,W=1.0),Translation=(X=0.0, ...(truncated)"
    ExecuteContext=@RigVMFunction_Sequence.B
   )
   > Shoulder_Armour_Setup = RigVMFunctionReferenceNode()
   > Skirt_Setup = RigVMFunctionReferenceNode()
   &RigUnit_ModifyTransforms(
    ItemToModify=((Item=(Type=Bone,Name="base_collar_r"),Transform=(Rotation=(X=-0.0271,Y=0.1063,Z=-0.3009,W=0.9473),Translation=(X=-2.105,Y=-2.421,Z=-2.115),Scale3D=(X=1,Y=1,Z=1))))
    WeightMinimum=0
    Weight=RigVMDispatchNode(BoneName=root, Default=0, Name=deformRig_r_upperarm_fr_alpha) {template="DISPATCH_RigDispatch_GetAnimAttribute(in Name,in BoneName,in Default,out Value,out Found)"}.Value
    ExecutePin=@RigVMFunction_Sequence.C
   )
   RigVMFunction_ControlFlowBranch = &RigVMFunction_ControlFlowBranch(Condition=false, ExecuteContext=())
   RigVMCollapseNode(ExecuteContext=@RigVMFunction_ControlFlowBranch.True)
  ]}
]
hierarchy [
 root {loc=<0 0 0>, rot=<0 0 0 1>}
 pelvis {parent=root, loc=<0 -0.5301 98.0571>, rot=<0.0127 0.707 -0.0127 -0.707>}
 spine_01 {parent=pelvis, loc=<6.1208 -0.9328 0>, rot=<0 0 0.034 -0.9994>}
 spine_02 {parent=spine_01, loc=<6.5997 -0.2095 0>, rot=<0 0 0.0238 -0.9997>}
 spine_03 {parent=spine_02, loc=<6.4914 -0.1237 0>, rot=<0 0 0.0149 -0.9999>}
 spine_04 {parent=spine_03, loc=<6.9598 0.0794 0>, rot=<0 0 -0.0132 -0.9999>}
 spine_05 {parent=spine_04, loc=<6.9987 0 0>, rot=<0 0 -0.0015 -1>}
 attach_cape {parent=spine_05, loc=<9.3216 -8.4501 0>, rot=<0.0616 -0.7044 -0.0616 -0.7044>}
 neck_01 {parent=spine_05, loc=<11.3138 0.1529 0>, rot=<0 0 -0.2003 -0.9797>}
 neck_02 {parent=neck_01, loc=<4.5627 0 0>, rot=<0 0 0.0504 -0.9987>}
 head {parent=neck_02, loc=<5.3914 0 0>, rot=<0 0 0.115 -0.9934>}
 attach_backpack {parent=spine_05, loc=<0 -8.35 0>, rot=<-0.0093 -0.707 0.0093 -0.707>}
 clavicle_l {parent=spine_05, loc=<5.4831 -0.1053 -3.2946>, rot=<0.0465 -0.7289 -0.009 -0.6829>}
 upperarm_l {parent=clavicle_l, loc=<9.8658 -1.2136 -0.4105>, rot=<0.0164 -0.3245 -0.0269 -0.9454>}
 lowerarm_l {parent=upperarm_l, loc=<26.465 -0.8345 0.259>, rot=<0.0036 0.0161 -0.2298 -0.9731>}
 hand_l {parent=lowerarm_l, loc=<22.7355 -0.3203 -0.6105>, rot=<-0.6784 -0.0223 0.0735 -0.7307>}
 thumb_01_l {parent=hand_l, loc=<1.3118 -0.3191 -2.459>, rot=<0.6546 -0.4042 0.0664 -0.6354>}
 thumb_02_l {parent=thumb_01_l, loc=<5.9336 -0.0448 0.0401>, rot=<-0.0037 0.0213 0.298 -0.9543>}
 thumb_03_l {parent=thumb_02_l, loc=<2.2223 0.3682 -0.0907>, rot=<0.0118 -0.013 0.0066 -0.9998>}
 index_metacarpal_l {parent=hand_l, loc=<1.6535 0.0183 -2.96>, rot=<0.0158 -0.0124 -0.0173 -0.9996>}
 index_01_l {parent=index_metacarpal_l, loc=<8.3008 0.0357 -0.0524>, rot=<-0.0044 0.0485 0.0924 -0.9945>}
 index_02_l {parent=index_01_l, loc=<3.3587 -0.3004 0.0452>, rot=<-0.002 0.0027 0.3225 -0.9466>}
 index_03_l {parent=index_02_l, loc=<2.5578 -0.123 -0.0252>, rot=<-0.002 0.0067 0.2558 -0.9667>}
 middle_metacarpal_l {parent=hand_l, loc=<2.1856 -0.1725 -0.5984>, rot=<0.006 0.0168 -0.0329 -0.9993>}
 middle_01_l {parent=middle_metacarpal_l, loc=<7.3275 -0.2447 0.0129>, rot=<-0.0208 0.0042 0.0858 -0.9961>}
 middle_02_l {parent=middle_01_l, loc=<4.0123 0.1401 0.1175>, rot=<-0.0218 0.0258 0.3512 -0.9357>}
 middle_03_l {parent=middle_02_l, loc=<2.6077 0.0534 0.1373>, rot=<0.0082 0.0188 0.1708 -0.9851>}
 ring_metacarpal_l {parent=hand_l, loc=<2.2074 -0.3625 1.3638>, rot=<-0.0284 0.0722 -0.0458 -0.9959>}
 ring_01_l {parent=ring_metacarpal_l, loc=<7.0295 -0.6197 -0.1446>, rot=<-0.0489 -0.0252 0.1892 -0.9804>}
 ring_02_l {parent=ring_01_l, loc=<3.6174 0.4647 -0.1027>, rot=<0 0 0.2806 -0.9598>}
 ring_03_l {parent=ring_02_l, loc=<2.4152 0.114 0.0199>, rot=<0 0 0.1189 -0.9929>}
 pinky_metacarpal_l {parent=hand_l, loc=<1.7616 -0.9753 3.3814>, rot=<-0.0397 0.0814 -0.069 -0.9935>}
 pinky_01_l {parent=pinky_metacarpal_l, loc=<7.1709 -0.4831 -0.1094>, rot=<-0.0654 -0.0035 0.2242 -0.9723>}
 pinky_02_l {parent=pinky_01_l, loc=<2.8274 0.0893 -0.1043>, rot=<0.0037 0.0011 0.2854 -0.9584>}
 pinky_03_l {parent=pinky_02_l, loc=<2.2046 0.0307 -0.0078>, rot=<-0.0002 0.0098 0.1445 -0.9895>}
 deform_wrist_dn_base_l {parent=hand_l, loc=<0 -1.5326 0>, rot=<0 0 0 -1>}
 deform_wrist_dn_l {parent=deform_wrist_dn_base_l, loc=<0 0 0>, rot=<0 0 0 1>}
 deform_wrist_up_base_l {parent=hand_l, loc=<0 2.3009 0>, rot=<0 0 0 -1>}
 deform_wrist_up_l {parent=deform_wrist_up_base_l, loc=<0 0 0>, rot=<0 0 0 1>}
 lowerarm_twist_02_l {parent=lowerarm_l, loc=<10.6707 0 0>, rot=<0 0 0 1>}
 lowerarm_twist_01_l {parent=lowerarm_l, loc=<18.316 0 0>, rot=<0 0 0 1>}
 deform_elbow_in_base_l {parent=lowerarm_l, loc=<0.1834 3.6062 -0.2572>, rot=<0 0 0 -1>}
 deform_elbow_in_l {parent=deform_elbow_in_base_l, loc=<0 0 0>, rot=<0 0 0 1>}
 deform_elbow_out_base_l {parent=lowerarm_l, loc=<0 -4 0>, rot=<0 0 0 1>}
 deform_elbow_out_l {parent=deform_elbow_out_base_l, loc=<0 0 0>, rot=<0 0 0 1>}
 upperarm_twist_01_l {parent=upperarm_l, loc=<7.2135 -0.1134 0>, rot=<0 0 0 1>}
 base_upperarm_twist_01_extra_l {parent=upperarm_twist_01_l, loc=<-7.2135 0.1134 0>, rot=<0 0 0 1>}
 upperarm_twist_01_extra_l {parent=base_upperarm_twist_01_extra_l, loc=<0 0 0>, rot=<0 0 0 1>}
 base_armourString {parent=upperarm_twist_01_extra_l, loc=<-0.8927 10.0984 1.1732>, rot=<0.2938 -0.2639 -0.5751 -0.7165>}
 dyn_armourString {parent=base_armourString, loc=<0 0 0>, rot=<0 0 0 1>}
 dyn_armourStringBow_l {parent=dyn_armourString, loc=<2.7074 -0.5279 -0.0101>, rot=<-0.3454 0.0371 0.3247 -0.8797>}
 dyn_armourString_l_01 {parent=dyn_armourString, loc=<2.5929 -0.4343 -0.3831>, rot=<-0.5045 -0.0674 0.3478 -0.7874>}
 dyn_armourString_l_02 {parent=dyn_armourString_l_01, loc=<3.9646 0.0267 0.0304>, rot=<0.0018 0.0237 0.049 -0.9985>}
 dyn_armourStringBow_r {parent=dyn_armourString, loc=<2.938 0.3165 -0.2216>, rot=<0.8243 -0.0542 -0.0766 -0.5584>}
 dyn_armourString_r_01 {parent=dyn_armourString, loc=<2.6093 0.0517 -0.503>, rot=<0.7585 -0.3159 -0.1057 -0.5601>}
 dyn_armourString_r_02 {parent=dyn_armourString_r_01, loc=<2.7864 0 -0.0038>, rot=<0 -0.0835 0 -0.9965>}
 base_armArmour_01 {parent=upperarm_twist_01_extra_l, loc=<-0.5461 -0.3554 5.6844>, rot=<-0.072 0.1106 0.0268 -0.9909>}
 dyn_armArmour_01 {parent=base_armArmour_01, loc=<0 0 0>, rot=<0 0 0 1>}
 base_armArmour_02 {parent=dyn_armArmour_01, loc=<4.7368 -0.073 0.0943>, rot=<0.0161 -0.0853 0.0055 -0.9962>}
 dyn_armArmour_02 {parent=base_armArmour_02, loc=<0 0 0>, rot=<0 0 0 1>}
 base_armArmour_03 {parent=dyn_armArmour_02, loc=<4.6486 0.0088 0.0586>, rot=<0.0022 -0.0833 -0.0262 -0.9962>}
 dyn_armArmour_03 {parent=base_armArmour_03, loc=<0 0 0>, rot=<0 0 0 1>}
 deform_upperarm_bk_base_l {parent=upperarm_twist_01_l, loc=<-6.4053 -4.3746 0.314>, rot=<0 0 0 -1>}
 deform_upperarm_bk_l {parent=deform_upperarm_bk_base_l, loc=<0 0 0>, rot=<0 0 0 1>}
 deform_upperarm_fr_base_l {parent=upperarm_twist_01_l, loc=<-5.6312 5.092 0.3354>, rot=<0 0 0 -1>}
 deform_upperarm_fr_l {parent=deform_upperarm_fr_base_l, loc=<0 0 0>, rot=<0 0 0 1>}
 deform_upperarm_up_base_l {parent=upperarm_twist_01_l, loc=<-8.891 0.4825 3.9517>, rot=<0 0 0 -1>}
 deform_upperarm_up_l {parent=deform_upperarm_up_base_l, loc=<0 0 0>, rot=<0 0 0 1>}
 upperarm_twist_02_l {parent=upperarm_l, loc=<19.7998 -0.1134 0>, rot=<0 0 0 1>}
 base_upperarm_twist_01_aim_l {parent=clavicle_l, loc=<21.3162 -0.6299 -9.3397>, rot=<0.0164 -0.3245 -0.0269 -0.9454>}
 base_shoulder_pad_end_l {parent=clavicle_l, loc=<26.2384 -0.4106 -6.5903>, rot=<0.0062 -0.1792 -0.0244 -0.9835>}
 clavicle_r {parent=spine_05, loc=<5.4831 -0.1053 3.2946>, rot=<0.7289 0.0465 -0.6829 0.009>}
 upperarm_r {parent=clavicle_r, loc=<-9.8658 1.2136 0.4105>, rot=<0.0164 -0.3245 -0.0269 -0.9454>}
 lowerarm_r {parent=upperarm_r, loc=<-26.465 0.8345 -0.259>, rot=<0.0036 0.0161 -0.2298 -0.9731>}
 hand_r {parent=lowerarm_r, loc=<-22.7355 0.3203 0.6106>, rot=<-0.6784 -0.0223 0.0735 -0.7307>}
 thumb_01_r {parent=hand_r, loc=<-1.3118 0.3191 2.459>, rot=<0.6546 -0.4042 0.0664 -0.6354>}
 thumb_02_r {parent=thumb_01_r, loc=<-5.9336 0.0448 -0.0401>, rot=<-0.0037 0.0213 0.298 -0.9543>}
 thumb_03_r {parent=thumb_02_r, loc=<-2.2223 -0.3682 0.0907>, rot=<0.0118 -0.013 0.0066 -0.9998>}
 index_metacarpal_r {parent=hand_r, loc=<-1.6535 -0.0183 2.96>, rot=<0.0158 -0.0124 -0.0173 -0.9996>}
 index_01_r {parent=index_metacarpal_r, loc=<-8.3008 -0.0357 0.0524>, rot=<-0.0044 0.0485 0.0924 -0.9945>}
 index_02_r {parent=index_01_r, loc=<-3.3587 0.3004 -0.0452>, rot=<-0.002 0.0027 0.3225 -0.9466>}
 index_03_r {parent=index_02_r, loc=<-2.5578 0.123 0.0252>, rot=<-0.002 0.0067 0.2558 -0.9667>}
 middle_metacarpal_r {parent=hand_r, loc=<-2.1856 0.1725 0.5984>, rot=<0.006 0.0168 -0.0329 -0.9993>}
 middle_01_r {parent=middle_metacarpal_r, loc=<-7.3275 0.2447 -0.0129>, rot=<-0.0208 0.0042 0.0858 -0.9961>}
 middle_02_r {parent=middle_01_r, loc=<-4.0123 -0.1401 -0.1175>, rot=<-0.0218 0.0258 0.3512 -0.9357>}
 middle_03_r {parent=middle_02_r, loc=<-2.6077 -0.0534 -0.1373>, rot=<0.0082 0.0188 0.1708 -0.9851>}
 ring_metacarpal_r {parent=hand_r, loc=<-2.2074 0.3625 -1.3638>, rot=<-0.0284 0.0722 -0.0458 -0.9959>}
 ring_01_r {parent=ring_metacarpal_r, loc=<-7.0295 0.6197 0.1446>, rot=<-0.0489 -0.0252 0.1892 -0.9804>}
 ring_02_r {parent=ring_01_r, loc=<-3.6174 -0.4647 0.1027>, rot=<0 0 0.2806 -0.9598>}
 ring_03_r {parent=ring_02_r, loc=<-2.4152 -0.114 -0.0199>, rot=<0 0 0.1189 -0.9929>}
 pinky_metacarpal_r {parent=hand_r, loc=<-1.7616 0.9753 -3.3814>, rot=<-0.0397 0.0814 -0.069 -0.9935>}
 pinky_01_r {parent=pinky_metacarpal_r, loc=<-7.1709 0.4831 0.1094>, rot=<-0.0654 -0.0035 0.2242 -0.9723>}
 pinky_02_r {parent=pinky_01_r, loc=<-2.8274 -0.0893 0.1043>, rot=<0.0037 0.0011 0.2854 -0.9584>}
 pinky_03_r {parent=pinky_02_r, loc=<-2.2046 -0.0307 0.0078>, rot=<-0.0002 0.0098 0.1445 -0.9895>}
 deform_wrist_dn_base_r {parent=hand_r, loc=<-0.0001 1.5328 0>, rot=<0 0 0 -1>}
 deform_wrist_dn_r {parent=deform_wrist_dn_base_r, loc=<0 0 0>, rot=<0 0 0 1>}
 deform_wrist_up_base_r {parent=hand_r, loc=<-0.0002 -2.3007 -0.0001>, rot=<0 0 0 -1>}
 deform_wrist_up_r {parent=deform_wrist_up_base_r, loc=<0 0 0>, rot=<0 0 0 1>}
 lowerarm_twist_02_r {parent=lowerarm_r, loc=<-10.6707 0 0>, rot=<0 0 0 -1>}
 lowerarm_twist_01_r {parent=lowerarm_r, loc=<-18.316 0 0>, rot=<0 0 0 -1>}
 deform_elbow_in_base_r {parent=lowerarm_r, loc=<-0.1834 -3.6062 0.2572>, rot=<0 0 0 -1>}
 deform_elbow_in_r {parent=deform_elbow_in_base_r, loc=<0 0 0>, rot=<0 0 0 1>}
 deform_elbow_out_base_r {parent=lowerarm_r, loc=<0 4 -0.0001>, rot=<0 0 0 -1>}
 deform_elbow_out_r {parent=deform_elbow_out_base_r, loc=<0 0 0>, rot=<0 0 0 1>}
 upperarm_twist_01_r {parent=upperarm_r, loc=<-7.2135 0.1134 0>, rot=<0 0 0 -1>}
 deform_upperarm_bk_base_r {parent=upperarm_twist_01_r, loc=<6.4053 4.3746 -0.314>, rot=<0 0 0 -1>}
 deform_upperarm_bk_r {parent=deform_upperarm_bk_base_r, loc=<0 0 0>, rot=<0 0 0 1>}
 deform_upperarm_fr_base_r {parent=upperarm_twist_01_r, loc=<5.6312 -5.092 -0.3354>, rot=<0 0 0 -1>}
 deform_upperarm_fr_r {parent=deform_upperarm_fr_base_r, loc=<0 0 0>, rot=<0 0 0 1>}
 deform_upperarm_up_base_r {parent=upperarm_twist_01_r, loc=<8.891 -0.4825 -3.9517>, rot=<0 0 0 -1>}
 deform_upperarm_up_r {parent=deform_upperarm_up_base_r, loc=<0 0 0>, rot=<0 0 0 1>}
 upperarm_twist_02_r {parent=upperarm_r, loc=<-19.7998 0.1134 0>, rot=<0 0 0 -1>}
 base_collar_l {parent=spine_05, loc=<9.9719 5.7458 -5.238>, rot=<0.0875 -0.8761 -0.3056 -0.3624>}
 dyn_collar_l_01 {parent=base_collar_l, loc=<0 0 0>, rot=<0 0 0 1>}
 dyn_collar_l_02 {parent=dyn_collar_l_01, loc=<3.8516 0.0227 0>, rot=<0 0 -0.0155 -0.9999>}
 base_collar_r {parent=spine_05, loc=<8.4244 7.185 5.662>, rot=<0.3905 0.2568 0.8832 -0.0406>}
 dyn_collar_r_01 {parent=base_collar_r, loc=<0 0 0>, rot=<0 0 0 1>}
 dyn_collar_r_02 {parent=dyn_collar_r_01, loc=<3.8505 0.0959 -0.0581>, rot=<-0.0065 0.0422 0.1514 -0.9876>}
 base_tie {parent=spine_05, loc=<1.6907 8.678 0>, rot=<-0.1072 0.0503 0.8989 0.4219>}
 dyn_tie_top1 {parent=base_tie, loc=<0 0 0>, rot=<0 0 0 1>}
 dyn_tie_01 {parent=dyn_tie_top1, loc=<4.0646 0.431 -0.1409>, rot=<0.0391 0.1118 -0.3273 -0.9375>}
 dyn_tie_02 {parent=dyn_tie_01, loc=<6.6782 -0.3991 -0.5173>, rot=<0 0 -0.202 -0.9794>}
 dyn_tie_03 {parent=dyn_tie_02, loc=<8.0845 0.0014 0>, rot=<0 0 0.0423 -0.9991>}
 dyn_tie_04 {parent=dyn_tie_03, loc=<6.9701 0.0098 0>, rot=<0 0 0.0483 -0.9988>}
 deform_pec_aim_base_l {parent=spine_04, loc=<3.2456 9.3309 -1.2237>, rot=<0.0603 -0.4033 0.3318 -0.8507>}
 deform_pec_aim_l {parent=deform_pec_aim_base_l, loc=<0 0 0>, rot=<0 0 0 1>}
 deform_pec_base_l {parent=deform_pec_aim_l, loc=<9.3665 0 0>, rot=<0 0 0 1>}
 deform_pec_l {parent=deform_pec_base_l, loc=<0 0 0>, rot=<0 0 0 1>}
 deform_pec_aim_base_r {parent=spine_04, loc=<3.2452 9.3309 1.2237>, rot=<-0.4033 -0.0603 0.8507 0.3318>}
 deform_pec_aim_r {parent=deform_pec_aim_base_r, loc=<0 0 0>, rot=<0 0 0 1>}
 deform_pec_base_r {parent=deform_pec_aim_r, loc=<-9.3667 -0.0003 -0.0004>, rot=<0 0 0 1>}
 deform_pec_r {parent=deform_pec_base_r, loc=<0 0 0>, rot=<0 0 0 1>}
 base_button_01 {parent=spine_04, loc=<8.6751 7.4946 9.5222>, rot=<0.7311 0.1356 0.4846 0.4607>}
 base_button_02 {parent=spine_04, loc=<7.2287 10.2524 6.4737>, rot=<-0.6602 -0.2534 -0.3702 -0.6025>}
 base_button_03 {parent=spine_04, loc=<4.7398 10.0372 9.6575>, rot=<0.7051 0.0539 0.5271 0.4713>}
 base_jacket_fn_l {parent=spine_04, loc=<-3.6938 11.1183 -8.8534>, rot=<0.0658 -0.8945 -0.4387 0.055>}
 dyn_jacket_fn_l_01 {parent=base_jacket_fn_l, loc=<0 0 0>, rot=<0 0 0 1>}
 dyn_jacket_fn_l_02 {parent=dyn_jacket_fn_l_01, loc=<4.3768 0.039 0>, rot=<0 0 -0.0155 -0.9999>}
 base_zipper {parent=dyn_jacket_fn_l_01, loc=<3.881 4.2901 -0.1186>, rot=<0.8966 0.0824 -0.0048 -0.4351>}
 dyn_zipper {parent=base_zipper, loc=<0 0 0>, rot=<0 0 0 1>}
 base_jacket_s_l {parent=spine_04, loc=<-2.8436 3.361 -12.1783>, rot=<0.0136 -0.9979 -0.0282 0.0574>}
 dyn_jacket_s_l_01 {parent=base_jacket_s_l, loc=<0 0 0>, rot=<0 0 0 1>}
 dyn_jacket_s_l_02 {parent=dyn_jacket_s_l_01, loc=<4.3768 0.039 0>, rot=<0 0 -0.0155 -0.9999>}
 base_jacket_fn_r {parent=spine_04, loc=<-3.6934 11.1183 8.8534>, rot=<-0.0098 0.3464 0.9342 -0.0852>}
 dyn_jacket_fn_r_01 {parent=base_jacket_fn_r, loc=<0 0 0>, rot=<0 0 0 -1>}
 dyn_jacket_fn_r_02 {parent=dyn_jacket_fn_r_01, loc=<4.3609 -0.2248 -0.3088>, rot=<0 0 -0.0155 -0.9999>}
 base_jacket_s_r {parent=spine_04, loc=<-2.8437 3.361 12.1783>, rot=<-0.045 0.0837 0.9953 -0.0159>}
 dyn_jacket_s_r_01 {parent=base_jacket_s_r, loc=<0 0 0>, rot=<0 0 0 -1>}
 dyn_jacket_s_r_02 {parent=dyn_jacket_s_r_01, loc=<4.3741 -0.1551 0.0473>, rot=<0 0 -0.0022 -1>}
 thigh_l {parent=pelvis, loc=<-1.0855 -0.2109 -8.7053>, rot=<0.0009 0.0355 0.0155 -0.9993>}
 calf_l {parent=thigh_l, loc=<-42.5284 -0.0019 -0.7297>, rot=<0.0006 -0.0161 -0.0391 -0.9991>}
 foot_l {parent=calf_l, loc=<-46.7401 -2.5014 -0.592>, rot=<0.0427 -0.0238 0.0353 -0.9982>}
 ball_l {parent=foot_l, loc=<-3.5658 10.8066 -0.1544>, rot=<0.0001 -0.0001 -0.7063 -0.7079>}
 base_shoe_l {parent=foot_l, loc=<3.5362 0.9338 0.538>, rot=<-0.481 -0.516 -0.4778 -0.5235>}
 base_shoeSideFlap_l {parent=base_shoe_l, loc=<1.0316 -4.2426 2.6575>, rot=<-0.4723 0.3396 0.4174 -0.6982>}
 dyn_shoe_l_01 {parent=base_shoeSideFlap_l, loc=<0 0 0>, rot=<0 0 0 1>}
 base_shoeSideFlap_r {parent=base_shoe_l, loc=<-0.4536 4.3426 2.6575>, rot=<0.7722 0.3401 0.3437 0.4123>}
 dyn_shoe_l_02 {parent=base_shoeSideFlap_r, loc=<0 0 0>, rot=<0 0 0 1>}
 base_shoeFlap_l {parent=base_shoe_l, loc=<4.4515 0.0045 0>, rot=<-0.5852 0.3968 -0.4908 -0.509>}
 dyn_shoeFlap_l_01 {parent=base_shoeFlap_l, loc=<0 0 0>, rot=<0 0 0 1>}
 dyn_shoeFlap_l_02 {parent=dyn_shoeFlap_l_01, loc=<3.509 0.0699 0>, rot=<0 0 0 1>}
 calf_twist_02_l {parent=calf_l, loc=<-19.1546 -1.7943 -0.2638>, rot=<0.0237 0.0095 -0.0035 -0.9997>}
 calf_twist_01_l {parent=calf_l, loc=<-35.8483 -1.9197 -0.5794>, rot=<0.0237 0.0095 -0.0035 -0.9997>}
 deform_knee_in_base_l {parent=calf_l, loc=<0.8031 -5.0231 0.2456>, rot=<0 0 0 -1>}
 deform_knee_in_l {parent=deform_knee_in_base_l, loc=<0 0 0>, rot=<0 0 0 1>}
 deform_knee_out_base_l {parent=calf_l, loc=<0 4 0>, rot=<0 0 0 -1>}
 deform_knee_out_l {parent=deform_knee_out_base_l, loc=<0 0 0>, rot=<0 0 0 1>}
 thigh_twist_01_l {parent=thigh_l, loc=<-27.952 0 0>, rot=<0.0013 0 0 -1>}
 deform_groin_base_l {parent=thigh_l, loc=<0.4801 5.6776 -0.2486>, rot=<0 0 0 -1>}
 deform_groin_l {parent=deform_groin_base_l, loc=<0 0 0>, rot=<0 0 0 1>}
 thigh_r {parent=pelvis, loc=<-1.0855 -0.2109 8.7053>, rot=<0.0355 -0.0009 0.9993 0.0155>}
 calf_r {parent=thigh_r, loc=<42.5284 0.0019 0.7297>, rot=<0.0006 -0.0161 -0.0391 -0.9991>}
 foot_r {parent=calf_r, loc=<46.7401 2.5014 0.592>, rot=<0.0427 -0.0238 0.0353 -0.9982>}
 ball_r {parent=foot_r, loc=<3.5658 -10.8066 0.1544>, rot=<0.0001 -0.0001 -0.7063 -0.7079>}
 calf_twist_02_r {parent=calf_r, loc=<19.1546 1.7943 0.2638>, rot=<0.0237 0.0095 -0.0035 -0.9997>}
 calf_twist_01_r {parent=calf_r, loc=<35.8483 1.9197 0.5794>, rot=<0.0237 0.0095 -0.0035 -0.9997>}
 deform_knee_in_base_r {parent=calf_r, loc=<-0.8031 5.0231 -0.2457>, rot=<0 0 0 -1>}
 deform_knee_in_r {parent=deform_knee_in_base_r, loc=<0 0 0>, rot=<0 0 0 1>}
 deform_knee_out_base_r {parent=calf_r, loc=<0 -4 0>, rot=<0 0 0 -1>}
 deform_knee_out_r {parent=deform_knee_out_base_r, loc=<0 0 0>, rot=<0 0 0 1>}
 thigh_twist_01_r {parent=thigh_r, loc=<27.952 0 0>, rot=<0.0013 0 0 -1>}
 deform_groin_base_r {parent=thigh_r, loc=<-0.4801 -5.6776 0.2486>, rot=<0 0 0 -1>}
 deform_groin_r {parent=deform_groin_base_r, loc=<0 0 0>, rot=<0 0 0 1>}
 deform_glute_base_l {parent=pelvis, loc=<-7.3466 -9.0213 -9.1245>, rot=<0.0172 0.031 -0.4477 -0.8935>}
 deform_glute_l {parent=deform_glute_base_l, loc=<0 0 0>, rot=<0 0 0 1>}
 deform_glute_base_r {parent=pelvis, loc=<-7.3464 -9.0213 9.1245>, rot=<0.031 -0.0172 0.8935 -0.4477>}
 deform_glute_r {parent=deform_glute_base_r, loc=<0 0 0>, rot=<0 0 0 1>}
 base_skirt_l_A {parent=pelvis, loc=<3.6873 5.083 -6.3538>, rot=<-0.0316 0.0425 0.993 0.1056>}
 dyn_skirt_l_A_01 {parent=base_skirt_l_A, loc=<0 0 0>, rot=<0 0 0 1>}
 dyn_skirt_l_A_02 {parent=dyn_skirt_l_A_01, loc=<6.4195 0.0475 0>, rot=<0 0 0.0102 -0.9999>}
 dyn_skirt_l_A_03 {parent=dyn_skirt_l_A_02, loc=<10.7822 -0.0708 0>, rot=<0 0 0 1>}
 base_skirt_l_B {parent=pelvis, loc=<4.1119 -0.4162 -12.437>, rot=<-0.0202 -0.9883 -0.0182 -0.1503>}
 dyn_skirt_l_B_01 {parent=base_skirt_l_B, loc=<0 0 0>, rot=<0 0 0 1>}
 dyn_skirt_l_B_02 {parent=dyn_skirt_l_B_01, loc=<6.4195 0.0475 0>, rot=<0 0 0.0102 -0.9999>}
 dyn_skirt_l_B_03 {parent=dyn_skirt_l_B_02, loc=<10.7822 -0.0708 0>, rot=<0 0 0 1>}
 base_skirtArmour {parent=dyn_skirt_l_B_01, loc=<1.7815 0.9074 0.2876>, rot=<-0.6401 0.0914 -0.064 -0.7602>}
 dyn_skirtArmour_01 {parent=base_skirtArmour, loc=<2.9073 0 0>, rot=<0 0 0 1>}
 dyn_skirtArmour_02 {parent=dyn_skirtArmour_01, loc=<4.1945 0.0259 0.0676>, rot=<-0.0016 -0.0368 0.0823 -0.9959>}
 dyn_skirtArmour_03 {parent=dyn_skirtArmour_02, loc=<7.646 0.0495 -0.0861>, rot=<0 0 0 1>}
 base_skirt_l_C {parent=pelvis, loc=<4.5676 -9.4242 -6.0057>, rot=<-0.0443 -0.0952 0.9915 -0.0765>}
 dyn_skirt_l_C_01 {parent=base_skirt_l_C, loc=<0 0 0>, rot=<0 0 0 1>}
 dyn_skirt_l_C_02 {parent=dyn_skirt_l_C_01, loc=<6.4195 0.0475 0>, rot=<0 0 0.0102 -0.9999>}
 dyn_skirt_l_C_03 {parent=dyn_skirt_l_C_02, loc=<10.7822 -0.0708 0>, rot=<0 0 0 1>}
 base_skirt_r_A {parent=pelvis, loc=<3.7299 6.2681 6.006>, rot=<0.0609 0.0325 0.9916 0.1091>}
 dyn_skirt_r_A_01 {parent=base_skirt_r_A, loc=<0 0 0>, rot=<0 0 0 1>}
 dyn_skirt_r_A_02 {parent=dyn_skirt_r_A_01, loc=<6.4195 0.0475 0>, rot=<0 0 0.0102 -0.9999>}
 dyn_skirt_r_A_03 {parent=dyn_skirt_r_A_02, loc=<10.7822 -0.0708 0>, rot=<0 0 0 1>}
 base_skirt_r_B {parent=pelvis, loc=<4.1116 -0.4164 12.0735>, rot=<0.1534 -0.0181 0.9879 0.0112>}
 dyn_skirt_r_B_01 {parent=base_skirt_r_B, loc=<0 0 0>, rot=<0 0 0.0028 -1>}
 dyn_skirt_r_B_02 {parent=dyn_skirt_r_B_01, loc=<6.4195 0.0475 0>, rot=<0 0 0.0103 -0.9999>}
 dyn_skirt_r_B_03 {parent=dyn_skirt_r_B_02, loc=<10.7822 -0.0708 0>, rot=<0 0 0 1>}
 base_skirt_r_C {parent=pelvis, loc=<4.5672 -9.4242 6.006>, rot=<0.0322 0.0604 0.994 -0.0858>}
 dyn_skirt_r_C_01 {parent=base_skirt_r_C, loc=<0 0 0>, rot=<0 0 0 1>}
 dyn_skirt_r_C_02 {parent=dyn_skirt_r_C_01, loc=<6.4195 0.0475 0>, rot=<0 0 0.0102 -0.9999>}
 dyn_skirt_r_C_03 {parent=dyn_skirt_r_C_02, loc=<10.7822 -0.0708 0>, rot=<0 0 0 1>}
 ALT_fix {parent=pelvis, loc=<0 0 0>, rot=<0 0 0 1>}
 pelvis_ctrl {controlType=EULER_TRANSFORM, type=Control, loc=<0 0 0>, rot=<0 0 0 1>}
 spine_01_ctrl {controlType=EULER_TRANSFORM, parent=pelvis_ctrl, type=Control, loc=<0 0 0>, rot=<0 0 0 1>}
 spine_02_ctrl {controlType=EULER_TRANSFORM, parent=spine_01_ctrl, type=Control, loc=<0 0 0>, rot=<0 0 0 1>}
 spine_03_ctrl {controlType=EULER_TRANSFORM, parent=spine_02_ctrl, type=Control, loc=<0 0 0>, rot=<0 0 0 1>}
 spine_04_ctrl {controlType=EULER_TRANSFORM, parent=spine_03_ctrl, type=Control, loc=<0 0 0>, rot=<0 0 0 1>}
 spine_05_ctrl {controlType=EULER_TRANSFORM, parent=spine_04_ctrl, type=Control, loc=<0 0 0>, rot=<0 0 0 1>}
 clavicle_l_ctrl {controlType=EULER_TRANSFORM, parent=spine_05_ctrl, type=Control, loc=<0 0 0>, rot=<0 0 0 1>}
 upperarm_l_ctrl {controlType=EULER_TRANSFORM, parent=clavicle_l_ctrl, type=Control, loc=<0 0 0>, rot=<0 0 0 1>}
 thigh_l_ctrl {controlType=EULER_TRANSFORM, parent=pelvis_ctrl, type=Control, loc=<0 0 0>, rot=<0 0 0 1>}
 calf_l_ctrl {controlType=EULER_TRANSFORM, parent=thigh_l_ctrl, type=Control, loc=<0 0 0>, rot=<0 0 0 1>}
 thigh_r_ctrl {controlType=EULER_TRANSFORM, parent=pelvis_ctrl, type=Control, loc=<0 0 0>, rot=<0 0 0 1>}
 calf_r_ctrl {controlType=EULER_TRANSFORM, parent=thigh_r_ctrl, type=Control, loc=<0 0 0>, rot=<0 0 0 1>}
]
variables [
 Emote Float {type=double, default=0}
 Get Crouching {type=bool, default=False}
 Get DBNO {type=bool, default=False}
 Get Emote {type=bool, default=False}
]
