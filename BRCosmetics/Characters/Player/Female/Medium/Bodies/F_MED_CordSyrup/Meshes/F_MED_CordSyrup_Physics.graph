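# Physics asset for the F_MED_CordSyrup female medium body cosmetic, covering its jacket, collars, tie, zipper and skirt.
# Capsules on the torso, left clavicle and upper arms form the body volume, with garment base bodies anchored to it.
# Jacket, collar, tie and zipper chains use tight angular limits, with the zipper hanging off the left front jacket.
# Skirt chains swing freely from the pelvis, cross-linked left to right at their lower two segments.
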PhysicsAsset /BRCosmetics/Characters/Player/Female/Medium/Bodies/F_MED_CordSyrup/Meshes/F_MED_CordSyrup_Physics
bodies {
 base_armourString [{shapeType=Capsule, length=10, radius=1.5, shapeName=clavicle_l_capsule, center=<1 -0.0849 -0.4439>, rotation=<rot 0.0001 0.0001 95.8923>}]
 base_button_01 [{shapeType=Capsule, length=0, radius=2.5, shapeName=pelvis_capsule, center=<-2.0399 0 0.2316>, rotation=<rot 90 180 180>}]
 base_collar_l [{shapeType=Capsule, length=0.5, radius=0.5, shapeName=base_collar_l_capsule}]
 base_collar_r [{shapeType=Capsule, length=0.5, radius=0.5, shapeName=base_collar_r_capsule}]
 base_jacket_fn_l [{shapeType=Capsule, length=0.5, radius=0.5, shapeName=base_jacket_fn_l_capsule}]
 base_jacket_fn_r [{shapeType=Capsule, length=0.5, radius=0.5, shapeName=base_jacket_fn_r_capsule}]
 base_jacket_s_l [{shapeType=Capsule, length=0.5, radius=0.5, shapeName=base_jacket_s_l_capsule}]
 base_jacket_s_r [{shapeType=Capsule, length=0.5, radius=0.5, shapeName=base_jacket_s_r_capsule}]
 base_tie [{shapeType=Capsule, length=4, radius=1, shapeName=dyn_tie_01_capsule, center=<2 -0.3765 0>}]
 base_zipper [{shapeType=Capsule, length=0.5, radius=0.5, shapeName=base_zipper_capsule}]
 clavicle_l [{shapeType=Capsule, length=0, radius=5.5, shapeName=clavicle_l_capsule, center=<6 0 0>, rotation=<rot 0 0 90>}]
 dyn_collar_l_01 [{shapeType=Capsule, length=3, radius=1.25, shapeName=dyn_collar_l_01_capsule, center=<1.7886 -0.4216 0>, rotation=<rot 0 0 90>}]
 dyn_collar_l_02 [{
   shapeType=Capsule
   length=2
   radius=0.75
   shapeName=dyn_collar_l_02_capsule
   center=<2 0.184 0.1455>
   rotation=<rot 79.8433 -179.9999 -179.9999>
  }]
 dyn_collar_r_01 [{
   shapeType=Capsule
   length=3
   radius=1
   shapeName=dyn_collar_r_01_capsule
   center=<2.0732 1.3056 0.0001>
   rotation=<rot 90 165.9638 -156.8503>
  }]
 dyn_collar_r_02 [{
   shapeType=Capsule
   length=2.5
   radius=0.75
   shapeName=dyn_collar_r_02_capsule
   center=<1.3198 0.0863 0.0851>
   rotation=<rot 90 -171.1193 -158.3687>
  }]
 dyn_jacket_fn_l_01 [{shapeType=Capsule, length=3, radius=3, shapeName=dyn_jacket_fn_l_01_capsule, center=<3 0 0>, rotation=<rot 90 0 0>}]
 dyn_jacket_fn_l_02 [{shapeType=Capsule, length=3, radius=3, shapeName=dyn_jacket_fn_l_02_capsule, center=<3 0 0>, rotation=<rot 90 0 0>}]
 dyn_jacket_fn_r_01 [{shapeType=Capsule, length=3, radius=3, shapeName=dyn_jacket_fn_r_01_capsule, center=<3 0 0>, rotation=<rot 90 0 0>}]
 dyn_jacket_fn_r_02 [{shapeType=Capsule, length=3, radius=3, shapeName=dyn_jacket_fn_r_02_capsule, center=<3 0 0>, rotation=<rot 90 0 0>}]
 dyn_jacket_s_l_01 [{shapeType=Capsule, length=3, radius=3, shapeName=dyn_jacket_s_l_01_capsule, center=<3 0 0>, rotation=<rot 90 0 0>}]
 dyn_jacket_s_l_02 [{shapeType=Capsule, length=3, radius=3, shapeName=dyn_jacket_s_l_02_capsule, center=<3 0 0>, rotation=<rot 90 0 0>}]
 dyn_jacket_s_r_01 [{shapeType=Capsule, length=3, radius=3, shapeName=dyn_jacket_s_r_01_capsule, center=<3 0 0>, rotation=<rot 90 0 0>}]
 dyn_jacket_s_r_02 [{shapeType=Capsule, length=3, radius=3, shapeName=dyn_jacket_s_r_02_capsule, center=<3 0 0>, rotation=<rot 90 0 0>}]
 dyn_skirt_l_A_01 [{
   shapeType=Capsule
   length=20.6751
   radius=4.731
   shapeName=dyn_skirt_l_A_01_capsule
   center=<6.0488 0.2378 1.6272>
   rotation=<rot -2.5569 -90.1526 -0.8378>
  }]
 dyn_skirt_l_A_02 [{
   shapeType=Capsule
   length=20.1751
   radius=5.231
   shapeName=dyn_skirt_l_A_01_capsule
   center=<5.8232 1.4021 1.655>
   rotation=<rot -0.442 -90.1837 -0.837>
  }]
 dyn_skirt_l_A_03 [{
   shapeType=Capsule
   length=21.1751
   radius=5.731
   shapeName=dyn_skirt_l_A_01_capsule
   center=<3.7104 1.7413 1.6167>
   rotation=<rot -0.442 -90.1837 -0.837>
  }]
 dyn_skirt_r_A_01 [{
   shapeType=Capsule
   length=11.6692
   radius=4.9104
   shapeName=dyn_skirt_r_A_01_capsule
   center=<5.6266 1.344 -2.5357>
   rotation=<rot 8.5765 -91.9827 5.8552>
  }]
 dyn_skirt_r_A_02 [{
   shapeType=Capsule
   length=11.1692
   radius=5.4104
   shapeName=dyn_skirt_r_A_01_capsule
   center=<5.4376 2.5349 -2.7124>
   rotation=<rot 5.0688 -92.3438 5.8123>
  }]
 dyn_skirt_r_A_03 [{
   shapeType=Capsule
   length=23.1692
   radius=5.9104
   shapeName=dyn_skirt_r_A_01_capsule
   center=<3.3504 2.9706 -2.5285>
   rotation=<rot 5.0688 -92.3438 5.8123>
  }]
 dyn_tie_01 [{shapeType=Capsule, length=4, radius=1.5, shapeName=dyn_tie_01_capsule, center=<2.9862 -0.3765 0>, rotation=<rot 90 0 0>}]
 dyn_tie_02 [{shapeType=Capsule, length=4.5, radius=1.5, shapeName=dyn_tie_02_capsule, center=<3.5 0 0>, rotation=<rot 90 0 0>}]
 dyn_tie_03 [{shapeType=Capsule, length=4.5, radius=1.5, shapeName=dyn_tie_03_capsule, center=<3 0 0>, rotation=<rot 90 0 0>}]
 dyn_tie_04 [{shapeType=Capsule, length=5, radius=1.5, shapeName=dyn_tie_04_capsule, center=<2.5 0 0>, rotation=<rot 90 0 0>}]
 dyn_tie_top1 [{shapeType=Capsule, length=4, radius=1, shapeName=dyn_tie_01_capsule, center=<2 -0.3765 0>}]
 dyn_zipper [{shapeType=Capsule, length=2.5, radius=0.75, shapeName=dyn_zipper_capsule, center=<2 0 0>, rotation=<rot 90 0 0>}]
 pelvis [{shapeType=Capsule, length=9.2488, radius=9, shapeName=pelvis_capsule, center=<0 1.1765 0>}]
 spine_01 [{shapeType=Capsule, length=5.2488, radius=9, shapeName=pelvis_capsule}]
 spine_02 [{shapeType=Capsule, length=4.2488, radius=9, shapeName=pelvis_capsule}]
 spine_03 [{shapeType=Capsule, length=6, radius=9, shapeName=spine_03_capsule, center=<0 0 0.1171>, rotation=<rot 90 0 0>}]
 spine_04 [
  {shapeType=Capsule, length=3.2488, radius=11, shapeName=pelvis_capsule, center=<1.3498 0.0895 0>}
  {shapeType=Capsule, length=9.5, radius=5, shapeName=spine_04_capsule, center=<-1.2944 6.8292 0>}
 ]
 spine_05 [{shapeType=Capsule, length=5.7488, radius=10.5, shapeName=pelvis_capsule, center=<-3.2904 2.0273 0>}]
 upperarm_l [{shapeType=Capsule, length=10, radius=5.5, shapeName=clavicle_l_capsule, center=<10 0 0>, rotation=<rot 90 0 0>}]
 upperarm_r [{shapeType=Capsule, length=15, radius=5.5, shapeName=clavicle_l_capsule, center=<-10 0 0>, rotation=<rot 90 0 0>}]
}
constraints [
 {
  bone1Name=spine_01
  bone2Name=pelvis
  swing1LimitDegrees=45
  swing1Motion=Free
  swing2LimitDegrees=45
  swing2Motion=Free
  twistLimitDegrees=45
  twistMotion=Free
 }
 {
  bone1Name=spine_02
  bone2Name=spine_01
  swing1LimitDegrees=45
  swing1Motion=Free
  swing2LimitDegrees=45
  swing2Motion=Free
  twistLimitDegrees=45
  twistMotion=Free
 }
 {
  bone1Name=spine_03
  bone2Name=spine_02
  swing1LimitDegrees=45
  swing1Motion=Free
  swing2LimitDegrees=45
  swing2Motion=Free
  twistLimitDegrees=45
  twistMotion=Free
 }
 {
  bone1Name=spine_04
  bone2Name=spine_03
  swing1LimitDegrees=45
  swing1Motion=Free
  swing2LimitDegrees=45
  swing2Motion=Free
  twistLimitDegrees=45
  twistMotion=Free
 }
 {
  bone1Name=base_jacket_fn_l
  bone2Name=spine_04
  swing1LimitDegrees=45
  swing1Motion=Free
  swing2LimitDegrees=45
  swing2Motion=Free
  twistLimitDegrees=45
  twistMotion=Free
 }
 {bone1Name=dyn_jacket_fn_l_01, bone2Name=base_jacket_fn_l, swing1LimitDegrees=5, swing2LimitDegrees=20, twistLimitDegrees=0}
 {
  bone1Name=base_zipper
  bone2Name=dyn_jacket_fn_l_01
  swing1LimitDegrees=45
  swing1Motion=Free
  swing2LimitDegrees=45
  swing2Motion=Free
  twistLimitDegrees=45
  twistMotion=Free
 }
 {bone1Name=dyn_zipper, bone2Name=base_zipper, swing1LimitDegrees=8, swing2LimitDegrees=15, twistLimitDegrees=0}
 {bone1Name=dyn_jacket_fn_l_02, bone2Name=dyn_jacket_fn_l_01, swing1LimitDegrees=10, swing2LimitDegrees=20, twistLimitDegrees=0}
 {
  bone1Name=base_jacket_fn_r
  bone2Name=spine_04
  swing1LimitDegrees=45
  swing1Motion=Free
  swing2LimitDegrees=45
  swing2Motion=Free
  twistLimitDegrees=45
  twistMotion=Free
 }
 {bone1Name=dyn_jacket_fn_r_01, bone2Name=base_jacket_fn_r, swing1LimitDegrees=5, swing2LimitDegrees=20, twistLimitDegrees=0}
 {bone1Name=dyn_jacket_fn_r_02, bone2Name=dyn_jacket_fn_r_01, swing1LimitDegrees=10, swing2LimitDegrees=20, twistLimitDegrees=0}
 {
  bone1Name=base_jacket_s_l
  bone2Name=spine_04
  swing1LimitDegrees=45
  swing1Motion=Free
  swing2LimitDegrees=45
  swing2Motion=Free
  twistLimitDegrees=45
  twistMotion=Free
 }
 {bone1Name=dyn_jacket_s_l_01, bone2Name=base_jacket_s_l, swing1LimitDegrees=0, swing2LimitDegrees=10, twistLimitDegrees=0}
 {bone1Name=dyn_jacket_s_l_02, bone2Name=dyn_jacket_s_l_01, swing1LimitDegrees=0, swing2LimitDegrees=15, twistLimitDegrees=0}
 {
  bone1Name=base_jacket_s_r
  bone2Name=spine_04
  swing1LimitDegrees=45
  swing1Motion=Free
  swing2LimitDegrees=45
  swing2Motion=Free
  twistLimitDegrees=45
  twistMotion=Free
 }
 {bone1Name=dyn_jacket_s_r_01, bone2Name=base_jacket_s_r, swing1LimitDegrees=0, swing2LimitDegrees=8, twistLimitDegrees=0}
 {bone1Name=dyn_jacket_s_r_02, bone2Name=dyn_jacket_s_r_01, swing1LimitDegrees=0, swing2LimitDegrees=12, twistLimitDegrees=0}
 {bone1Name=dyn_tie_01, bone2Name=dyn_tie_top1, swing1LimitDegrees=25, swing2LimitDegrees=2.5, twistLimitDegrees=0}
 {bone1Name=dyn_tie_02, bone2Name=dyn_tie_01, swing1LimitDegrees=40, swing2LimitDegrees=4, twistLimitDegrees=0}
 {bone1Name=dyn_tie_03, bone2Name=dyn_tie_02, swing1LimitDegrees=45, swing2LimitDegrees=5, twistLimitDegrees=0}
 {bone1Name=dyn_tie_04, bone2Name=dyn_tie_03, swing1LimitDegrees=50, swing2LimitDegrees=5, twistLimitDegrees=0}
 {
  bone1Name=base_collar_l
  bone2Name=spine_05
  swing1LimitDegrees=45
  swing1Motion=Free
  swing2LimitDegrees=45
  swing2Motion=Free
  twistLimitDegrees=45
  twistMotion=Free
 }
 {bone1Name=dyn_collar_l_01, bone2Name=base_collar_l, swing1LimitDegrees=10, swing2LimitDegrees=8, twistLimitDegrees=0}
 {bone1Name=dyn_collar_l_02, bone2Name=dyn_collar_l_01, swing1LimitDegrees=8, swing2LimitDegrees=10, twistLimitDegrees=0}
 {
  bone1Name=base_collar_r
  bone2Name=spine_05
  swing1LimitDegrees=45
  swing1Motion=Free
  swing2LimitDegrees=45
  swing2Motion=Free
  twistLimitDegrees=45
  twistMotion=Free
 }
 {bone1Name=dyn_collar_r_01, bone2Name=base_collar_r, swing1LimitDegrees=7, swing2LimitDegrees=5, twistLimitDegrees=0}
 {bone1Name=dyn_collar_r_02, bone2Name=dyn_collar_r_01, swing1LimitDegrees=6, swing2LimitDegrees=5, twistLimitDegrees=5}
 {
  bone1Name=clavicle_l
  bone2Name=spine_05
  swing1LimitDegrees=45
  swing1Motion=Free
  swing2LimitDegrees=45
  swing2Motion=Free
  twistLimitDegrees=45
  twistMotion=Free
 }
 {
  bone1Name=spine_05
  bone2Name=spine_04
  swing1LimitDegrees=45
  swing1Motion=Free
  swing2LimitDegrees=45
  swing2Motion=Free
  twistLimitDegrees=45
  twistMotion=Free
 }
 {
  bone1Name=base_armourString
  bone2Name=upperarm_l
  swing1LimitDegrees=45
  swing1Motion=Free
  swing2LimitDegrees=45
  swing2Motion=Free
  twistLimitDegrees=45
  twistMotion=Free
 }
 {
  bone1Name=upperarm_l
  bone2Name=clavicle_l
  swing1LimitDegrees=45
  swing1Motion=Free
  swing2LimitDegrees=45
  swing2Motion=Free
  twistLimitDegrees=45
  twistMotion=Free
 }
 {
  bone1Name=upperarm_r
  bone2Name=spine_05
  swing1LimitDegrees=45
  swing1Motion=Free
  swing2LimitDegrees=45
  swing2Motion=Free
  twistLimitDegrees=45
  twistMotion=Free
 }
 {bone1Name=base_tie, bone2Name=spine_04, swing1LimitDegrees=1, swing2LimitDegrees=1, twistLimitDegrees=0}
 {bone1Name=dyn_tie_top1, bone2Name=base_tie, swing1LimitDegrees=2, swing2LimitDegrees=1, twistLimitDegrees=0}
 {
  bone1Name=dyn_skirt_l_A_01
  bone2Name=pelvis
  swing1LimitDegrees=45
  swing1Motion=Free
  swing2LimitDegrees=45
  swing2Motion=Free
  twistLimitDegrees=45
  twistMotion=Free
 }
 {
  bone1Name=dyn_skirt_r_A_01
  bone2Name=pelvis
  swing1LimitDegrees=45
  swing1Motion=Free
  swing2LimitDegrees=45
  swing2Motion=Free
  twistLimitDegrees=45
  twistMotion=Free
 }
 {
  bone1Name=dyn_skirt_l_A_02
  bone2Name=dyn_skirt_l_A_01
  swing1LimitDegrees=45
  swing1Motion=Free
  swing2LimitDegrees=45
  swing2Motion=Free
  twistLimitDegrees=45
  twistMotion=Free
 }
 {
  bone1Name=dyn_skirt_r_A_02
  bone2Name=dyn_skirt_r_A_01
  swing1LimitDegrees=45
  swing1Motion=Free
  swing2LimitDegrees=45
  swing2Motion=Free
  twistLimitDegrees=45
  twistMotion=Free
 }
 {
  bone1Name=dyn_skirt_l_A_03
  bone2Name=dyn_skirt_l_A_02
  swing1LimitDegrees=45
  swing1Motion=Free
  swing2LimitDegrees=45
  swing2Motion=Free
  twistLimitDegrees=45
  twistMotion=Free
 }
 {
  bone1Name=dyn_skirt_r_A_03
  bone2Name=dyn_skirt_r_A_02
  swing1LimitDegrees=45
  swing1Motion=Free
  swing2LimitDegrees=45
  swing2Motion=Free
  twistLimitDegrees=45
  twistMotion=Free
 }
 {
  bone1Name=dyn_skirt_l_A_03
  bone2Name=dyn_skirt_r_A_03
  swing1LimitDegrees=45
  swing1Motion=Free
  swing2LimitDegrees=45
  swing2Motion=Free
  twistLimitDegrees=45
  twistMotion=Free
 }
 {
  bone1Name=dyn_skirt_l_A_02
  bone2Name=dyn_skirt_r_A_02
  swing1LimitDegrees=45
  swing1Motion=Free
  swing2LimitDegrees=45
  swing2Motion=Free
  twistLimitDegrees=45
  twistMotion=Free
 }
 {
  bone1Name=base_button_01
  bone2Name=spine_04
  swing1LimitDegrees=45
  swing1Motion=Free
  swing2LimitDegrees=45
  swing2Motion=Free
  twistLimitDegrees=45
  twistMotion=Free
 }
]
