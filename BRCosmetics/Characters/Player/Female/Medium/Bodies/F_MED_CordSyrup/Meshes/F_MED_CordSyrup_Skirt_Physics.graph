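# Physics asset for the skirt and skirt armour on the CordSyrup medium female player cosmetic body.
# Six three-segment capsule chains hang from base bones constrained to the pelvis, with tight swing and twist limits along each chain and capsules for pelvis and thighs.
# An armour chain branches from the left B chain.
# Extra angular-free constraints tie it back to that chain and link B to C mid-segments on each side.
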PhysicsAsset /BRCosmetics/Characters/Player/Female/Medium/Bodies/F_MED_CordSyrup/Meshes/F_MED_CordSyrup_Skirt_Physics
bodies {
 base_skirtArmour [{shapeType=Capsule, length=0, radius=0.5, shapeName=base_skirt_l_B_capsule}]
 base_skirt_l_A [{shapeType=Capsule, length=0, radius=0.5, shapeName=base_skirt_l_A_capsule}]
 base_skirt_l_B [{shapeType=Capsule, length=0, radius=0.5, shapeName=base_skirt_l_B_capsule}]
 base_skirt_l_C [{shapeType=Capsule, length=0, radius=0.5, shapeName=base_skirt_l_C_capsule}]
 base_skirt_r_A [{shapeType=Capsule, length=0, radius=0.5, shapeName=base_skirt_r_A_capsule}]
 base_skirt_r_B [{shapeType=Capsule, length=0, radius=0.5, shapeName=base_skirt_r_B_capsule}]
 base_skirt_r_C [{shapeType=Capsule, length=0, radius=0.5, shapeName=base_skirt_r_C_capsule}]
 dyn_skirtArmour_01 [{shapeType=Capsule, length=13, radius=4, shapeName=dyn_skirt_l_B_01_capsule, center=<0.9846 3.3599 0>}]
 dyn_skirtArmour_02 [{shapeType=Capsule, length=16, radius=4, shapeName=dyn_skirt_l_B_01_capsule, center=<0.9846 3.0976 0>}]
 dyn_skirtArmour_03 [{shapeType=Capsule, length=19.5, radius=6.5, shapeName=dyn_skirt_l_B_01_capsule, center=<0.9846 1.4755 0>}]
 dyn_skirt_l_A_01 [{shapeType=Capsule, length=13, radius=4, shapeName=dyn_skirt_l_A_01_capsule, rotation=<rot 0 -0.0003 -11.403>}]
 dyn_skirt_l_A_02 [{
   shapeType=Capsule
   length=14.25
   radius=5.75
   shapeName=dyn_skirt_l_A_01_capsule
   center=<5.0951 -5.2683 -0.6111>
   rotation=<rot 0 -0.0003 -6.6164>
  }]
 dyn_skirt_l_A_03 [{shapeType=Capsule, length=20, radius=6, shapeName=dyn_skirt_l_A_01_capsule, center=<5.0951 -5.8465 -0.1771>, rotation=<rot 0 0 -1.7349>}]
 dyn_skirt_l_B_01 [{shapeType=Capsule, length=12, radius=5, shapeName=dyn_skirt_l_B_01_capsule, center=<1.9394 0 4.1312>, rotation=<rot 0 0.0003 89.9999>}]
 dyn_skirt_l_B_02 [{shapeType=Capsule, length=14.5, radius=5.5, shapeName=dyn_skirt_l_B_01_capsule, center=<1.9394 0 4.6648>, rotation=<rot 0 0.0003 89.9999>}]
 dyn_skirt_l_B_03 [{shapeType=Capsule, length=17.25, radius=6.75, shapeName=dyn_skirt_l_B_01_capsule, center=<1.9394 0 4.6648>, rotation=<rot 0 0.0003 89.9999>}]
 dyn_skirt_l_C_01 [{shapeType=Capsule, length=12, radius=5, shapeName=dyn_skirt_l_C_01_capsule, center=<3.9792 3.6951 0>}]
 dyn_skirt_l_C_02 [{shapeType=Capsule, length=14.25, radius=5.75, shapeName=dyn_skirt_l_C_01_capsule, center=<3.9792 2.6704 0>}]
 dyn_skirt_l_C_03 [{shapeType=Capsule, length=19, radius=7, shapeName=dyn_skirt_l_C_01_capsule, center=<3.9792 2.2823 0>}]
 dyn_skirt_r_A_01 [{shapeType=Capsule, length=13, radius=4, shapeName=dyn_skirt_r_A_01_capsule}]
 dyn_skirt_r_A_02 [{shapeType=Capsule, length=15, radius=5, shapeName=dyn_skirt_r_A_01_capsule, center=<5.0951 -3.8664 0>}]
 dyn_skirt_r_A_03 [{shapeType=Capsule, length=20, radius=6, shapeName=dyn_skirt_r_A_01_capsule, center=<5.0951 -3.422 0>}]
 dyn_skirt_r_B_01 [{shapeType=Capsule, length=12, radius=5, shapeName=dyn_skirt_l_B_01_capsule, center=<4.5821 0 4.0818>, rotation=<rot 0 0.0003 89.9999>}]
 dyn_skirt_r_B_02 [{shapeType=Capsule, length=15, radius=5, shapeName=dyn_skirt_l_B_01_capsule, center=<4.5821 0 4.5545>, rotation=<rot 0 0.0003 89.9999>}]
 dyn_skirt_r_B_03 [{shapeType=Capsule, length=19, radius=5, shapeName=dyn_skirt_l_B_01_capsule, center=<4.5821 0 4.0818>, rotation=<rot 0 0.0003 89.9999>}]
 dyn_skirt_r_C_01 [{shapeType=Capsule, length=14.5, radius=5, shapeName=dyn_skirt_r_C_01_capsule, center=<3.9792 3.6951 0>}]
 dyn_skirt_r_C_02 [{shapeType=Capsule, length=16.75, radius=5.75, shapeName=dyn_skirt_r_C_01_capsule, center=<3.9792 2.6704 0>}]
 dyn_skirt_r_C_03 [{shapeType=Capsule, length=21.5, radius=7, shapeName=dyn_skirt_r_C_01_capsule, center=<3.9792 2.2823 0>}]
 pelvis [
  {shapeType=Capsule, length=9.5, radius=7.5, shapeName=pelvis_capsule, center=<0 -1.2851 0>}
  {shapeType=Capsule, length=0, radius=7, shapeName=pelvis_capsule, center=<0 -2.7782 0>}
 ]
 thigh_l [{
   shapeType=Capsule
   length=25.5
   radius=8
   shapeName=thigh_r_capsule
   center=<-19.8267 -0.058 0.6076>
   rotation=<rot 89.8498 0.0076 -178.2437>
  }]
 thigh_r [{shapeType=Capsule, length=25.5, radius=8, shapeName=thigh_r_capsule, center=<19.8267 0.058 -0.6076>, rotation=<rot 89.8498 0.0076 1.7564>}]
}
constraints [
 {
  bone1Name=base_skirt_l_A
  bone2Name=pelvis
  swing1LimitDegrees=45
  swing1Motion=Free
  swing2LimitDegrees=45
  swing2Motion=Free
  twistLimitDegrees=45
  twistMotion=Free
 }
 {bone1Name=dyn_skirt_l_A_01, bone2Name=base_skirt_l_A, swing1LimitDegrees=5, swing2LimitDegrees=2, twistLimitDegrees=0}
 {bone1Name=dyn_skirt_l_A_02, bone2Name=dyn_skirt_l_A_01, swing1LimitDegrees=10, swing2LimitDegrees=1, twistLimitDegrees=1}
 {bone1Name=dyn_skirt_l_A_03, bone2Name=dyn_skirt_l_A_02, swing1LimitDegrees=10, swing2LimitDegrees=1, twistLimitDegrees=1}
 {
  bone1Name=base_skirt_l_B
  bone2Name=pelvis
  swing1LimitDegrees=45
  swing1Motion=Free
  swing2LimitDegrees=45
  swing2Motion=Free
  twistLimitDegrees=45
  twistMotion=Free
 }
 {bone1Name=dyn_skirt_l_B_01, bone2Name=base_skirt_l_B, swing1LimitDegrees=1, swing2LimitDegrees=15, twistLimitDegrees=1}
 {bone1Name=dyn_skirt_l_B_02, bone2Name=dyn_skirt_l_B_01, swing1LimitDegrees=1, swing2LimitDegrees=15, twistLimitDegrees=1}
 {bone1Name=dyn_skirt_l_B_03, bone2Name=dyn_skirt_l_B_02, swing1LimitDegrees=1, swing2LimitDegrees=15, twistLimitDegrees=1}
 {
  bone1Name=base_skirt_l_C
  bone2Name=pelvis
  swing1LimitDegrees=45
  swing1Motion=Free
  swing2LimitDegrees=45
  swing2Motion=Free
  twistLimitDegrees=45
  twistMotion=Free
 }
 {bone1Name=dyn_skirt_l_C_01, bone2Name=base_skirt_l_C, swing1LimitDegrees=12, swing2LimitDegrees=10, twistLimitDegrees=1}
 {bone1Name=dyn_skirt_l_C_02, bone2Name=dyn_skirt_l_C_01, swing1LimitDegrees=12, swing2LimitDegrees=10, twistLimitDegrees=1}
 {bone1Name=dyn_skirt_l_C_03, bone2Name=dyn_skirt_l_C_02, swing1LimitDegrees=12, swing2LimitDegrees=10, twistLimitDegrees=1}
 {
  bone1Name=base_skirt_r_A
  bone2Name=pelvis
  swing1LimitDegrees=45
  swing1Motion=Free
  swing2LimitDegrees=45
  swing2Motion=Free
  twistLimitDegrees=45
  twistMotion=Free
 }
 {bone1Name=dyn_skirt_r_A_01, bone2Name=base_skirt_r_A, swing1LimitDegrees=5, swing2LimitDegrees=2, twistLimitDegrees=0}
 {bone1Name=dyn_skirt_r_A_02, bone2Name=dyn_skirt_r_A_01, swing1LimitDegrees=10, swing2LimitDegrees=1, twistLimitDegrees=1}
 {bone1Name=dyn_skirt_r_A_03, bone2Name=dyn_skirt_r_A_02, swing1LimitDegrees=10, swing2LimitDegrees=1, twistLimitDegrees=1}
 {
  bone1Name=base_skirt_r_B
  bone2Name=pelvis
  swing1LimitDegrees=45
  swing1Motion=Free
  swing2LimitDegrees=45
  swing2Motion=Free
  twistLimitDegrees=45
  twistMotion=Free
 }
 {bone1Name=dyn_skirt_r_B_01, bone2Name=base_skirt_r_B, swing1LimitDegrees=1, swing2LimitDegrees=10, twistLimitDegrees=1}
 {bone1Name=dyn_skirt_r_B_02, bone2Name=dyn_skirt_r_B_01, swing1LimitDegrees=1, swing2LimitDegrees=10, twistLimitDegrees=1}
 {bone1Name=dyn_skirt_r_B_03, bone2Name=dyn_skirt_r_B_02, swing1LimitDegrees=1, swing2LimitDegrees=10, twistLimitDegrees=1}
 {
  bone1Name=base_skirt_r_C
  bone2Name=pelvis
  swing1LimitDegrees=45
  swing1Motion=Free
  swing2LimitDegrees=45
  swing2Motion=Free
  twistLimitDegrees=45
  twistMotion=Free
 }
 {bone1Name=dyn_skirt_r_C_01, bone2Name=base_skirt_r_C, swing1LimitDegrees=12, swing2LimitDegrees=10, twistLimitDegrees=1}
 {bone1Name=dyn_skirt_r_C_02, bone2Name=dyn_skirt_r_C_01, swing1LimitDegrees=12, swing2LimitDegrees=10, twistLimitDegrees=1}
 {bone1Name=dyn_skirt_r_C_03, bone2Name=dyn_skirt_r_C_02, swing1LimitDegrees=12, swing2LimitDegrees=10, twistLimitDegrees=1}
 {
  bone1Name=thigh_r
  bone2Name=pelvis
  swing1LimitDegrees=45
  swing1Motion=Free
  swing2LimitDegrees=45
  swing2Motion=Free
  twistLimitDegrees=45
  twistMotion=Free
 }
 {
  bone1Name=base_skirtArmour
  bone2Name=dyn_skirt_l_B_01
  swing1LimitDegrees=45
  swing1Motion=Free
  swing2LimitDegrees=45
  swing2Motion=Free
  twistLimitDegrees=45
  twistMotion=Free
 }
 {bone1Name=dyn_skirtArmour_01, bone2Name=base_skirtArmour, swing1LimitDegrees=5, swing2LimitDegrees=2, twistLimitDegrees=0}
 {bone1Name=dyn_skirtArmour_02, bone2Name=dyn_skirtArmour_01, swing1LimitDegrees=5, swing2LimitDegrees=2, twistLimitDegrees=0}
 {bone1Name=dyn_skirtArmour_03, bone2Name=dyn_skirtArmour_02, swing1LimitDegrees=5, swing2LimitDegrees=2, twistLimitDegrees=0}
 {
  bone1Name=thigh_l
  bone2Name=pelvis
  swing1LimitDegrees=45
  swing1Motion=Free
  swing2LimitDegrees=45
  swing2Motion=Free
  twistLimitDegrees=45
  twistMotion=Free
 }
 {
  bone1Name=dyn_skirt_l_B_02
  bone2Name=dyn_skirtArmour_01
  swing1LimitDegrees=45
  swing1Motion=Free
  swing2LimitDegrees=45
  swing2Motion=Free
  twistLimitDegrees=45
  twistMotion=Free
 }
 {
  bone1Name=dyn_skirt_l_B_02
  bone2Name=dyn_skirtArmour_02
  swing1LimitDegrees=45
  swing1Motion=Free
  swing2LimitDegrees=45
  swing2Motion=Free
  twistLimitDegrees=45
  twistMotion=Free
 }
 {
  bone1Name=dyn_skirt_l_B_03
  bone2Name=dyn_skirtArmour_03
  swing1LimitDegrees=45
  swing1Motion=Free
  swing2LimitDegrees=45
  swing2Motion=Free
  twistLimitDegrees=45
  twistMotion=Free
 }
 {
  bone1Name=dyn_skirt_r_C_02
  bone2Name=dyn_skirt_r_B_02
  swing1LimitDegrees=45
  swing1Motion=Free
  swing2LimitDegrees=45
  swing2Motion=Free
  twistLimitDegrees=45
  twistMotion=Free
 }
 {
  bone1Name=dyn_skirt_l_C_02
  bone2Name=dyn_skirt_l_B_02
  swing1LimitDegrees=45
  swing1Motion=Free
  swing2LimitDegrees=45
  swing2Motion=Free
  twistLimitDegrees=45
  twistMotion=Free
 }
]
