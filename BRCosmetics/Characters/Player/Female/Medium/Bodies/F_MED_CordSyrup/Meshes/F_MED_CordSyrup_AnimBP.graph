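# Animation Blueprint for the F_MED_CordSyrup female body cosmetic, driving secondary motion and corrective posing for its outfit.
# It copies the attached parent's pose through a Control Rig, then constrains buttons, armour, tie, zipper and shoe flaps to weighted body bones.
# Cached poses chain AnimDynamics for strings, shoes, skirt and jacket, alongside a pelvis-space RigidBody skirt simulation.
# Additive bone offsets correct clipping-prone pieces, some weighted by DBNO or parachuting state.
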AnimBlueprint /BRCosmetics/Characters/Player/Female/Medium/Bodies/F_MED_CordSyrup/Meshes/F_MED_CordSyrup_AnimBP
target_skeleton=&/F_MED_CordSyrup_Skeleton
graphs [
 AnimGraph {nodes [
   RigidBody!(
    ComponentPose
    Alpha
    OverrideWorldGravity:Vector="0,0,0"
    ExternalForce
    SimSpaceSettings:Sim Space Settings Structure=(WorldAlpha=0.25,VelocityScaleZ=0.5,DampingAlpha=1,MaxLinearVelocity=500,MaxAngularVelocity=3,MaxLinearAcceleration=700,MaxAngularAcceleration=10000,ExternalLinearDragV=(X=0.2,Y=0.2,Z=0),ExternalLinearVelocity=(X=0,Y=0,Z=0),ExternalAngularVelocity=(X=0,Y=0,Z=0))
   ) -> (Pose) {
    bOverrideWorldGravity=true
    overridePhysicsAsset=&/F_MED_CordSyrup_Skirt_Physics
    simulationSpace=BaseBoneSpace
    alphaBoolBlend {blendOption=Linear}
    alphaScaleBiasClamp {bInterpResult=true}
    baseBoneRef {boneName=pelvis}
    simSpaceSettings {maxAngularVelocity=3, maxLinearAcceleration=700, maxLinearVelocity=500, velocityScaleZ=0.5, worldAlpha=0.25, externalLinearDragV=<0.2 0.2>}
   }
   n1 = "Transform (Modify) Bone - Bone: base_skirt_l_C"(
    ComponentPose=Local To Component(LocalPose=Use cached pose 'RBS2PostRBLogic'())
    Alpha=1
    Translation="0,0.75,0"
    Rotation="0,6,0"
    Scale="1,1,1"
   ) {
    rotationMode=BMM_Additive
    rotationSpace=BCS_BoneSpace
    translationMode=BMM_Additive
    translationSpace=BCS_BoneSpace
    alphaBoolBlend {blendOption=Linear}
    boneToModify {boneName=base_skirt_l_C}
    rotation {yaw=6}
    translation {y=0.75}
   }
   n2 = "Transform (Modify) Bone - Bone: base_skirt_r_C"(ComponentPose=@n1, Alpha=1, Translation="0,0.75,0", Rotation="0,6,0", Scale="1,1,1") {
    rotationMode=BMM_Additive
    rotationSpace=BCS_BoneSpace
    translationMode=BMM_Additive
    translationSpace=BCS_BoneSpace
    alphaBoolBlend {blendOption=Linear}
    boneToModify {boneName=base_skirt_r_C}
    rotation {yaw=6}
    translation {y=0.75}
   }
   n3 = Get is DBNO() -> (is DBNO:double)
   n4 = "Transform (Modify) Bone - Bone: base_skirt_l_C"(ComponentPose=@n2, Alpha=@n3, Translation="0,0,0", Rotation="0,-5,0", Scale="1,1,1") {
    rotationMode=BMM_Additive
    rotationSpace=BCS_BoneSpace
    translationMode=BMM_Additive
    translationSpace=BCS_BoneSpace
    alphaBoolBlend {blendOption=Linear}
    boneToModify {boneName=base_skirt_l_C}
    rotation {yaw=-5}
   }
   n5 = "Transform (Modify) Bone - Bone: base_skirt_r_C"(ComponentPose=@n4, Alpha=@n3, Translation="0,0,0", Rotation="0,-5,0", Scale="1,1,1") {
    rotationMode=BMM_Additive
    rotationSpace=BCS_BoneSpace
    translationMode=BMM_Additive
    translationSpace=BCS_BoneSpace
    alphaBoolBlend {blendOption=Linear}
    boneToModify {boneName=base_skirt_r_C}
    rotation {yaw=-5}
   }
   n6 = "Transform (Modify) Bone - Bone: upperarm_twist_01_extra_l"(ComponentPose=@n5, Alpha=@n3, Translation="3.376,0.931,1.934", Rotation="0,0,0", Scale="0.9,0.9,0.9") {
    rotationMode=BMM_Additive
    rotationSpace=BCS_BoneSpace
    scaleMode=BMM_Additive
    scaleSpace=BCS_BoneSpace
    translationMode=BMM_Additive
    translationSpace=BCS_BoneSpace
    alphaBoolBlend {blendOption=Linear}
    boneToModify {boneName=upperarm_twist_01_extra_l}
    scale=<0.9 0.9 0.9>
    translation=<3.376 0.931 1.934>
   }
   n7 = "Transform (Modify) Bone - Bone: base_armArmour_01"(ComponentPose=@n6, Alpha=@n3, Translation="-1.175,0.566,-2.101", Rotation="7.619,-19.060,-4.775", Scale="0.9,0.9,0.9") {
    rotationMode=BMM_Additive
    rotationSpace=BCS_BoneSpace
    scaleMode=BMM_Additive
    scaleSpace=BCS_BoneSpace
    translationMode=BMM_Additive
    translationSpace=BCS_BoneSpace
    alphaBoolBlend {blendOption=Linear}
    boneToModify {boneName=base_armArmour_01}
    rotation=<rot 7.619 -19.06 -4.775>
    scale=<0.9 0.9 0.9>
    translation=<-1.175 0.566 -2.101>
   }
   n8 = "Transform (Modify) Bone - Bone: base_tie"(ComponentPose=@n7, Alpha=@n3, Translation="0,0,0", Rotation="0,-15,0", Scale="0.7,1,1") {
    rotationMode=BMM_Additive
    rotationSpace=BCS_BoneSpace
    scaleMode=BMM_Additive
    scaleSpace=BCS_BoneSpace
    translationMode=BMM_Additive
    translationSpace=BCS_BoneSpace
    alphaBoolBlend {blendOption=Linear}
    boneToModify {boneName=base_tie}
    rotation {yaw=-15}
    scale {x=0.7}
   }
   n9 = Get is Parachuting() -> (is Parachuting:double)
   n10 = "Transform (Modify) Bone - Bone: base_upperarm_twist_01_extra_l"(ComponentPose=@n8, Alpha=@n9, Translation="1,-0.8,-1", Rotation="0,0,0", Scale="0.9,0.9,0.9") {
    scaleMode=BMM_Additive
    scaleSpace=BCS_BoneSpace
    translationMode=BMM_Additive
    translationSpace=BCS_BoneSpace
    alphaBoolBlend {blendOption=Linear}
    boneToModify {boneName=base_upperarm_twist_01_extra_l}
    scale=<0.9 0.9 0.9>
    translation=<1 -0.8 -1>
   }
   n11 = "Transform (Modify) Bone - Bone: base_armArmour_01"(ComponentPose=@n10, Alpha=@n9, Translation="0,0,1", Rotation="5,0,0", Scale="0.9,0.9,0.9") {
    rotationMode=BMM_Additive
    rotationSpace=BCS_BoneSpace
    scaleMode=BMM_Additive
    scaleSpace=BCS_BoneSpace
    translationMode=BMM_Additive
    translationSpace=BCS_BoneSpace
    alphaBoolBlend {blendOption=Linear}
    boneToModify {boneName=base_armArmour_01}
    rotation {pitch=5}
    scale=<0.9 0.9 0.9>
    translation {z=1}
   }
   n12 = "Transform (Modify) Bone - Bone: base_upperarm_twist_01_extra_l"!(ComponentPose=@n11, bAlphaBoolEnabled:bool=True, Translation="0,0,1", Rotation="-0.000,0.000,31.456", Scale="0.9,0.9,0.9") -> (Pose) {
    alphaInputType=Bool
    rotationMode=BMM_Additive
    rotationSpace=BCS_BoneSpace
    scaleMode=BMM_Additive
    scaleSpace=BCS_BoneSpace
    translationMode=BMM_Additive
    translationSpace=BCS_BoneSpace
    alphaBoolBlend {blendInTime=0.1, blendOption=Linear, blendOutTime=0.4}
    boneToModify {boneName=base_upperarm_twist_01_extra_l}
    rotation=<rot 0 0 31.456>
    scale=<0.9 0.9 0.9>
    translation {z=1}
   }
   n13 = "Transform (Modify) Bone - Bone: base_skirt_l_A"!(ComponentPose=@n12, bAlphaBoolEnabled:bool=True, Translation="0,-1.153,0.082", Rotation="-2.851,8.366,5.955", Scale="1,1,1") -> (Pose) {
    alphaInputType=Bool
    rotationMode=BMM_Additive
    rotationSpace=BCS_BoneSpace
    translationMode=BMM_Additive
    translationSpace=BCS_BoneSpace
    alphaBoolBlend {blendOption=Linear}
    boneToModify {boneName=base_skirt_l_A}
    rotation=<rot -2.851 8.366 5.955>
    translation {y=-1.153, z=0.082}
   }
   n14 = "Transform (Modify) Bone - Bone: base_skirt_l_B"!(ComponentPose=@n13, bAlphaBoolEnabled:bool=True, Translation="0.241,-0.310,2.816", Rotation="-12.703,-14.290,-13.613", Scale="1,1,1") -> (Pose) {
    alphaInputType=Bool
    rotationMode=BMM_Additive
    rotationSpace=BCS_BoneSpace
    translationMode=BMM_Additive
    translationSpace=BCS_BoneSpace
    alphaBoolBlend {blendOption=Linear}
    boneToModify {boneName=base_skirt_l_B}
    rotation=<rot -12.703 -14.29 -13.613>
    translation=<0.241 -0.31 2.816>
   }
   n15 = "Transform (Modify) Bone - Bone: base_skirtArmour"!(ComponentPose=@n14, bAlphaBoolEnabled:bool=True, Translation="-0.000,0.916,0.000", Rotation="-0.059,2.555,1.437", Scale="1,1,1") -> (Pose) {
    alphaInputType=Bool
    rotationMode=BMM_Additive
    rotationSpace=BCS_BoneSpace
    translationMode=BMM_Additive
    translationSpace=BCS_BoneSpace
    alphaBoolBlend {blendOption=Linear}
    boneToModify {boneName=base_skirtArmour}
    rotation=<rot -0.059 2.555 1.437>
    translation {y=0.916}
   }
   n16 = "Transform (Modify) Bone - Bone: base_skirt_r_A"!(ComponentPose=@n15, bAlphaBoolEnabled:bool=True, Translation="0,0,0", Rotation="10.035,2.936,-0.068", Scale="1,1,1") -> (Pose) {
    alphaInputType=Bool
    rotationMode=BMM_Additive
    rotationSpace=BCS_BoneSpace
    translationMode=BMM_Additive
    translationSpace=BCS_BoneSpace
    alphaBoolBlend {blendOption=Linear}
    boneToModify {boneName=base_skirt_r_A}
    rotation=<rot 10.035 2.936 -0.068>
   }
   n17 = Anim Dynamics - Bone: base_jacket_fn_l!(ComponentPose=@n16, Alpha=1, GravityOverride:Vector="0,0,-980", ExternalForce="0,0,0") -> (Pose) {
    angularBiasOverride=1
    angularDampingOverride=1
    angularSpringConstant=20000
    bAngularSpring=true
    bOverrideAngularBias=true
    bOverrideAngularDamping=true
    bUseGravityOverride=true
    gravityScale=0.35
    simulationSpace=RootRelative
    alphaBoolBlend {blendOption=Linear}
    boundBone {boneName=base_jacket_fn_l}
    componentAppliedLinearAccClamp=<3000 3000 3000>
    componentLinearAccScale=<0.3 0.3 0.3>
    componentLinearVelScale=<0.3 0.3 0.3>
    gravityOverride {z=-980}
    physicsBodyDefinitions [{
      boundBone {boneName=base_jacket_fn_l}
      boxExtents=<30 40 40>
      constraintSetup {angularLimitsMax=<15 15 5>, angularLimitsMin=<-15 -15 -15>, angularTarget=<1 0 0>}
      localJointOffset=<5 0 0>
     }]
   }
   n18 = Anim Dynamics - Bone: dyn_jacket_fn_l_01!(ComponentPose=@n17, Alpha=1, GravityOverride:Vector="0,0,-980", ExternalForce="0,0,0") -> (Pose) {
    angularBiasOverride=1
    angularDampingOverride=1
    angularSpringConstant=20000
    bAngularSpring=true
    bOverrideAngularBias=true
    bOverrideAngularDamping=true
    bUseGravityOverride=true
    gravityScale=0.35
    simulationSpace=RootRelative
    alphaBoolBlend {blendOption=Linear}
    boundBone {boneName=dyn_jacket_fn_l_01}
    componentAppliedLinearAccClamp=<3000 3000 3000>
    componentLinearAccScale=<0.3 0.3 0.3>
    componentLinearVelScale=<0.3 0.3 0.3>
    gravityOverride {z=-980}
    physicsBodyDefinitions [{
      boundBone {boneName=dyn_jacket_fn_l_01}
      boxExtents=<20 30 30>
      constraintSetup {angularLimitsMax=<15 15 5>, angularLimitsMin=<-15 -15 -15>, angularTarget=<1 0 0>}
      localJointOffset=<5 0 0>
     }]
   }
   n19 = Anim Dynamics - Bone: dyn_jacket_fn_r_01!(ComponentPose=@n18, Alpha=1, GravityOverride:Vector="0,0,-980", ExternalForce="0,0,0") -> (Pose) {
    angularBiasOverride=1
    angularDampingOverride=1
    angularSpringConstant=20000
    bAngularSpring=true
    bOverrideAngularBias=true
    bOverrideAngularDamping=true
    bUseGravityOverride=true
    gravityScale=0.35
    simulationSpace=RootRelative
    alphaBoolBlend {blendOption=Linear}
    boundBone {boneName=dyn_jacket_fn_r_01}
    componentAppliedLinearAccClamp=<3000 3000 3000>
    componentLinearAccScale=<0.3 0.3 0.3>
    componentLinearVelScale=<0.3 0.3 0.3>
    gravityOverride {z=-980}
    physicsBodyDefinitions [{
      boundBone {boneName=dyn_jacket_fn_r_01}
      boxExtents=<30 40 40>
      constraintSetup {angularLimitsMax=<15 15 5>, angularLimitsMin=<-15 -15 -15>, angularTarget=<1 0 0>}
      localJointOffset=<5 0 0>
     }]
   }
   n20 = Anim Dynamics - Bone: dyn_jacket_fn_r_02!(ComponentPose=@n19, Alpha=1, GravityOverride:Vector="0,0,-980", ExternalForce="0,0,0") -> (Pose) {
    angularBiasOverride=1
    angularDampingOverride=1
    angularSpringConstant=20000
    bAngularSpring=true
    bOverrideAngularBias=true
    bOverrideAngularDamping=true
    bUseGravityOverride=true
    gravityScale=0.35
    simulationSpace=RootRelative
    alphaBoolBlend {blendOption=Linear}
    boundBone {boneName=dyn_jacket_fn_r_02}
    componentAppliedLinearAccClamp=<3000 3000 3000>
    componentLinearAccScale=<0.3 0.3 0.3>
    componentLinearVelScale=<0.3 0.3 0.3>
    gravityOverride {z=-980}
    physicsBodyDefinitions [{
      boundBone {boneName=dyn_jacket_fn_r_02}
      boxExtents=<20 30 30>
      constraintSetup {angularLimitsMax=<15 15 5>, angularLimitsMin=<-15 -15 -15>, angularTarget=<1 0 0>}
      localJointOffset=<5 0 0>
     }]
   }
   n21 = "Constraint - base_zipper (PP)"!(ComponentPose=@n20, Alpha=1, ConstraintWeights_0=0.2, ConstraintWeights_1:float=0.8) -> (Pose) {
    alphaBoolBlend {blendOption=Linear}
    boneToModify {boneName=base_zipper}
    constraintSetup [{targetBone {boneName=dyn_jacket_fn_l_01}}, {targetBone {boneName=dyn_jacket_fn_l_02}}]
    constraintWeights [1,1]
   }
   n22 = Anim Dynamics - Bone: dyn_zipper!(ComponentPose=@n21, Alpha=1, GravityOverride:Vector="0,0,-980", ExternalForce) -> (Pose) {
    angularBiasOverride=1
    angularDampingOverride=1
    angularSpringConstant=20000
    bAngularSpring=true
    bOverrideAngularBias=true
    bOverrideAngularDamping=true
    bUseGravityOverride=true
    gravityScale=0.35
    simulationSpace=RootRelative
    alphaBoolBlend {blendOption=Linear}
    boundBone {boneName=dyn_zipper}
    componentAppliedLinearAccClamp=<3000 3000 3000>
    componentLinearAccScale=<0.3 0.3 0.3>
    componentLinearVelScale=<0.3 0.3 0.3>
    gravityOverride {z=-980}
    physicsBodyDefinitions [{
      boundBone {boneName=dyn_zipper}
      boxExtents=<20 30 30>
      constraintSetup {angularLimitsMax=<5 5 5>, angularLimitsMin=<-5 -5 -15>, angularTarget=<1 0 0>}
      localJointOffset=<5 0 0>
     }]
   }
   Output Pose(Result=Component To Local(ComponentPose=@n22)) {layerGroup=None}
   n23 = Control Rig!(
    Source=Copy Pose From Mesh() {bUseAttachedParent=true}
    Alpha
    Get DBNO:bool=Get bIsDBNO()
    Get Crouching:bool=Get bIsCrouching()
    Get Emote:bool=Get bIsPlayingEmote()
    Emote Float:double=Get is Emote Float() -> (is Emote Float:double)
   ) -> (Pose) {
    alphaBoolBlend {blendOption=Linear}
    controlRigAssetReference {blueprintRigClass=&/F_MED_CordSyrup_CtrlRig.F_MED_CordSyrup_CtrlRig_C}
   }
   n24 = "Constraint - base_button_01 (PPPPPP)"!(
    ComponentPose=Local To Component(LocalPose=@n23)
    Alpha=1
    ConstraintWeights_0=0.2361
    ConstraintWeights_1:float=0.225
    ConstraintWeights_2:float=0.2138
    ConstraintWeights_3:float=0.1858
    ConstraintWeights_4:float=0.0741
    ConstraintWeights_5:float=0.0652
   ) -> (Pose) {
    alphaBoolBlend {blendOption=Linear}
    boneToModify {boneName=base_button_01}
    constraintSetup [
     {targetBone {boneName=deform_pec_r}}
     {targetBone {boneName=spine_05}}
     {targetBone {boneName=spine_04}}
     {targetBone {boneName=clavicle_r}}
     {targetBone {boneName=deform_upperarm_fr_r}}
     {targetBone {boneName=upperarm_twist_01_r}}
    ]
    constraintWeights [1,1,1,1,1,1]
   }
   n25 = "Constraint - base_button_02 (PPPPP)"!(
    ComponentPose=@n24
    Alpha=1
    ConstraintWeights_0=0.3825
    ConstraintWeights_1:float=0.3053
    ConstraintWeights_2:float=0.2564
    ConstraintWeights_3:float=0.0504
    ConstraintWeights_4:float=0.0053
   ) -> (Pose) {
    alphaBoolBlend {blendOption=Linear}
    boneToModify {boneName=base_button_02}
    constraintSetup [
     {targetBone {boneName=deform_pec_r}}
     {targetBone {boneName=spine_04}}
     {targetBone {boneName=spine_05}}
     {targetBone {boneName=clavicle_r}}
     {targetBone {boneName=upperarm_twist_01_r}}
    ]
    constraintWeights [1,1,1,1,1]
   }
   n26 = "Constraint - base_button_03 (PPPPP)"!(
    ComponentPose=@n25
    Alpha=1
    ConstraintWeights_0=0.434
    ConstraintWeights_1:float=0.3696
    ConstraintWeights_2:float=0.1282
    ConstraintWeights_3:float=0.0523
    ConstraintWeights_4:float=0.0159
   ) -> (Pose) {
    alphaBoolBlend {blendOption=Linear}
    boneToModify {boneName=base_button_03}
    constraintSetup [
     {targetBone {boneName=deform_pec_r}}
     {targetBone {boneName=spine_04}}
     {targetBone {boneName=spine_05}}
     {targetBone {boneName=clavicle_r}}
     {targetBone {boneName=upperarm_twist_01_r}}
    ]
    constraintWeights [1,1,1,1,1]
   }
   n27 = "Constraint - base_armArmour_01 (PPPPP)"!(
    ComponentPose=@n26
    Alpha=1
    ConstraintWeights_0=0.4754
    ConstraintWeights_1:float=0.2081
    ConstraintWeights_2:float=0.1938
    ConstraintWeights_3:float=0.0815
    ConstraintWeights_4:float=0.0412
   ) -> (Pose) {
    alphaBoolBlend {blendOption=Linear}
    boneToModify {boneName=base_armArmour_01}
    constraintSetup [
     {targetBone {boneName=upperarm_twist_01_l}}
     {targetBone {boneName=deform_upperarm_up_l}}
     {targetBone {boneName=clavicle_l}}
     {targetBone {boneName=deform_upperarm_bk_l}}
     {targetBone {boneName=deform_upperarm_fr_l}}
    ]
    constraintWeights [1,1,1,1,1]
   }
   n28 = "Constraint - base_armArmour_02 (PPPPP)"!(
    ComponentPose=@n27
    Alpha=1
    ConstraintWeights_0=0.8081
    ConstraintWeights_1:float=0.0995
    ConstraintWeights_2:float=0.0403
    ConstraintWeights_3:float=0.0292
    ConstraintWeights_4:float=0.0228
   ) -> (Pose) {
    alphaBoolBlend {blendOption=Linear}
    boneToModify {boneName=base_armArmour_02}
    constraintSetup [
     {targetBone {boneName=upperarm_twist_01_l}}
     {targetBone {boneName=deform_upperarm_up_l}}
     {targetBone {boneName=clavicle_l}}
     {targetBone {boneName=deform_upperarm_bk_l}}
     {targetBone {boneName=deform_upperarm_fr_l}}
    ]
    constraintWeights [1,1,1,1,1]
   }
   n29 = "Constraint - base_armArmour_03 (PPP)"!(ComponentPose=@n28, Alpha=1, ConstraintWeights_0=0.8961, ConstraintWeights_1:float=0.0548, ConstraintWeights_2:float=0.049) -> (Pose) {
    alphaBoolBlend {blendOption=Linear}
    boneToModify {boneName=base_armArmour_03}
    constraintSetup [{targetBone {boneName=upperarm_twist_01_l}}, {targetBone {boneName=upperarm_twist_02_l}}, {targetBone {boneName=upperarm_l}}]
    constraintWeights [1,1,1]
   }
   n30 = "Constraint - base_zipper (PP)"(ComponentPose=@n29, Alpha=1, ConstraintWeights_0=0.876, ConstraintWeights_1=0.124) {
    alphaBoolBlend {blendOption=Linear}
    boneToModify {boneName=base_zipper}
    constraintSetup [{targetBone {boneName=dyn_jacket_fn_l_01}}, {targetBone {boneName=dyn_jacket_fn_l_02}}]
    constraintWeights [1,1]
   }
   n31 = "Constraint - base_tie (PP)"!(ComponentPose=@n30, Alpha=1, ConstraintWeights_0=0.8243, ConstraintWeights_1:float=0.1757) -> (Pose) {
    alphaBoolBlend {blendOption=Linear}
    boneToModify {boneName=base_tie}
    constraintSetup [{targetBone {boneName=spine_05}}, {targetBone {boneName=spine_04}}]
    constraintWeights [1,1]
   }
   n32 = "Constraint - base_shoeFlap_l (PP)"!(ComponentPose=@n31, Alpha=1, ConstraintWeights_0=0.5946, ConstraintWeights_1:float=0.4054) -> (Pose) {
    alphaBoolBlend {blendOption=Linear}
    boneToModify {boneName=base_shoeFlap_l}
    constraintSetup [{targetBone {boneName=calf_twist_01_l}}, {targetBone {boneName=foot_l}}]
    constraintWeights [1,1]
   }
   n33 = "Constraint - base_shoeSideFlap_l (PPPP)"!(
    ComponentPose=@n32
    Alpha=1
    ConstraintWeights_0=0.8637
    ConstraintWeights_1:float=0.0691
    ConstraintWeights_2:float=0.0349
    ConstraintWeights_3:float=0.0323
   ) -> (Pose) {
    alphaBoolBlend {blendOption=Linear}
    boneToModify {boneName=base_shoeSideFlap_l}
    constraintSetup [
     {targetBone {boneName=calf_twist_01_l}}
     {targetBone {boneName=foot_l}}
     {targetBone {boneName=calf_l}}
     {targetBone {boneName=calf_twist_02_l}}
    ]
    constraintWeights [1,1,1,1]
   }
   n34 = "Constraint - base_shoeSideFlap_r (PPPP)"!(
    ComponentPose=@n33
    Alpha=1
    ConstraintWeights_0=0.9348
    ConstraintWeights_1:float=0.027
    ConstraintWeights_2:float=0.0226
    ConstraintWeights_3:float=0.0156
   ) -> (Pose) {
    alphaBoolBlend {blendOption=Linear}
    boneToModify {boneName=base_shoeSideFlap_r}
    constraintSetup [
     {targetBone {boneName=calf_twist_01_l}}
     {targetBone {boneName=foot_l}}
     {targetBone {boneName=calf_l}}
     {targetBone {boneName=calf_twist_02_l}}
    ]
    constraintWeights [1,1,1,1]
   }
   Save cached pose 'Con2AnimDyn'(Pose=Component To Local(ComponentPose=@n34)) {cacheName=Con2AnimDyn}
   n35 = Anim Dynamics - Bone: dyn_armourString!(
    ComponentPose=Local To Component(LocalPose=Use cached pose 'Con2AnimDyn'())
    Alpha=1
    GravityOverride:Vector="0,0,-980"
    ExternalForce="0,0,0"
   ) -> (Pose) {
    angularDampingOverride=1
    angularSpringConstant=20000
    bAngularSpring=true
    bOverrideAngularDamping=true
    bUseGravityOverride=true
    gravityScale=0.35
    simulationSpace=RootRelative
    alphaBoolBlend {blendOption=Linear}
    boundBone {boneName=dyn_armourString}
    componentLinearAccScale=<1 1 1>
    componentLinearVelScale=<1 1 1>
    gravityOverride {z=-980}
    physicsBodyDefinitions [{
      boundBone {boneName=dyn_armourString}
      boxExtents=<20 30 30>
      constraintSetup {angularLimitsMax=<15 0 10>, angularLimitsMin=<-15 -25 -35>, angularTarget=<1 0 0>}
      localJointOffset=<8 0 0>
     }]
   }
   n36 = Anim Dynamics - Bone: dyn_armourStringBow_l!(ComponentPose=@n35, Alpha=1, GravityOverride:Vector="0,0,-980", ExternalForce="0,0,0") -> (Pose) {
    angularDampingOverride=1
    angularSpringConstant=20000
    bAngularSpring=true
    bOverrideAngularDamping=true
    bUseGravityOverride=true
    gravityScale=0.35
    simulationSpace=RootRelative
    alphaBoolBlend {blendOption=Linear}
    boundBone {boneName=dyn_armourStringBow_l}
    componentLinearAccScale=<1 1 1>
    componentLinearVelScale=<1 1 1>
    gravityOverride {z=-980}
    physicsBodyDefinitions [{
      boundBone {boneName=dyn_armourStringBow_l}
      boxExtents=<20 30 30>
      constraintSetup {angularLimitsMax=<15 15 15>, angularLimitsMin=<-15 -15 -5>, angularTarget=<1 0 0>}
      localJointOffset=<8 0 0>
     }]
   }
   n37 = Anim Dynamics - Bone: dyn_armourString_l_01!(ComponentPose=@n36, Alpha=1, GravityOverride:Vector="0,0,-980", ExternalForce="0,0,0") -> (Pose) {
    angularDampingOverride=1
    angularSpringConstant=20000
    bAngularSpring=true
    bOverrideAngularDamping=true
    bUseGravityOverride=true
    gravityScale=0.35
    simulationSpace=RootRelative
    alphaBoolBlend {blendOption=Linear}
    boundBone {boneName=dyn_armourString_l_01}
    componentLinearAccScale=<1 1 1>
    componentLinearVelScale=<1 1 1>
    gravityOverride {z=-980}
    physicsBodyDefinitions [{
      boundBone {boneName=dyn_armourString_l_01}
      boxExtents=<20 30 30>
      constraintSetup {angularLimitsMax=<15 10 15>, angularLimitsMin=<-15 -30 0>, angularTarget=<1 0 0>}
      localJointOffset=<10 0 0>
     }]
   }
   n38 = Anim Dynamics - Bone: dyn_armourString_l_02!(ComponentPose=@n37, Alpha=1, GravityOverride:Vector="0,0,-980", ExternalForce="0,0,0") -> (Pose) {
    angularDampingOverride=1
    angularSpringConstant=20000
    bAngularSpring=true
    bOverrideAngularDamping=true
    bUseGravityOverride=true
    gravityScale=0.35
    simulationSpace=RootRelative
    alphaBoolBlend {blendOption=Linear}
    boundBone {boneName=dyn_armourString_l_02}
    componentLinearAccScale=<1 1 1>
    componentLinearVelScale=<1 1 1>
    gravityOverride {z=-980}
    physicsBodyDefinitions [{
      boundBone {boneName=dyn_armourString_l_02}
      boxExtents=<20 30 30>
      constraintSetup {angularLimitsMax=<15 10 15>, angularLimitsMin=<-15 -30 0>, angularTarget=<1 0 0>}
      localJointOffset=<8 0 0>
     }]
   }
   n39 = Anim Dynamics - Bone: dyn_armourStringBow_r!(ComponentPose=@n38, Alpha=1, GravityOverride:Vector="0,0,-980", ExternalForce="0,0,0") -> (Pose) {
    angularDampingOverride=1
    angularSpringConstant=20000
    bAngularSpring=true
    bOverrideAngularDamping=true
    bUseGravityOverride=true
    gravityScale=0.35
    simulationSpace=RootRelative
    alphaBoolBlend {blendOption=Linear}
    boundBone {boneName=dyn_armourStringBow_r}
    componentLinearAccScale=<1 1 1>
    componentLinearVelScale=<1 1 1>
    gravityOverride {z=-980}
    physicsBodyDefinitions [{
      boundBone {boneName=dyn_armourStringBow_r}
      boxExtents=<20 30 30>
      constraintSetup {angularLimitsMax=<15 15 5>, angularLimitsMin=<-15 -15 -15>, angularTarget=<1 0 0>}
      localJointOffset=<6 0 0>
     }]
   }
   n40 = Anim Dynamics - Bone: dyn_armourString_r_01!(ComponentPose=@n39, Alpha=1, GravityOverride:Vector="0,0,-980", ExternalForce="0,0,0") -> (Pose) {
    angularDampingOverride=1
    angularSpringConstant=20000
    bAngularSpring=true
    bOverrideAngularDamping=true
    bUseGravityOverride=true
    gravityScale=0.35
    simulationSpace=RootRelative
    alphaBoolBlend {blendOption=Linear}
    boundBone {boneName=dyn_armourString_r_01}
    componentLinearAccScale=<1 1 1>
    componentLinearVelScale=<1 1 1>
    gravityOverride {z=-980}
    physicsBodyDefinitions [{
      boundBone {boneName=dyn_armourString_r_01}
      boxExtents=<20 30 30>
      constraintSetup {angularLimitsMax=<15 15 5>, angularLimitsMin=<-15 -15 -15>, angularTarget=<1 0 0>}
      localJointOffset=<8 0 0>
     }]
   }
   n41 = Anim Dynamics - Bone: dyn_armourString_r_02!(ComponentPose=@n40, Alpha=1, GravityOverride:Vector="0,0,-980", ExternalForce="0,0,0") -> (Pose) {
    angularDampingOverride=1
    angularSpringConstant=20000
    bAngularSpring=true
    bOverrideAngularDamping=true
    bUseGravityOverride=true
    gravityScale=0.35
    simulationSpace=RootRelative
    alphaBoolBlend {blendOption=Linear}
    boundBone {boneName=dyn_armourString_r_02}
    componentLinearAccScale=<1 1 1>
    componentLinearVelScale=<1 1 1>
    gravityOverride {z=-980}
    physicsBodyDefinitions [{
      boundBone {boneName=dyn_armourString_r_02}
      boxExtents=<20 30 30>
      constraintSetup {angularLimitsMax=<15 15 15>, angularLimitsMin=<-15 -15 -15>, angularTarget=<1 0 0>}
      localJointOffset=<6 0 0>
     }]
   }
   n42 = Anim Dynamics - Bone: dyn_shoe_l_01!(ComponentPose=@n41, Alpha=1, GravityOverride:Vector="0,0,-980", ExternalForce="0,0,0") -> (Pose) {
    angularDampingOverride=1
    angularSpringConstant=20000
    bAngularSpring=true
    bOverrideAngularDamping=true
    bUseGravityOverride=true
    gravityScale=0.35
    simulationSpace=RootRelative
    alphaBoolBlend {blendOption=Linear}
    boundBone {boneName=dyn_shoe_l_01}
    componentLinearAccScale=<1 1 1>
    componentLinearVelScale=<1 1 1>
    gravityOverride {z=-980}
    physicsBodyDefinitions [{
      boundBone {boneName=dyn_shoe_l_01}
      boxExtents=<20 30 30>
      constraintSetup {angularLimitsMax=<15 15 15>, angularLimitsMin=<-15 -15 -15>, angularTarget=<1 0 0>}
      localJointOffset=<4 0 0>
     }]
   }
   n43 = Anim Dynamics - Bone: dyn_shoe_l_02!(ComponentPose=@n42, Alpha=1, GravityOverride:Vector="0,0,-980", ExternalForce="0,0,0") -> (Pose) {
    angularDampingOverride=1
    angularSpringConstant=20000
    bAngularSpring=true
    bOverrideAngularDamping=true
    bUseGravityOverride=true
    gravityScale=0.35
    simulationSpace=RootRelative
    alphaBoolBlend {blendOption=Linear}
    boundBone {boneName=dyn_shoe_l_02}
    componentLinearAccScale=<1 1 1>
    componentLinearVelScale=<1 1 1>
    gravityOverride {z=-980}
    physicsBodyDefinitions [{
      boundBone {boneName=dyn_shoe_l_02}
      boxExtents=<20 30 30>
      constraintSetup {angularLimitsMax=<15 15 15>, angularLimitsMin=<-15 -15 -15>, angularTarget=<1 0 0>}
      localJointOffset=<4 0 0>
     }]
   }
   n44 = Anim Dynamics - Bone: dyn_shoeFlap_l_01!(ComponentPose=@n43, Alpha=1, GravityOverride:Vector="0,0,-980", ExternalForce="0,0,0") -> (Pose) {
    angularDampingOverride=1
    angularSpringConstant=20000
    bAngularSpring=true
    bOverrideAngularDamping=true
    bUseGravityOverride=true
    gravityScale=0.35
    simulationSpace=RootRelative
    alphaBoolBlend {blendOption=Linear}
    boundBone {boneName=dyn_shoeFlap_l_01}
    componentLinearAccScale=<1 1 1>
    componentLinearVelScale=<1 1 1>
    gravityOverride {z=-980}
    physicsBodyDefinitions [{
      boundBone {boneName=dyn_shoeFlap_l_01}
      boxExtents=<20 30 30>
      constraintSetup {angularLimitsMax=<15 15 15>, angularLimitsMin=<-15 -15 -15>, angularTarget=<1 0 0>}
      localJointOffset=<4 0 0>
     }]
   }
   n45 = Anim Dynamics - Bone: dyn_shoeFlap_l_02!(ComponentPose=@n44, Alpha=1, GravityOverride:Vector="0,0,-980", ExternalForce="0,0,0") -> (Pose) {
    angularDampingOverride=1
    angularSpringConstant=20000
    bAngularSpring=true
    bOverrideAngularDamping=true
    bUseGravityOverride=true
    gravityScale=0.35
    simulationSpace=RootRelative
    alphaBoolBlend {blendOption=Linear}
    boundBone {boneName=dyn_shoeFlap_l_02}
    componentLinearAccScale=<1 1 1>
    componentLinearVelScale=<1 1 1>
    gravityOverride {z=-980}
    physicsBodyDefinitions [{
      boundBone {boneName=dyn_shoeFlap_l_02}
      boxExtents=<20 30 30>
      constraintSetup {angularLimitsMax=<15 15 15>, angularLimitsMin=<-15 -15 -15>, angularTarget=<1 0 0>}
      localJointOffset=<4 0 0>
     }]
   }
   n46 = "Transform (Modify) Bone - Bone: dyn_armourString"!(ComponentPose=@n45, bAlphaBoolEnabled:bool=True, Translation="0,0,0", Rotation="10,-5,5", Scale="1,1,1") -> (Pose) {
    alphaInputType=Bool
    rotationMode=BMM_Additive
    rotationSpace=BCS_BoneSpace
    translationMode=BMM_Additive
    translationSpace=BCS_BoneSpace
    alphaBoolBlend {blendOption=Linear}
    boneToModify {boneName=dyn_armourString}
    rotation=<rot 10 -5 5>
   }
   Save cached pose 'AnimDyn2RBs'(Pose=Component To Local(ComponentPose=@n46)) {cacheName=AnimDyn2RBs}
   n47 = Anim Dynamics - Bone: dyn_skirt_r_C_01(ComponentPose=Local To Component(LocalPose=Use cached pose 'AnimDyn2RBs'()), Alpha=1, ExternalForce="0,0,0") {
    angularDampingOverride=1
    angularSpringConstant=20000
    bAngularSpring=true
    bOverrideAngularDamping=true
    bUseGravityOverride=true
    gravityScale=0.35
    alphaBoolBlend {blendOption=Linear}
    boundBone {boneName=dyn_skirt_r_C_01}
    componentAppliedLinearAccClamp=<3000 3000 3000>
    componentLinearAccScale=<0.5 0.5 0.5>
    componentLinearVelScale=<0.5 0.5 0.5>
    gravityOverride {z=-980}
    physicsBodyDefinitions [{
      boundBone {boneName=dyn_skirt_r_C_01}
      boxExtents=<20 30 30>
      constraintSetup {angularLimitsMax=<0 0 5>, angularLimitsMin=<0 0 -5>, angularTarget=<1 0 0>}
      localJointOffset=<4 0 0>
     }]
    relativeSpaceBone {boneName=pelvis}
   }
   n48 = Anim Dynamics - Bone: dyn_skirt_r_C_02(ComponentPose=@n47, Alpha=1, ExternalForce="0,0,0") {
    angularDampingOverride=1
    angularSpringConstant=20000
    bAngularSpring=true
    bOverrideAngularDamping=true
    bUseGravityOverride=true
    gravityScale=0.35
    alphaBoolBlend {blendOption=Linear}
    boundBone {boneName=dyn_skirt_r_C_02}
    componentAppliedLinearAccClamp=<3000 3000 3000>
    componentLinearAccScale=<0.5 0.5 0.5>
    componentLinearVelScale=<0.5 0.5 0.5>
    gravityOverride {z=-980}
    physicsBodyDefinitions [{
      boundBone {boneName=dyn_skirt_r_C_02}
      boxExtents=<20 30 30>
      constraintSetup {angularLimitsMax=<0 0 5>, angularLimitsMin=<0 0 -5>, angularTarget=<1 0 0>}
      localJointOffset=<4 0 0>
     }]
    relativeSpaceBone {boneName=pelvis}
   }
   n49 = Anim Dynamics - Bone: dyn_skirt_r_C_03(ComponentPose=@n48, Alpha=1, ExternalForce="0,0,0") {
    angularDampingOverride=1
    angularSpringConstant=20000
    bAngularSpring=true
    bOverrideAngularDamping=true
    bUseGravityOverride=true
    gravityScale=0.35
    alphaBoolBlend {blendOption=Linear}
    boundBone {boneName=dyn_skirt_r_C_03}
    componentAppliedLinearAccClamp=<3000 3000 3000>
    componentLinearAccScale=<0.5 0.5 0.5>
    componentLinearVelScale=<0.5 0.5 0.5>
    gravityOverride {z=-980}
    physicsBodyDefinitions [{
      boundBone {boneName=dyn_skirt_r_C_03}
      boxExtents=<20 30 30>
      constraintSetup {angularLimitsMax=<0 0 5>, angularLimitsMin=<0 0 -5>, angularTarget=<1 0 0>}
      localJointOffset=<4 0 0>
     }]
    relativeSpaceBone {boneName=pelvis}
   }
   n50 = Anim Dynamics - Bone: dyn_skirt_r_B_01(ComponentPose=@n49, Alpha=1, ExternalForce="0,0,0") {
    angularDampingOverride=1
    angularSpringConstant=20000
    bAngularSpring=true
    bOverrideAngularDamping=true
    bUseGravityOverride=true
    gravityScale=0.35
    alphaBoolBlend {blendOption=Linear}
    boundBone {boneName=dyn_skirt_r_B_01}
    componentAppliedLinearAccClamp=<3000 3000 3000>
    componentLinearAccScale=<0.5 0.5 0.5>
    componentLinearVelScale=<0.5 0.5 0.5>
    gravityOverride {z=-980}
    physicsBodyDefinitions [{
      boundBone {boneName=dyn_skirt_r_B_01}
      boxExtents=<20 30 30>
      constraintSetup {angularLimitsMax=<0 0 0>, angularLimitsMin=<0 -10 -5>, angularTarget=<1 0 0>}
      localJointOffset=<4 0 0>
     }]
    relativeSpaceBone {boneName=pelvis}
   }
   n51 = Anim Dynamics - Bone: dyn_skirt_r_B_02(ComponentPose=@n50, Alpha=1, ExternalForce="0,0,0") {
    angularDampingOverride=1
    angularSpringConstant=20000
    bAngularSpring=true
    bOverrideAngularDamping=true
    bUseGravityOverride=true
    gravityScale=0.35
    alphaBoolBlend {blendOption=Linear}
    boundBone {boneName=dyn_skirt_r_B_02}
    componentAppliedLinearAccClamp=<3000 3000 3000>
    componentLinearAccScale=<0.5 0.5 0.5>
    componentLinearVelScale=<0.5 0.5 0.5>
    gravityOverride {z=-980}
    physicsBodyDefinitions [{
      boundBone {boneName=dyn_skirt_r_B_02}
      boxExtents=<20 30 30>
      constraintSetup {angularLimitsMax=<0 2 0>, angularLimitsMin=<0 -7 -5>, angularTarget=<1 0 0>}
      localJointOffset=<4 0 0>
     }]
    relativeSpaceBone {boneName=pelvis}
   }
   n52 = Anim Dynamics - Bone: dyn_skirt_r_B_03(ComponentPose=@n51, Alpha=1, ExternalForce="0,0,0") {
    angularDampingOverride=1
    angularSpringConstant=20000
    bAngularSpring=true
    bOverrideAngularDamping=true
    bUseGravityOverride=true
    gravityScale=0.35
    alphaBoolBlend {blendOption=Linear}
    boundBone {boneName=dyn_skirt_r_B_03}
    componentAppliedLinearAccClamp=<3000 3000 3000>
    componentLinearAccScale=<0.5 0.5 0.5>
    componentLinearVelScale=<0.5 0.5 0.5>
    gravityOverride {z=-980}
    physicsBodyDefinitions [{
      boundBone {boneName=dyn_skirt_r_B_03}
      boxExtents=<20 30 30>
      constraintSetup {angularLimitsMax=<0 2 0>, angularLimitsMin=<0 -7 0>, angularTarget=<1 0 0>}
      localJointOffset=<4 0 0>
     }]
    relativeSpaceBone {boneName=pelvis}
   }
   n53 = Anim Dynamics - Bone: dyn_skirt_r_A_01(ComponentPose=@n52, Alpha=1, ExternalForce="0,0,0") {
    angularDampingOverride=1
    angularSpringConstant=20000
    bAngularSpring=true
    bOverrideAngularDamping=true
    bUseGravityOverride=true
    gravityScale=0.35
    alphaBoolBlend {blendOption=Linear}
    boundBone {boneName=dyn_skirt_r_A_01}
    componentAppliedLinearAccClamp=<3000 3000 3000>
    componentLinearAccScale=<0.5 0.5 0.5>
    componentLinearVelScale=<0.5 0.5 0.5>
    gravityOverride {z=-980}
    physicsBodyDefinitions [{
      boundBone {boneName=dyn_skirt_r_A_01}
      boxExtents=<20 30 30>
      constraintSetup {angularLimitsMax=<0 0 0>, angularLimitsMin=<0 0 -15>, angularTarget=<1 0 0>}
      localJointOffset=<4 0 0>
     }]
    relativeSpaceBone {boneName=pelvis}
   }
   n54 = Anim Dynamics - Bone: dyn_skirt_r_A_02(ComponentPose=@n53, Alpha=1, ExternalForce="0,0,0") {
    angularDampingOverride=1
    angularSpringConstant=20000
    bAngularSpring=true
    bOverrideAngularDamping=true
    bUseGravityOverride=true
    gravityScale=0.35
    alphaBoolBlend {blendOption=Linear}
    boundBone {boneName=dyn_skirt_r_A_02}
    componentAppliedLinearAccClamp=<3000 3000 3000>
    componentLinearAccScale=<0.5 0.5 0.5>
    componentLinearVelScale=<0.5 0.5 0.5>
    gravityOverride {z=-980}
    physicsBodyDefinitions [{
      boundBone {boneName=dyn_skirt_r_A_02}
      boxExtents=<20 30 30>
      constraintSetup {angularLimitsMax=<0 0 0>, angularLimitsMin=<0 0 -15>, angularTarget=<1 0 0>}
      localJointOffset=<4 0 0>
     }]
    relativeSpaceBone {boneName=pelvis}
   }
   n55 = Anim Dynamics - Bone: dyn_skirt_r_A_03(ComponentPose=@n54, Alpha=1, ExternalForce="0,0,0") {
    angularDampingOverride=1
    angularSpringConstant=20000
    bAngularSpring=true
    bOverrideAngularDamping=true
    bUseGravityOverride=true
    gravityScale=0.35
    alphaBoolBlend {blendOption=Linear}
    boundBone {boneName=dyn_skirt_r_A_03}
    componentAppliedLinearAccClamp=<3000 3000 3000>
    componentLinearAccScale=<0.5 0.5 0.5>
    componentLinearVelScale=<0.5 0.5 0.5>
    gravityOverride {z=-980}
    physicsBodyDefinitions [{
      boundBone {boneName=dyn_skirt_r_A_03}
      boxExtents=<20 30 30>
      constraintSetup {angularLimitsMax=<0 0 0>, angularLimitsMin=<0 0 -10>, angularTarget=<1 0 0>}
      localJointOffset=<4 0 0>
     }]
    relativeSpaceBone {boneName=pelvis}
   }
   n56 = Anim Dynamics - Bone: dyn_skirt_l_C_01(ComponentPose=@n55, Alpha=1, ExternalForce="0,0,0") {
    angularDampingOverride=1
    angularSpringConstant=20000
    bAngularSpring=true
    bOverrideAngularDamping=true
    bUseGravityOverride=true
    gravityScale=0.35
    alphaBoolBlend {blendOption=Linear}
    boundBone {boneName=dyn_skirt_l_C_01}
    componentAppliedLinearAccClamp=<3000 3000 3000>
    componentLinearAccScale=<0.5 0.5 0.5>
    componentLinearVelScale=<0.5 0.5 0.5>
    gravityOverride {z=-980}
    physicsBodyDefinitions [{
      boundBone {boneName=dyn_skirt_l_C_01}
      boxExtents=<20 30 30>
      constraintSetup {angularLimitsMax=<0 0 5>, angularLimitsMin=<0 0 0>, angularTarget=<1 0 0>}
      localJointOffset=<4 0 0>
     }]
    relativeSpaceBone {boneName=pelvis}
   }
   n57 = Anim Dynamics - Bone: dyn_skirt_l_C_02(ComponentPose=@n56, Alpha=1, ExternalForce="0,0,0") {
    angularDampingOverride=1
    angularSpringConstant=20000
    bAngularSpring=true
    bOverrideAngularDamping=true
    bUseGravityOverride=true
    gravityScale=0.35
    alphaBoolBlend {blendOption=Linear}
    boundBone {boneName=dyn_skirt_l_C_02}
    componentAppliedLinearAccClamp=<3000 3000 3000>
    componentLinearAccScale=<0.5 0.5 0.5>
    componentLinearVelScale=<0.5 0.5 0.5>
    gravityOverride {z=-980}
    physicsBodyDefinitions [{
      boundBone {boneName=dyn_skirt_l_C_02}
      boxExtents=<20 30 30>
      constraintSetup {angularLimitsMax=<0 0 5>, angularLimitsMin=<0 0 0>, angularTarget=<1 0 0>}
      localJointOffset=<4 0 0>
     }]
    relativeSpaceBone {boneName=pelvis}
   }
   n58 = Anim Dynamics - Bone: dyn_skirt_l_C_03(ComponentPose=@n57, Alpha=1, ExternalForce="0,0,0") {
    angularDampingOverride=1
    angularSpringConstant=20000
    bAngularSpring=true
    bOverrideAngularDamping=true
    bUseGravityOverride=true
    gravityScale=0.35
    alphaBoolBlend {blendOption=Linear}
    boundBone {boneName=dyn_skirt_l_C_03}
    componentAppliedLinearAccClamp=<3000 3000 3000>
    componentLinearAccScale=<0.5 0.5 0.5>
    componentLinearVelScale=<0.5 0.5 0.5>
    gravityOverride {z=-980}
    physicsBodyDefinitions [{
      boundBone {boneName=dyn_skirt_l_C_03}
      boxExtents=<20 30 30>
      constraintSetup {angularLimitsMax=<0 0 5>, angularLimitsMin=<0 0 0>, angularTarget=<1 0 0>}
      localJointOffset=<4 0 0>
     }]
    relativeSpaceBone {boneName=pelvis}
   }
   n59 = Anim Dynamics - Bone: dyn_skirt_l_B_01(ComponentPose=@n58, Alpha=1, ExternalForce="0,0,0") {
    angularDampingOverride=1
    angularSpringConstant=20000
    bAngularSpring=true
    bOverrideAngularDamping=true
    bUseGravityOverride=true
    gravityScale=0.35
    alphaBoolBlend {blendOption=Linear}
    boundBone {boneName=dyn_skirt_l_B_01}
    componentAppliedLinearAccClamp=<3000 3000 3000>
    componentLinearAccScale=<0.5 0.5 0.5>
    componentLinearVelScale=<0.5 0.5 0.5>
    gravityOverride {z=-980}
    physicsBodyDefinitions [{
      boundBone {boneName=dyn_skirt_l_B_01}
      boxExtents=<20 30 30>
      constraintSetup {angularLimitsMax=<0 -3 0>, angularLimitsMin=<0 -10 0>, angularTarget=<1 0 0>}
      localJointOffset=<4 0 0>
     }]
    relativeSpaceBone {boneName=pelvis}
   }
   n60 = Anim Dynamics - Bone: dyn_skirt_l_B_02(ComponentPose=@n59, Alpha=1, ExternalForce="0,0,0") {
    angularDampingOverride=1
    angularSpringConstant=20000
    bAngularSpring=true
    bOverrideAngularDamping=true
    bUseGravityOverride=true
    gravityScale=0.35
    alphaBoolBlend {blendOption=Linear}
    boundBone {boneName=dyn_skirt_l_B_02}
    componentAppliedLinearAccClamp=<3000 3000 3000>
    componentLinearAccScale=<0.5 0.5 0.5>
    componentLinearVelScale=<0.5 0.5 0.5>
    gravityOverride {z=-980}
    physicsBodyDefinitions [{
      boundBone {boneName=dyn_skirt_l_B_02}
      boxExtents=<20 30 30>
      constraintSetup {angularLimitsMax=<0 0 0>, angularLimitsMin=<0 -7 0>, angularTarget=<1 0 0>}
      localJointOffset=<4 0 0>
     }]
    relativeSpaceBone {boneName=pelvis}
   }
   n61 = Anim Dynamics - Bone: dyn_skirt_l_B_03(ComponentPose=@n60, Alpha=1, ExternalForce="0,0,0") {
    angularDampingOverride=1
    angularSpringConstant=20000
    bAngularSpring=true
    bOverrideAngularDamping=true
    bUseGravityOverride=true
    gravityScale=0.35
    alphaBoolBlend {blendOption=Linear}
    boundBone {boneName=dyn_skirt_l_B_03}
    componentAppliedLinearAccClamp=<3000 3000 3000>
    componentLinearAccScale=<0.5 0.5 0.5>
    componentLinearVelScale=<0.5 0.5 0.5>
    gravityOverride {z=-980}
    physicsBodyDefinitions [{
      boundBone {boneName=dyn_skirt_l_B_03}
      boxExtents=<20 30 30>
      constraintSetup {angularLimitsMax=<0 0 0>, angularLimitsMin=<0 -5 0>, angularTarget=<1 0 0>}
      localJointOffset=<4 0 0>
     }]
    relativeSpaceBone {boneName=pelvis}
   }
   n62 = Anim Dynamics - Bone: dyn_skirtArmour_01(ComponentPose=@n61, Alpha=1, ExternalForce="0,0,0") {
    angularDampingOverride=1
    angularSpringConstant=20000
    bAngularSpring=true
    bOverrideAngularDamping=true
    bUseGravityOverride=true
    gravityScale=0.35
    alphaBoolBlend {blendOption=Linear}
    boundBone {boneName=dyn_skirtArmour_01}
    componentAppliedLinearAccClamp=<3000 3000 3000>
    componentLinearAccScale=<0.3 0.3 0.3>
    componentLinearVelScale=<0.3 0.3 0.3>
    gravityOverride {z=-980}
    physicsBodyDefinitions [{
      boundBone {boneName=dyn_skirtArmour_01}
      boxExtents=<20 30 30>
      constraintSetup {angularLimitsMax=<0 0 5>, angularLimitsMin=<0 0 0>, angularTarget=<1 0 0>}
      localJointOffset=<6 0 0>
     }]
    relativeSpaceBone {boneName=pelvis}
   }
   n63 = Anim Dynamics - Bone: dyn_skirtArmour_02(ComponentPose=@n62, Alpha=1, ExternalForce="0,0,0") {
    angularDampingOverride=1
    angularSpringConstant=20000
    bAngularSpring=true
    bOverrideAngularDamping=true
    bUseGravityOverride=true
    gravityScale=0.35
    alphaBoolBlend {blendOption=Linear}
    boundBone {boneName=dyn_skirtArmour_02}
    componentAppliedLinearAccClamp=<3000 3000 3000>
    componentLinearAccScale=<0.3 0.3 0.3>
    componentLinearVelScale=<0.3 0.3 0.3>
    gravityOverride {z=-980}
    physicsBodyDefinitions [{
      boundBone {boneName=dyn_skirtArmour_02}
      boxExtents=<20 30 30>
      constraintSetup {angularLimitsMax=<0 0 5>, angularLimitsMin=<0 0 0>, angularTarget=<1 0 0>}
      localJointOffset=<6 0 0>
     }]
    relativeSpaceBone {boneName=pelvis}
   }
   n64 = Anim Dynamics - Bone: dyn_skirtArmour_03(ComponentPose=@n63, Alpha=1, ExternalForce="0,0,0") {
    angularDampingOverride=1
    angularSpringConstant=20000
    bAngularSpring=true
    bOverrideAngularDamping=true
    bUseGravityOverride=true
    gravityScale=0.35
    alphaBoolBlend {blendOption=Linear}
    boundBone {boneName=dyn_skirtArmour_03}
    componentAppliedLinearAccClamp=<3000 3000 3000>
    componentLinearAccScale=<0.3 0.3 0.3>
    componentLinearVelScale=<0.3 0.3 0.3>
    gravityOverride {z=-980}
    physicsBodyDefinitions [{
      boundBone {boneName=dyn_skirtArmour_03}
      boxExtents=<20 30 30>
      constraintSetup {angularLimitsMax=<0 0 5>, angularLimitsMin=<0 0 0>, angularTarget=<1 0 0>}
      localJointOffset=<6 0 0>
     }]
    relativeSpaceBone {boneName=pelvis}
   }
   n65 = Anim Dynamics - Bone: dyn_skirt_l_A_01(ComponentPose=@n64, Alpha=1, ExternalForce="0,0,0") {
    angularDampingOverride=1
    angularSpringConstant=20000
    bAngularSpring=true
    bOverrideAngularDamping=true
    bUseGravityOverride=true
    gravityScale=0.35
    alphaBoolBlend {blendOption=Linear}
    boundBone {boneName=dyn_skirt_l_A_01}
    componentAppliedLinearAccClamp=<3000 3000 3000>
    componentLinearAccScale=<0.5 0.5 0.5>
    componentLinearVelScale=<0.5 0.5 0.5>
    gravityOverride {z=-980}
    physicsBodyDefinitions [{
      boundBone {boneName=dyn_skirt_l_A_01}
      boxExtents=<20 30 30>
      constraintSetup {angularLimitsMax=<0 0 0>, angularLimitsMin=<0 0 -15>, angularTarget=<1 0 0>}
      localJointOffset=<4 0 0>
     }]
    relativeSpaceBone {boneName=pelvis}
   }
   n66 = Anim Dynamics - Bone: dyn_skirt_l_A_02(ComponentPose=@n65, Alpha=1, ExternalForce="0,0,0") {
    angularDampingOverride=1
    angularSpringConstant=20000
    bAngularSpring=true
    bOverrideAngularDamping=true
    bUseGravityOverride=true
    gravityScale=0.35
    alphaBoolBlend {blendOption=Linear}
    boundBone {boneName=dyn_skirt_l_A_02}
    componentAppliedLinearAccClamp=<3000 3000 3000>
    componentLinearAccScale=<0.5 0.5 0.5>
    componentLinearVelScale=<0.5 0.5 0.5>
    gravityOverride {z=-980}
    physicsBodyDefinitions [{
      boundBone {boneName=dyn_skirt_l_A_02}
      boxExtents=<20 30 30>
      constraintSetup {angularLimitsMax=<0 0 0>, angularLimitsMin=<0 0 -15>, angularTarget=<1 0 0>}
      localJointOffset=<4 0 0>
     }]
    relativeSpaceBone {boneName=pelvis}
   }
   n67 = Anim Dynamics - Bone: dyn_skirt_l_A_03(ComponentPose=@n66, Alpha=1, ExternalForce="0,0,0") {
    angularDampingOverride=1
    angularSpringConstant=20000
    bAngularSpring=true
    bOverrideAngularDamping=true
    bUseGravityOverride=true
    gravityScale=0.35
    alphaBoolBlend {blendOption=Linear}
    boundBone {boneName=dyn_skirt_l_A_03}
    componentAppliedLinearAccClamp=<3000 3000 3000>
    componentLinearAccScale=<0.5 0.5 0.5>
    componentLinearVelScale=<0.5 0.5 0.5>
    gravityOverride {z=-980}
    physicsBodyDefinitions [{
      boundBone {boneName=dyn_skirt_l_A_03}
      boxExtents=<20 30 30>
      constraintSetup {angularLimitsMax=<0 0 0>, angularLimitsMin=<0 0 -10>, angularTarget=<1 0 0>}
      localJointOffset=<4 0 0>
     }]
    relativeSpaceBone {boneName=pelvis}
   }
   n68 = "Transform (Modify) Bone - Bone: base_skirt_l_A"!(ComponentPose=@n67, Alpha:float=Get Is Crouching() -> (Is Crouching:double), Translation="0,0,0", Rotation="2.996,-4,0", Scale="1,1,1") -> (Pose) {
    rotationMode=BMM_Additive
    rotationSpace=BCS_BoneSpace
    translationMode=BMM_Additive
    translationSpace=BCS_BoneSpace
    alphaBoolBlend {blendOption=Linear}
    alphaScaleBiasClamp {bInterpResult=true}
    boneToModify {boneName=base_skirt_l_A}
    rotation {pitch=2.996, yaw=-4}
   }
   n69 = Get is Crouch Walking() -> (is Crouch Walking:double)
   n70 = "Transform (Modify) Bone - Bone: base_skirt_l_A"(ComponentPose=@n68, Alpha=@n69, Translation="0,0,0", Rotation="4,-3.5,0", Scale="1,1,1") {
    rotationMode=BMM_Additive
    rotationSpace=BCS_BoneSpace
    translationMode=BMM_Additive
    translationSpace=BCS_BoneSpace
    alphaBoolBlend {blendOption=Linear}
    alphaScaleBiasClamp {bInterpResult=true, interpSpeedDecreasing=6}
    boneToModify {boneName=base_skirt_l_A}
    rotation {pitch=4, yaw=-3.5}
   }
   n71 = "Transform (Modify) Bone - Bone: base_skirt_l_B"!(ComponentPose=@n70, Alpha:float=@n69, Translation="0,0,0", Rotation="4,5,0", Scale="1,1,1") -> (Pose) {
    rotationMode=BMM_Additive
    rotationSpace=BCS_BoneSpace
    translationMode=BMM_Additive
    translationSpace=BCS_BoneSpace
    alphaBoolBlend {blendOption=Linear}
    alphaScaleBiasClamp {bInterpResult=true, interpSpeedDecreasing=6}
    boneToModify {boneName=base_skirt_l_B}
    rotation {pitch=4, yaw=5}
   }
   n72 = "Transform (Modify) Bone - Bone: base_skirt_r_A"!(ComponentPose=@n71, Alpha:float=@n69, Translation="0,0,0", Rotation="-5.859,-4.523,0", Scale="1,1,1") -> (Pose) {
    rotationMode=BMM_Additive
    rotationSpace=BCS_BoneSpace
    translationMode=BMM_Additive
    translationSpace=BCS_BoneSpace
    alphaBoolBlend {blendOption=Linear}
    alphaScaleBiasClamp {bInterpResult=true, interpSpeedDecreasing=6}
    boneToModify {boneName=base_skirt_r_A}
    rotation {pitch=-5.859, yaw=-4.523}
   }
   n73 = "Transform (Modify) Bone - Bone: base_skirt_l_A"(ComponentPose=@n72, Alpha=0, Translation="0,0,0", Rotation="0,-15,0", Scale="1,1,1") {
    rotationMode=BMM_Additive
    rotationSpace=BCS_BoneSpace
    alphaBoolBlend {blendOption=Linear}
    alphaScaleBiasClamp {bInterpResult=true}
    boneToModify {boneName=base_skirt_l_A}
    rotation {yaw=-15}
   }
   n74 = Get Leg_L_Up() -> (Leg_L_Up:float)
   n75 = "Transform (Modify) Bone - Bone: base_skirt_l_B"(ComponentPose=@n73, Alpha=@n74, Translation="0,0,0", Rotation="5,0,0", Scale="1,1,1") {
    rotationMode=BMM_Additive
    rotationSpace=BCS_BoneSpace
    alphaBoolBlend {blendOption=Linear}
    alphaScaleBiasClamp {bInterpResult=true}
    boneToModify {boneName=base_skirt_l_B}
    rotation {pitch=5}
   }
   n76 = "Transform (Modify) Bone - Bone: base_skirt_l_C"(ComponentPose=@n75, Alpha=@n74, Translation="0,0.75,0", Rotation="0,3,0", Scale="1,1,1") {
    rotationMode=BMM_Additive
    rotationSpace=BCS_BoneSpace
    translationMode=BMM_Additive
    translationSpace=BCS_BoneSpace
    alphaBoolBlend {blendOption=Linear}
    boneToModify {boneName=base_skirt_l_C}
    rotation {yaw=3}
    translation {y=0.75}
   }
   n77 = "Transform (Modify) Bone - Bone: base_skirt_r_A"(ComponentPose=@n76, Alpha=0, Translation="0,0,0", Rotation="20,-15,0", Scale="1,1,1") {
    rotationMode=BMM_Additive
    rotationSpace=BCS_BoneSpace
    alphaBoolBlend {blendOption=Linear}
    boneToModify {boneName=base_skirt_r_A}
    rotation {pitch=20, yaw=-15}
   }
   n78 = Get Leg_R_Up() -> (Leg_R_Up:float)
   n79 = "Transform (Modify) Bone - Bone: base_skirt_r_B"(ComponentPose=@n77, Alpha=@n78, Translation="0,0,0", Rotation="5,5,0", Scale="1,1,1") {
    rotationMode=BMM_Additive
    rotationSpace=BCS_BoneSpace
    alphaBoolBlend {blendOption=Linear}
    alphaScaleBiasClamp {bInterpResult=true, interpSpeedIncreasing=5}
    boneToModify {boneName=base_skirt_r_B}
    rotation {pitch=5, yaw=5}
   }
   n80 = "Transform (Modify) Bone - Bone: base_skirt_r_C"(ComponentPose=@n79, Alpha=@n78, Translation="0,0.75,0", Rotation="0,3,0", Scale="1,1,1") {
    rotationMode=BMM_Additive
    rotationSpace=BCS_BoneSpace
    translationMode=BMM_Additive
    translationSpace=BCS_BoneSpace
    alphaBoolBlend {blendOption=Linear}
    boneToModify {boneName=base_skirt_r_C}
    rotation {yaw=3}
    translation {y=0.75}
   }
   n81 = Get Is Twisting in Place() -> (Is Twisting in Place:double)
   n82 = "Transform (Modify) Bone - Bone: base_skirt_r_A"(ComponentPose=@n80, Alpha=@n81, Translation="0,0,0", Rotation="5,-10,0", Scale="1,1,1") {
    rotationMode=BMM_Additive
    rotationSpace=BCS_BoneSpace
    alphaBoolBlend {blendOption=Linear}
    alphaScaleBiasClamp {bInterpResult=true, interpSpeedIncreasing=5}
    boneToModify {boneName=base_skirt_r_A}
    rotation {pitch=5, yaw=-10}
   }
   n83 = "Transform (Modify) Bone - Bone: base_skirt_r_B"(ComponentPose=@n82, Alpha=@n81, Translation="0,0,0", Rotation="5,5,0", Scale="1,1,1") {
    rotationMode=BMM_Additive
    rotationSpace=BCS_BoneSpace
    alphaBoolBlend {blendOption=Linear}
    alphaScaleBiasClamp {bInterpResult=true, interpSpeedIncreasing=5}
    boneToModify {boneName=base_skirt_r_B}
    rotation {pitch=5, yaw=5}
   }
   n84 = RigidBody(
    ComponentPose=@n83
    OverrideWorldGravity="0,0,0"
    SimSpaceSettings=(WorldAlpha=0.5,VelocityScaleZ=0,DampingAlpha=1,MaxLinearVelocity=500,MaxAngularVelocity=3,MaxLinearAcceleration=700,MaxAngularAcceleration=10000,ExternalLinearDragV=(X=0.2,Y=0.2,Z=0),ExternalLinearVelocity=(X=0,Y=0,Z=0),ExternalAngularVelocity=(X=0,Y=0,Z=0))
   ) {
    bOverrideWorldGravity=true
    overridePhysicsAsset=&/F_MED_CordSyrup_Physics
    simulationSpace=BaseBoneSpace
    alphaBoolBlend {blendOption=Linear}
    baseBoneRef {boneName=spine_03}
    simSpaceSettings {maxAngularVelocity=3, maxLinearAcceleration=700, maxLinearVelocity=500, velocityScaleZ=0, worldAlpha=0.5, externalLinearDragV=<0.2 0.2>}
   }
   Save cached pose 'RBS2PostRBLogic'(Pose=Component To Local(ComponentPose=@n84)) {cacheName=RBS2PostRBLogic}
  ]}
 EventGraph {nodes [
   n1 = Event BlueprintUpdateAnimation()
   > Set Delta Time X(Delta Time X:double=@n1.DeltaTimeX) -> (Output_Get:double)
   > n2 = Sequence!(execute) -> (then_0, then_1, then_2:Exec, then_3:Exec)
   > Set is DBNO(is DBNO:double="To Float (Boolean)"(InBool=Get bIsDBNO())) -> (Output_Get:double)
   > Set is Parachuting(is Parachuting:double="To Float (Boolean)"(InBool=Get bIsParachuteOpened())) -> (Output_Get:double)
   > Set Is Knee or Baseball Sliding(Is Knee or Baseball Sliding:bool=OR Boolean(A=Get bIsBaseballSliding(), B=Get bIsKneeSliding())) -> (Output_Get:bool)
   > Set Is Crouching(Is Crouching:double="To Float (Boolean)"(InBool=Get bIsCrouching())) -> (Output_Get:double)
   n3 = Get bIsCrouching()
   n4 = Get bIsMoving2D()
   > Set is Crouch Walking(is Crouch Walking:double="To Float (Boolean)"(InBool=AND Boolean(A=@n4, B=@n3))) -> (Output_Get:double)
   n5 = OR Boolean(A=Get bIsPlayingEmote(), B=Get bIsSkyDiving())
   Set is Emote Float(
    execute=@n2.then_3
    is Emote Float:double=FInterpTo_Constant!(
     Current:double=Get is Emote Float()
     Target:double="To Float (Boolean)"(InBool=@n5)
     DeltaTime:double=Get Delta Time X() -> (Delta Time X:double)
     InterpSpeed:double=4
    ) -> (ReturnValue:double)
   ) -> (Output_Get:double)
   n6 = Get bIsSprinting()
   n7 = Get bIsTacticalSprinting()
   Set Sprinting(execute=@n2.then_1, Sprinting:double="To Float (Boolean)"(InBool=OR Boolean(A=@n6, B=@n7))) -> (Output_Get:double)
   n8 = "To Float (Boolean)"(InBool=NOT Boolean(A=Get bIsCrouching()))
   n9 = Get SkelMesh()
   > Set Leg_L_Up(Leg_L_Up:float=float * float!(
     A=GetAlphaFromReferenceTransformDelta(SkeletalMeshComponent=@n9, BoneName=thigh_l, BaseName=pelvis, InputRange=(X=0,Y=-30), TwistAxis="0,0,1")
     B=Get Sprinting() -> (Sprinting:double)
     C:double=@n8
    ) -> (ReturnValue) {is=title K2Node_PromotableOperator}) -> (Output_Get:float)
   n10 = GetAlphaFromReferenceTransformDelta(SkeletalMeshComponent=@n9, BoneName=thigh_r, BaseName=pelvis, InputRange=(X=0,Y=-30), TwistAxis="0,0,1")
   > Set Leg_R_Up(Leg_R_Up:float=float * float(A=@n10, B=Get Sprinting(), C=@n8)) -> (Output_Get:float)
   n11 = Get bIsCrouching()
   n12 = Get PawnSpeed2D()
   > Set Is Twisting in Place(Is Twisting in Place:double=float * float!(
     A=@n10
     B=Remap Float(Input=@n12, OldRange=(X=0,Y=1), NewRange=(X=1,Y=0), Clamp=true) -> (Output=~)
     C:bool=Get bIsOnGround() -> (bIsOnGround:bool)
     D:bool=NOT Boolean(A=@n11)
    ) -> (ReturnValue)) -> (Output_Get:double)
   n13 = Gravity Override State Function DEPRECATED(
    execute=@n2.then_2
    CharacterPartAnimInstance=Self-Reference()
    GameTimeInSeconds=GetGameTimeInSeconds()
    GravityOverrideParams=Get Gravity Override Params()
    RBAN Params=Get RBAN Params()
   )
   > Set Gravity Override(Gravity Override=@n13)
   > n14 = Set RBAN Settings Out(RBAN Settings Out:RBANSetting Struct Structure=@n13.RBAN SettingsOut) -> (Output_Get:RBANSetting Struct Structure)
   Event BlueprintInitializeAnimation()
   > Set SkelMesh(SkelMesh=GetOwnerSkeletalMeshComponent())
   Break Sim Space Settings!(SimSpaceSettings:"Sim Space Settings Structure (by ref)"=Break RBANSetting Struct!(RBANSettingStruct:"RBANSetting Struct Structure (by ref)"=@n14) -> (
     ComponentLinearAccScale_28_#1:Vector=~
     ComponentLinearVelScale_31_#2:Vector=~
     SimSpaceSettings_34_#3:Sim Space Settings Structure
    ) {is=title K2Node_BreakStruct}.SimSpaceSettings_34_#3) -> (
    WorldAlpha:float=~
    VelocityScaleZ:float=~
    DampingAlpha:float=~
    MaxLinearVelocity:float=~
    MaxAngularVelocity:float=~
    MaxLinearAcceleration:float=~
    MaxAngularAcceleration:float=~
    ExternalLinearDragV:Vector=~
    ExternalLinearVelocity:Vector=~
    ExternalAngularVelocity:Vector=~
   ) {is=title K2Node_BreakStruct}
  ]}
]
variables [
 Delta Time X {type=number}
 Gravity Override {type=Vector}
 Gravity Override Params {
  type=GravityOverrideParamsStruct
  default {
   skydive_FloatingDown {gravityOverride=<0 -500 -980>, windAmplitude=<200 -500 150>, windOffset=<0 -600 150>}
   skydive_GoingUp {gravityOverride=<0 -300 -500>, windAmplitude=<200 -500 150>}
  }
 }
 is Crouch Walking {type=number}
 Is Crouching {type=number}
 is DBNO {type=number}
 is Emote Float {type=number}
 Is Knee or Baseball Sliding {type=boolean}
 is Parachuting {type=number}
 Is Twisting in Place {type=number}
 Leg_L_Up {type=number}
 Leg_R_Up {type=number}
 RBAN Params {
  type=RBANParamsStruct
  default {
   dBNO {simSpaceSettings {
     maxAngularAcceleration=3
     maxAngularVelocity=5
     maxLinearAcceleration=500
     maxLinearVelocity=1500
     velocityScaleZ=1
     externalLinearDragV=<1 1 0>
    }}
   dBNO_Carried {simSpaceSettings {
     maxAngularAcceleration=3
     maxAngularVelocity=5
     maxLinearAcceleration=500
     maxLinearVelocity=1500
     velocityScaleZ=1
     externalLinearDragV=<1 1 0>
    }}
   emoteOrMelee {simSpaceSettings {
     maxAngularAcceleration=3
     maxAngularVelocity=5
     maxLinearAcceleration=500
     maxLinearVelocity=1500
     velocityScaleZ=1
     externalLinearDragV=<1 1 0>
    }}
   front_End {simSpaceSettings {
     maxAngularAcceleration=3
     maxAngularVelocity=5
     maxLinearAcceleration=500
     maxLinearVelocity=1500
     velocityScaleZ=1
     externalLinearDragV=<1 1 0>
    }}
   onGround_Crouch_Moving {
    componentLinearAccScale=<1 1 1>
    componentLinearVelScale=<1 1 1>
    simSpaceSettings {
     maxAngularAcceleration=3
     maxAngularVelocity=5
     maxLinearAcceleration=500
     maxLinearVelocity=1500
     velocityScaleZ=1
     externalLinearDragV=<1 1 0>
    }
   }
   onGround_Falling {simSpaceSettings {
     maxAngularAcceleration=3
     maxAngularVelocity=5
     maxLinearAcceleration=500
     maxLinearVelocity=1500
     velocityScaleZ=1
     externalLinearDragV=<1 1 0>
    }}
   onGround_Moving {
    componentLinearAccScale=<1 1 1>
    componentLinearVelScale=<1 1 1>
    simSpaceSettings {
     maxAngularAcceleration=5
     maxAngularVelocity=10
     maxLinearAcceleration=500
     maxLinearVelocity=1500
     velocityScaleZ=1
     externalLinearDragV=<1 1 0>
    }
   }
   onGround_Standing {
    componentLinearAccScale=<1 1 1>
    componentLinearVelScale=<1 1 1>
    simSpaceSettings {
     maxAngularAcceleration=3
     maxAngularVelocity=5
     maxLinearAcceleration=500
     maxLinearVelocity=1500
     velocityScaleZ=1
     externalLinearDragV=<1 1 0>
    }
   }
   rail_Grinding {simSpaceSettings {
     maxAngularAcceleration=3
     maxAngularVelocity=5
     maxLinearAcceleration=500
     maxLinearVelocity=1500
     velocityScaleZ=1
     worldAlpha=1
     externalLinearDragV=<1 1 0>
    }}
   riding_Moving {simSpaceSettings {
     maxAngularAcceleration=3
     maxAngularVelocity=5
     maxLinearAcceleration=500
     maxLinearVelocity=1500
     velocityScaleZ=1
     externalLinearDragV=<1 1 0>
    }}
   riding_Moving_Targeting {simSpaceSettings {
     maxAngularAcceleration=3
     maxAngularVelocity=5
     maxLinearAcceleration=500
     maxLinearVelocity=1500
     velocityScaleZ=1
     externalLinearDragV=<1 1 0>
    }}
   riding_Standing {simSpaceSettings {
     maxAngularAcceleration=3
     maxAngularVelocity=5
     maxLinearAcceleration=500
     maxLinearVelocity=1500
     velocityScaleZ=1
     externalLinearDragV=<1 1 0>
    }}
   skydive_DivingDown {
    componentLinearAccScale=<0 0 0>
    componentLinearVelScale=<0 0 0>
    simSpaceSettings {
     maxAngularAcceleration=3
     maxAngularVelocity=5
     maxLinearAcceleration=500
     maxLinearVelocity=1500
     velocityScaleZ=1
     externalLinearDragV=<1 1 0>
    }
   }
   skydive_FloatingDown {
    componentLinearAccScale=<0 0 0>
    componentLinearVelScale=<0 0 0>
    simSpaceSettings {
     maxAngularAcceleration=3
     maxAngularVelocity=5
     maxLinearAcceleration=500
     maxLinearVelocity=1500
     velocityScaleZ=1
     externalLinearDragV=<1 1 0>
    }
   }
   skydive_GoingUp {simSpaceSettings {
     maxAngularAcceleration=3
     maxAngularVelocity=5
     maxLinearAcceleration=500
     maxLinearVelocity=1500
     velocityScaleZ=1
     externalLinearDragV=<1 1 0>
    }}
   skydive_Parachuting {
    componentLinearAccScale=<1 1 1>
    componentLinearVelScale=<1 1 1>
    simSpaceSettings {
     maxAngularAcceleration=3
     maxAngularVelocity=5
     maxLinearAcceleration=500
     maxLinearVelocity=1500
     velocityScaleZ=1
     externalLinearDragV=<1 1 0>
    }
   }
   skydive_Surfing {simSpaceSettings {
     maxAngularAcceleration=3
     maxAngularVelocity=5
     maxLinearAcceleration=500
     maxLinearVelocity=1500
     velocityScaleZ=1
     externalLinearDragV=<1 1 0>
    }}
   swimming {simSpaceSettings {
     maxAngularAcceleration=3
     maxAngularVelocity=5
     maxLinearAcceleration=500
     maxLinearVelocity=1500
     velocityScaleZ=1
     externalLinearDragV=<1 1 0>
    }}
   tactical_Sprint {
    componentLinearAccScale=<1 1 1>
    componentLinearVelScale=<1 1 1>
    simSpaceSettings {
     maxAngularAcceleration=3
     maxAngularVelocity=5
     maxLinearAcceleration=500
     maxLinearVelocity=1500
     velocityScaleZ=1
     externalLinearDragV=<1 1 0>
    }
   }
   ziplining {simSpaceSettings {
     maxAngularAcceleration=0.5
     maxAngularVelocity=1
     maxLinearAcceleration=50
     maxLinearVelocity=200
     velocityScaleZ=1
     externalLinearDragV=<0.5 0.5 0>
    }}
  }
 }
 RBAN Settings Out {
  type=RBANSettingStruct
  default {
   componentLinearAccScale=<0 0 0>
   componentLinearVelScale=<0 0 0>
   simSpaceSettings {
    dampingAlpha=1
    maxAngularAcceleration=1000
    maxAngularVelocity=3
    maxLinearAcceleration=300
    maxLinearVelocity=700
    velocityScaleZ=0
    worldAlpha=1
    externalAngularVelocity=<0 0 0>
    externalLinearDragV=<0 0.5 0>
    externalLinearVelocity=<0 0 0>
   }
  }
 }
 SkelMesh {type=SkeletalMeshComponent}
 Sprinting {type=number}
]
