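# Physics asset for the CordSyrup female medium body cosmetic, named as the collider set for its cape.
# Capsules cover torso, clavicles, head, feet, a left upper-arm twist bone, and dynamic skirt, skirt armour, jacket and arm armour bones.
# Thigh, calf, neck and main arm bodies have no shapes.
# Constraints follow the bone chains with free swing and twist, so their angular limits do not restrict motion.
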PhysicsAsset /BRCosmetics/Characters/Player/Female/Medium/Bodies/F_MED_CordSyrup/Meshes/F_MED_CordSyrup_CapeColliders
bodies {
 calf_l []
 calf_r []
 clavicle_l [{shapeType=Capsule, length=6.5623, radius=7.5194, center=<3.9566 -1.0862 -2.1127>, rotation=<rot -77.5566 177.3915 7.7314>}]
 clavicle_r [{shapeType=Capsule, length=6.5623, radius=7.5194, center=<-3.9566 1.0862 2.1127>, rotation=<rot -77.5566 177.3915 -172.2686>}]
 dyn_armArmour_01 [{
   shapeType=Capsule
   length=4.2586
   radius=3.736
   shapeName=dyn_armArmour_01_capsule
   center=<2.7617 -0.1571 -1.6812>
   rotation=<rot 0 0.5155 -86.3264>
  }]
 dyn_armArmour_02 [{
   shapeType=Capsule
   length=6.2696
   radius=4.0856
   shapeName=dyn_armArmour_02_capsule
   center=<3.03 -0.0549 -1.7356>
   rotation=<rot 16.1882 -148.9171 -90>
  }]
 dyn_armArmour_03 [{
   shapeType=Capsule
   length=5.6664
   radius=3.5577
   shapeName=dyn_armArmour_03_capsule
   center=<4.4535 -0.3055 -0.8679>
   rotation=<rot 4.8529 -175.9975 -90>
  }]
 dyn_jacket_fn_l_01 [{shapeType=Capsule, length=3, radius=4, shapeName=dyn_jacket_fn_l_01_capsule, center=<4 0 0>, rotation=<rot 90 0 0>}]
 dyn_jacket_fn_l_02 [{
   shapeType=Capsule
   length=3.3083
   radius=2.6997
   shapeName=dyn_jacket_fn_l_02_capsule
   center=<1.3766 -0.6443 0.0691>
   rotation=<rot 0 -160.5173 -76.0209>
  }]
 dyn_jacket_fn_r_01 [{shapeType=Capsule, length=3, radius=4, shapeName=dyn_jacket_fn_r_01_capsule, center=<4 0 0>, rotation=<rot 90 0 0>}]
 dyn_jacket_s_l_01 [{
   shapeType=Capsule
   length=5.5647
   radius=3.8354
   shapeName=dyn_jacket_s_l_01_capsule
   center=<3.2742 1.2068 0.1976>
   rotation=<rot 0 -106.1666 -86.7091>
  }]
 dyn_jacket_s_r_01 [{
   shapeType=Capsule
   length=5.6258
   radius=3.9561
   shapeName=dyn_jacket_s_r_01_capsule
   center=<3.2721 -1.1993 0.1056>
   rotation=<rot 0 -70.2046 -84.3545>
  }]
 dyn_skirtArmour_01 [{shapeType=Capsule, length=9, radius=5.5, shapeName=dyn_skirtArmour_01_capsule, center=<3 0 0>}]
 dyn_skirtArmour_02 [{shapeType=Capsule, length=9, radius=5.5, shapeName=dyn_skirtArmour_02_capsule, center=<3 0 0>}]
 dyn_skirtArmour_03 [{shapeType=Capsule, length=9, radius=5.5, shapeName=dyn_skirtArmour_03_capsule, center=<3 0 0>}]
 dyn_skirt_l_A_01 [{shapeType=Capsule, length=15, radius=6.5, shapeName=dyn_skirt_l_A_01_capsule, center=<4.5 0 0>}]
 dyn_skirt_l_A_02 [{shapeType=Capsule, length=15, radius=6.5, shapeName=dyn_skirt_l_A_02_capsule, center=<4.5 0 0>}]
 dyn_skirt_l_A_03 [{shapeType=Capsule, length=15, radius=6.5, shapeName=dyn_skirt_l_A_03_capsule, center=<4.5 0 0>}]
 dyn_skirt_l_B_01 [{shapeType=Capsule, length=16.5, radius=6.5, shapeName=dyn_skirt_l_B_01_capsule, center=<4.5 0 0>, rotation=<rot 0 0 90>}]
 dyn_skirt_l_B_02 [{shapeType=Capsule, length=16.5, radius=6.5, shapeName=dyn_skirt_l_B_02_capsule, center=<4.5 0 0>, rotation=<rot 0 0 90>}]
 dyn_skirt_l_B_03 [{shapeType=Capsule, length=16.5, radius=6.5, shapeName=dyn_skirt_l_B_03_capsule, center=<4.5 0 0>, rotation=<rot 0 0 90>}]
 dyn_skirt_l_C_01 [{shapeType=Capsule, length=18, radius=6.5, shapeName=dyn_skirt_l_C_01_capsule, center=<4.5 0 0.7465>}]
 dyn_skirt_l_C_02 [{shapeType=Capsule, length=18, radius=6.5, shapeName=dyn_skirt_l_C_02_capsule, center=<4.5 0 0.7465>}]
 dyn_skirt_l_C_03 [{shapeType=Capsule, length=18, radius=6.5, shapeName=dyn_skirt_l_C_03_capsule, center=<4.5 0 0.7465>}]
 dyn_skirt_r_A_01 [{shapeType=Capsule, length=15, radius=6.5, shapeName=dyn_skirt_r_A_01_capsule, center=<4.5 0 0>}]
 dyn_skirt_r_A_02 [{shapeType=Capsule, length=15, radius=6.5, shapeName=dyn_skirt_r_A_02_capsule, center=<4.5 0 0>}]
 dyn_skirt_r_A_03 [{shapeType=Capsule, length=15, radius=6.5, shapeName=dyn_skirt_r_A_03_capsule, center=<4.5 0 0>}]
 dyn_skirt_r_B_01 [{shapeType=Capsule, length=16.5, radius=6.5, shapeName=dyn_skirt_r_B_01_capsule, center=<4.5 0 0>, rotation=<rot 0 0 90>}]
 dyn_skirt_r_B_02 [{shapeType=Capsule, length=16.5, radius=6.5, shapeName=dyn_skirt_r_B_02_capsule, center=<4.5 0 0>, rotation=<rot 0 0 90>}]
 dyn_skirt_r_B_03 [{shapeType=Capsule, length=16.5, radius=6.5, shapeName=dyn_skirt_r_B_03_capsule, center=<4.5 0 0>, rotation=<rot 0 0 90>}]
 dyn_skirt_r_C_01 [{shapeType=Capsule, length=18, radius=6.5, shapeName=dyn_skirt_r_C_01_capsule, center=<4.5 0 -0.5129>}]
 dyn_skirt_r_C_02 [{shapeType=Capsule, length=18, radius=6.5, shapeName=dyn_skirt_r_C_02_capsule, center=<4.5 0 -0.5129>}]
 dyn_skirt_r_C_03 [{shapeType=Capsule, length=18, radius=6.5, shapeName=dyn_skirt_r_C_03_capsule, center=<4.5 0 -0.5129>}]
 foot_l [{shapeType=Capsule, length=11.0765, radius=8.0344, center=<-4.1811 6.0027 -0.338>, rotation=<rot 0.0388 1.4701 -89.3982>}]
 foot_r [{shapeType=Capsule, length=11.0765, radius=8.0344, center=<4.1811 -6.0027 0.338>, rotation=<rot 0.0388 1.4701 90.6018>}]
 head [{shapeType=Capsule, length=0, radius=18, center=<9.0697 0.0746 0>, rotation=<rot 0 -90 0>}]
 lowerarm_l []
 lowerarm_r []
 neck_02 []
 pelvis [{shapeType=Capsule, length=9.6042, radius=11.1441, center=<-2.4471 -2.5289 -0.1981>}]
 spine_01 [{shapeType=Capsule, length=4.3888, radius=14.3638, center=<-8.9021 -0.8973 0>}]
 spine_02 [
  {shapeType=Capsule, length=2.5219, radius=9.7317, center=<-2.6885 -1.2641 0>}
  {shapeType=Capsule, length=20.6875, radius=9.3125, center=<-1.8317 1.9362 0>, rotation=<rot 90 0 0>}
 ]
 spine_03 [
  {shapeType=Capsule, length=5.1863, radius=9.4854, center=<1.7466 -0.1222 0>}
  {shapeType=Capsule, length=20.7188, radius=9.2812, center=<-1.2163 0.549 0>, rotation=<rot 90 180 -175.0732>}
 ]
 spine_04 [
  {shapeType=Capsule, length=9.0076, radius=10.3916, center=<5.6234 -0.6779 0>}
  {shapeType=Capsule, length=10.0312, radius=9.2812, center=<-1.2326 1.308 0>, rotation=<rot 90 0 0>}
 ]
 spine_05 [{shapeType=Capsule, length=9.678, radius=12.3571, center=<0.6969 1.91 0>}]
 thigh_l []
 thigh_r []
 upperarm_l []
 upperarm_r []
 upperarm_twist_01_extra_l [{
   shapeType=Capsule
   length=10.5409
   radius=7.6925
   shapeName=upperarm_twist_01_extra_l_capsule
   center=<-1.0985 0.3455 3.5421>
   rotation=<rot 0 -178.0541 -83.4781>
  }]
}
constraints [
 {
  bone1Name=spine_01
  bone2Name=pelvis
  swing1LimitDegrees=45
  swing1Motion=Free
  swing2LimitDegrees=45
  swing2Motion=Free
  twistLimitDegrees=45
  twistMotion=Free
 }
 {
  bone1Name=spine_02
  bone2Name=spine_01
  swing1LimitDegrees=45
  swing1Motion=Free
  swing2LimitDegrees=45
  swing2Motion=Free
  twistLimitDegrees=45
  twistMotion=Free
 }
 {
  bone1Name=spine_03
  bone2Name=spine_02
  swing1LimitDegrees=45
  swing1Motion=Free
  swing2LimitDegrees=45
  swing2Motion=Free
  twistLimitDegrees=45
  twistMotion=Free
 }
 {
  bone1Name=spine_04
  bone2Name=spine_03
  swing1LimitDegrees=45
  swing1Motion=Free
  swing2LimitDegrees=45
  swing2Motion=Free
  twistLimitDegrees=45
  twistMotion=Free
 }
 {
  bone1Name=spine_05
  bone2Name=spine_04
  swing1LimitDegrees=45
  swing1Motion=Free
  swing2LimitDegrees=45
  swing2Motion=Free
  twistLimitDegrees=45
  twistMotion=Free
 }
 {
  bone1Name=clavicle_l
  bone2Name=spine_05
  swing1LimitDegrees=45
  swing1Motion=Free
  swing2LimitDegrees=45
  swing2Motion=Free
  twistLimitDegrees=45
  twistMotion=Free
 }
 {
  bone1Name=upperarm_l
  bone2Name=clavicle_l
  swing1LimitDegrees=45
  swing1Motion=Free
  swing2LimitDegrees=45
  swing2Motion=Free
  twistLimitDegrees=45
  twistMotion=Free
 }
 {
  bone1Name=lowerarm_l
  bone2Name=upperarm_l
  swing1LimitDegrees=45
  swing1Motion=Free
  swing2LimitDegrees=45
  swing2Motion=Free
  twistLimitDegrees=45
  twistMotion=Free
 }
 {
  bone1Name=neck_02
  bone2Name=spine_05
  swing1LimitDegrees=45
  swing1Motion=Free
  swing2LimitDegrees=45
  swing2Motion=Free
  twistLimitDegrees=45
  twistMotion=Free
 }
 {
  bone1Name=head
  bone2Name=neck_02
  swing1LimitDegrees=45
  swing1Motion=Free
  swing2LimitDegrees=45
  swing2Motion=Free
  twistLimitDegrees=45
  twistMotion=Free
 }
 {
  bone1Name=thigh_l
  bone2Name=pelvis
  swing1LimitDegrees=45
  swing1Motion=Free
  swing2LimitDegrees=45
  swing2Motion=Free
  twistLimitDegrees=45
  twistMotion=Free
 }
 {
  bone1Name=foot_l
  bone2Name=calf_l
  swing1LimitDegrees=45
  swing1Motion=Free
  swing2LimitDegrees=45
  swing2Motion=Free
  twistLimitDegrees=45
  twistMotion=Free
 }
 {
  bone1Name=lowerarm_r
  bone2Name=upperarm_r
  swing1LimitDegrees=45
  swing1Motion=Free
  swing2LimitDegrees=45
  swing2Motion=Free
  twistLimitDegrees=45
  twistMotion=Free
 }
 {
  bone1Name=upperarm_r
  bone2Name=clavicle_r
  swing1LimitDegrees=45
  swing1Motion=Free
  swing2LimitDegrees=45
  swing2Motion=Free
  twistLimitDegrees=45
  twistMotion=Free
 }
 {
  bone1Name=clavicle_r
  bone2Name=spine_05
  swing1LimitDegrees=45
  swing1Motion=Free
  swing2LimitDegrees=45
  swing2Motion=Free
  twistLimitDegrees=45
  twistMotion=Free
 }
 {
  bone1Name=thigh_r
  bone2Name=pelvis
  swing1LimitDegrees=45
  swing1Motion=Free
  swing2LimitDegrees=45
  swing2Motion=Free
  twistLimitDegrees=45
  twistMotion=Free
 }
 {
  bone1Name=foot_r
  bone2Name=calf_r
  swing1LimitDegrees=45
  swing1Motion=Free
  swing2LimitDegrees=45
  swing2Motion=Free
  twistLimitDegrees=45
  twistMotion=Free
 }
 {
  bone1Name=calf_r
  bone2Name=thigh_r
  swing1LimitDegrees=45
  swing1Motion=Free
  swing2LimitDegrees=45
  swing2Motion=Free
  twistLimitDegrees=45
  twistMotion=Free
 }
 {
  bone1Name=calf_l
  bone2Name=thigh_l
  swing1LimitDegrees=45
  swing1Motion=Free
  swing2LimitDegrees=45
  swing2Motion=Free
  twistLimitDegrees=45
  twistMotion=Free
 }
 {
  bone1Name=dyn_skirt_l_A_01
  bone2Name=pelvis
  swing1LimitDegrees=45
  swing1Motion=Free
  swing2LimitDegrees=45
  swing2Motion=Free
  twistLimitDegrees=45
  twistMotion=Free
 }
 {
  bone1Name=dyn_skirt_l_A_02
  bone2Name=dyn_skirt_l_A_01
  swing1LimitDegrees=45
  swing1Motion=Free
  swing2LimitDegrees=45
  swing2Motion=Free
  twistLimitDegrees=45
  twistMotion=Free
 }
 {
  bone1Name=dyn_skirt_l_A_03
  bone2Name=dyn_skirt_l_A_02
  swing1LimitDegrees=45
  swing1Motion=Free
  swing2LimitDegrees=45
  swing2Motion=Free
  twistLimitDegrees=45
  twistMotion=Free
 }
 {
  bone1Name=dyn_skirt_l_B_01
  bone2Name=pelvis
  swing1LimitDegrees=45
  swing1Motion=Free
  swing2LimitDegrees=45
  swing2Motion=Free
  twistLimitDegrees=45
  twistMotion=Free
 }
 {
  bone1Name=dyn_skirtArmour_03
  bone2Name=dyn_skirtArmour_02
  swing1LimitDegrees=45
  swing1Motion=Free
  swing2LimitDegrees=45
  swing2Motion=Free
  twistLimitDegrees=45
  twistMotion=Free
 }
 {
  bone1Name=dyn_skirtArmour_01
  bone2Name=dyn_skirt_l_B_01
  swing1LimitDegrees=45
  swing1Motion=Free
  swing2LimitDegrees=45
  swing2Motion=Free
  twistLimitDegrees=45
  twistMotion=Free
 }
 {
  bone1Name=dyn_skirtArmour_02
  bone2Name=dyn_skirtArmour_01
  swing1LimitDegrees=45
  swing1Motion=Free
  swing2LimitDegrees=45
  swing2Motion=Free
  twistLimitDegrees=45
  twistMotion=Free
 }
 {
  bone1Name=dyn_skirt_l_B_02
  bone2Name=dyn_skirt_l_B_01
  swing1LimitDegrees=45
  swing1Motion=Free
  swing2LimitDegrees=45
  swing2Motion=Free
  twistLimitDegrees=45
  twistMotion=Free
 }
 {
  bone1Name=dyn_skirt_l_B_03
  bone2Name=dyn_skirt_l_B_02
  swing1LimitDegrees=45
  swing1Motion=Free
  swing2LimitDegrees=45
  swing2Motion=Free
  twistLimitDegrees=45
  twistMotion=Free
 }
 {
  bone1Name=dyn_skirt_l_C_01
  bone2Name=pelvis
  swing1LimitDegrees=45
  swing1Motion=Free
  swing2LimitDegrees=45
  swing2Motion=Free
  twistLimitDegrees=45
  twistMotion=Free
 }
 {
  bone1Name=dyn_skirt_l_C_02
  bone2Name=dyn_skirt_l_C_01
  swing1LimitDegrees=45
  swing1Motion=Free
  swing2LimitDegrees=45
  swing2Motion=Free
  twistLimitDegrees=45
  twistMotion=Free
 }
 {
  bone1Name=dyn_skirt_l_C_03
  bone2Name=dyn_skirt_l_C_02
  swing1LimitDegrees=45
  swing1Motion=Free
  swing2LimitDegrees=45
  swing2Motion=Free
  twistLimitDegrees=45
  twistMotion=Free
 }
 {
  bone1Name=dyn_skirt_r_A_01
  bone2Name=pelvis
  swing1LimitDegrees=45
  swing1Motion=Free
  swing2LimitDegrees=45
  swing2Motion=Free
  twistLimitDegrees=45
  twistMotion=Free
 }
 {
  bone1Name=dyn_skirt_r_A_02
  bone2Name=dyn_skirt_r_A_01
  swing1LimitDegrees=45
  swing1Motion=Free
  swing2LimitDegrees=45
  swing2Motion=Free
  twistLimitDegrees=45
  twistMotion=Free
 }
 {
  bone1Name=dyn_skirt_r_A_03
  bone2Name=dyn_skirt_r_A_02
  swing1LimitDegrees=45
  swing1Motion=Free
  swing2LimitDegrees=45
  swing2Motion=Free
  twistLimitDegrees=45
  twistMotion=Free
 }
 {
  bone1Name=dyn_skirt_r_B_01
  bone2Name=pelvis
  swing1LimitDegrees=45
  swing1Motion=Free
  swing2LimitDegrees=45
  swing2Motion=Free
  twistLimitDegrees=45
  twistMotion=Free
 }
 {
  bone1Name=dyn_skirt_r_B_02
  bone2Name=dyn_skirt_r_B_01
  swing1LimitDegrees=45
  swing1Motion=Free
  swing2LimitDegrees=45
  swing2Motion=Free
  twistLimitDegrees=45
  twistMotion=Free
 }
 {
  bone1Name=dyn_skirt_r_B_03
  bone2Name=dyn_skirt_r_B_02
  swing1LimitDegrees=45
  swing1Motion=Free
  swing2LimitDegrees=45
  swing2Motion=Free
  twistLimitDegrees=45
  twistMotion=Free
 }
 {
  bone1Name=dyn_skirt_r_C_01
  bone2Name=pelvis
  swing1LimitDegrees=45
  swing1Motion=Free
  swing2LimitDegrees=45
  swing2Motion=Free
  twistLimitDegrees=45
  twistMotion=Free
 }
 {
  bone1Name=dyn_skirt_r_C_02
  bone2Name=dyn_skirt_r_C_01
  swing1LimitDegrees=45
  swing1Motion=Free
  swing2LimitDegrees=45
  swing2Motion=Free
  twistLimitDegrees=45
  twistMotion=Free
 }
 {
  bone1Name=dyn_skirt_r_C_03
  bone2Name=dyn_skirt_r_C_02
  swing1LimitDegrees=45
  swing1Motion=Free
  swing2LimitDegrees=45
  swing2Motion=Free
  twistLimitDegrees=45
  twistMotion=Free
 }
 {
  bone1Name=dyn_jacket_fn_l_01
  bone2Name=spine_04
  swing1LimitDegrees=45
  swing1Motion=Free
  swing2LimitDegrees=45
  swing2Motion=Free
  twistLimitDegrees=45
  twistMotion=Free
 }
 {
  bone1Name=dyn_jacket_fn_l_02
  bone2Name=dyn_jacket_fn_l_01
  swing1LimitDegrees=45
  swing1Motion=Free
  swing2LimitDegrees=45
  swing2Motion=Free
  twistLimitDegrees=45
  twistMotion=Free
 }
 {
  bone1Name=dyn_jacket_fn_r_01
  bone2Name=spine_04
  swing1LimitDegrees=45
  swing1Motion=Free
  swing2LimitDegrees=45
  swing2Motion=Free
  twistLimitDegrees=45
  twistMotion=Free
 }
 {
  bone1Name=dyn_jacket_s_l_01
  bone2Name=spine_04
  swing1LimitDegrees=45
  swing1Motion=Free
  swing2LimitDegrees=45
  swing2Motion=Free
  twistLimitDegrees=45
  twistMotion=Free
 }
 {
  bone1Name=dyn_jacket_s_r_01
  bone2Name=spine_04
  swing1LimitDegrees=45
  swing1Motion=Free
  swing2LimitDegrees=45
  swing2Motion=Free
  twistLimitDegrees=45
  twistMotion=Free
 }
 {
  bone1Name=upperarm_twist_01_extra_l
  bone2Name=upperarm_l
  swing1LimitDegrees=45
  swing1Motion=Free
  swing2LimitDegrees=45
  swing2Motion=Free
  twistLimitDegrees=45
  twistMotion=Free
 }
 {
  bone1Name=dyn_armArmour_01
  bone2Name=upperarm_twist_01_extra_l
  swing1LimitDegrees=45
  swing1Motion=Free
  swing2LimitDegrees=45
  swing2Motion=Free
  twistLimitDegrees=45
  twistMotion=Free
 }
 {
  bone1Name=dyn_armArmour_02
  bone2Name=dyn_armArmour_01
  swing1LimitDegrees=45
  swing1Motion=Free
  swing2LimitDegrees=45
  swing2Motion=Free
  twistLimitDegrees=45
  twistMotion=Free
 }
 {
  bone1Name=dyn_armArmour_03
  bone2Name=dyn_armArmour_02
  swing1LimitDegrees=45
  swing1Motion=Free
  swing2LimitDegrees=45
  swing2Motion=Free
  twistLimitDegrees=45
  twistMotion=Free
 }
]
