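# Physics asset for the CordSyrup female medium character's FaceAcc mesh part, providing capsule bodies and constraints for its dynamic hair.
# Three two-segment back-hair chains and two three-segment hair-piece chains hang from base bones under the head, with limited swing and zero-degree twist limits.
# Larger capsules on the head, neck, upper spine, clavicles and upper arms form the body shapes around the hair, joined by free-rotation constraints.
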PhysicsAsset /BRCosmetics/Characters/Player/Female/Medium/Bodies/F_MED_CordSyrup/Meshes/Parts/F_MED_CordSyrup_FaceAcc_Physics
bodies {
 base_backHair_c [{shapeType=Capsule, length=0, radius=0.5, shapeName=base_backHair_c_capsule, rotation=<rot 90 0 0>}]
 base_backHair_l [{shapeType=Capsule, length=0, radius=0.5, shapeName=base_backHair_l_capsule, rotation=<rot 90 0 0>}]
 base_backHair_r [{shapeType=Capsule, length=0, radius=0.5, shapeName=base_backHair_r_capsule, rotation=<rot 90 0 0>}]
 base_hair [{shapeType=Capsule, length=0, radius=0.5, shapeName=base_hair_capsule, rotation=<rot 90 0 0>}]
 base_hairPiece_l [{shapeType=Capsule, length=0, radius=0.5, shapeName=base_hairPiece_l_capsule, rotation=<rot 90 0 0>}]
 base_hairPiece_r [{shapeType=Capsule, length=0, radius=0.5, shapeName=base_hairPiece_r_capsule, rotation=<rot 90 0 0>}]
 base_upperarm_twist_01_extra_l [{shapeType=Capsule, length=4, radius=7, shapeName=base_upperarm_twist_01_extra_l_capsule, center=<1.3865 0 3.6259>, rotation=<rot 0 0 90>}]
 clavicle_l [{shapeType=Capsule, length=3, radius=4.5, shapeName=clavicle_r_capsule, center=<0.0293 0 3.0561>, rotation=<rot 34.9421 -0.0002 179.9997>}]
 clavicle_r [
  {shapeType=Capsule, length=3, radius=4.5, shapeName=clavicle_r_capsule, center=<-0.0293 0 -3.0561>, rotation=<rot 34.9421 -0.0002 -0.0003>}
  {shapeType=Capsule, length=3, radius=4, shapeName=clavicle_r_capsule, center=<-5.5352 4.9468 -0.2027>, rotation=<rot 90 0 0>}
 ]
 dyn_backHair_c_01 [{shapeType=Capsule, length=5.5, radius=1.5, shapeName=dyn_backHair_c_01_capsule, center=<2.4146 0 -0.1721>, rotation=<rot 0 0 90>}]
 dyn_backHair_c_02 [{shapeType=Capsule, length=7.5, radius=1.5, shapeName=dyn_backHair_c_02_capsule, center=<2.248 0 0.3662>, rotation=<rot 0 0 90>}]
 dyn_backHair_l_01 [{shapeType=Capsule, length=4.5, radius=1.5, shapeName=dyn_backHair_l_01_capsule, center=<2 0 0>, rotation=<rot 0 0 90>}]
 dyn_backHair_l_02 [
  {
   shapeType=Capsule
   length=5
   radius=1.15
   shapeName=dyn_backHair_l_02_capsule
   center=<1.4845 0.4559 0>
   rotation=<rot 1.6119 6.4268 105.1151>
  }
  {
   shapeType=Capsule
   length=0.5
   radius=0.5
   shapeName=dyn_backHair_l_02_capsule
   center=<2.8552 1.1327 -0.1083>
   rotation=<rot -73.3035 -90.8364 122.1005>
  }
 ]
 dyn_backHair_r_01 [{
   shapeType=Capsule
   length=5
   radius=1.5
   shapeName=dyn_backHair_r_01_capsule
   center=<3.6123 -0.3617 -0.2308>
   rotation=<rot -0.0001 -0.0001 79.6111>
  }]
 dyn_backHair_r_02 [{
   shapeType=Capsule
   length=5
   radius=0.75
   shapeName=dyn_backHair_r_02_capsule
   center=<2.0743 -1.1613 -0.8002>
   rotation=<rot -2.3607 -7.8487 73.3642>
  }]
 dyn_hairPiece_l_01 [{
   shapeType=Capsule
   length=3.5
   radius=1.2
   shapeName=dyn_hairPiece_l_01_capsule
   center=<2.8145 0.1645 0.2247>
   rotation=<rot 88.4915 -179.9991 -173.3348>
  }]
 dyn_hairPiece_l_02 [{
   shapeType=Capsule
   length=2
   radius=1
   shapeName=dyn_hairPiece_l_02_capsule
   center=<2.1147 -0.3622 0.115>
   rotation=<rot 82.8954 -179.9997 -179.9998>
  }]
 dyn_hairPiece_l_03 [{
   shapeType=Capsule
   length=2
   radius=1
   shapeName=dyn_hairPiece_l_03_capsule
   center=<1.9234 1.1067 -0.4987>
   rotation=<rot 76.9626 24.6756 0.0001>
  }]
 dyn_hairPiece_r_01 [{
   shapeType=Capsule
   length=3
   radius=1.1
   shapeName=dyn_hairPiece_r_01_capsule
   center=<2.8602 -0.1691 0>
   rotation=<rot 90 179.6419 177.6822>
  }]
 dyn_hairPiece_r_02 [{
   shapeType=Capsule
   length=2
   radius=1
   shapeName=dyn_hairPiece_r_02_capsule
   center=<1.9972 0 0.2801>
   rotation=<rot 85.9646 179.9998 179.9999>
  }]
 dyn_hairPiece_r_03 [{
   shapeType=Capsule
   length=2.8
   radius=1.2
   shapeName=dyn_hairPiece_l_03_capsule
   center=<2.6271 0.0299 -1.1702>
   rotation=<rot 76.052 19.7619 -169.5936>
  }]
 head [
  {shapeType=Capsule, length=2, radius=7.5, shapeName=head_capsule, center=<5.7947 0.1347 0>, rotation=<rot 90 90 113.8055>}
  {shapeType=Capsule, length=1.5, radius=6.2, shapeName=head_capsule, center=<-0.1924 4.3106 0>, rotation=<rot 90 -90 -47.7652>}
 ]
 neck_01 [{shapeType=Capsule, length=3, radius=4.5, shapeName=neck_01_capsule, center=<0.3257 1.2795 0>, rotation=<rot 0 0 90>}]
 neck_02 [{shapeType=Capsule, length=0, radius=4.5, shapeName=neck_02_capsule, center=<0 0.3588 0>, rotation=<rot 90 0 0>}]
 spine_05 [{shapeType=Capsule, length=3, radius=10.5, shapeName=spine_05_capsule, center=<2.605 0.555 0>}]
 upperarm_l [{shapeType=Capsule, length=2, radius=1, shapeName=upperarm_l_capsule, rotation=<rot 90 0 0>}]
 upperarm_r [{
   shapeType=Capsule
   length=12
   radius=5.5
   shapeName=clavicle_l_capsule
   center=<-9.971 1.7236 0.8122>
   rotation=<rot 87.4211 -179.9992 -179.9993>
  }]
}
constraints [
 {
  bone1Name=upperarm_l
  bone2Name=clavicle_l
  swing1LimitDegrees=45
  swing1Motion=Free
  swing2LimitDegrees=45
  swing2Motion=Free
  twistLimitDegrees=45
  twistMotion=Free
 }
 {
  bone1Name=base_upperarm_twist_01_extra_l
  bone2Name=upperarm_l
  swing1LimitDegrees=45
  swing1Motion=Free
  swing2LimitDegrees=45
  swing2Motion=Free
  twistLimitDegrees=45
  twistMotion=Free
 }
 {
  bone1Name=clavicle_r
  bone2Name=spine_05
  swing1LimitDegrees=45
  swing1Motion=Free
  swing2LimitDegrees=45
  swing2Motion=Free
  twistLimitDegrees=45
  twistMotion=Free
 }
 {
  bone1Name=upperarm_r
  bone2Name=clavicle_r
  swing1LimitDegrees=45
  swing1Motion=Free
  swing2LimitDegrees=45
  swing2Motion=Free
  twistLimitDegrees=45
  twistMotion=Free
 }
 {
  bone1Name=neck_01
  bone2Name=spine_05
  swing1LimitDegrees=45
  swing1Motion=Free
  swing2LimitDegrees=45
  swing2Motion=Free
  twistLimitDegrees=45
  twistMotion=Free
 }
 {
  bone1Name=neck_02
  bone2Name=neck_01
  swing1LimitDegrees=45
  swing1Motion=Free
  swing2LimitDegrees=45
  swing2Motion=Free
  twistLimitDegrees=45
  twistMotion=Free
 }
 {
  bone1Name=head
  bone2Name=neck_02
  swing1LimitDegrees=45
  swing1Motion=Free
  swing2LimitDegrees=45
  swing2Motion=Free
  twistLimitDegrees=45
  twistMotion=Free
 }
 {
  bone1Name=base_hair
  bone2Name=head
  swing1LimitDegrees=45
  swing1Motion=Free
  swing2LimitDegrees=45
  swing2Motion=Free
  twistLimitDegrees=45
  twistMotion=Free
 }
 {
  bone1Name=base_backHair_c
  bone2Name=base_hair
  swing1LimitDegrees=45
  swing1Motion=Free
  swing2LimitDegrees=45
  swing2Motion=Free
  twistLimitDegrees=45
  twistMotion=Free
 }
 {bone1Name=dyn_backHair_c_01, bone2Name=base_backHair_c, swing1LimitDegrees=25, swing2LimitDegrees=25, twistLimitDegrees=0}
 {bone1Name=dyn_backHair_c_02, bone2Name=dyn_backHair_c_01, swing1LimitDegrees=25, swing2LimitDegrees=25, twistLimitDegrees=0}
 {
  bone1Name=base_backHair_l
  bone2Name=base_hair
  swing1LimitDegrees=45
  swing1Motion=Free
  swing2LimitDegrees=45
  swing2Motion=Free
  twistLimitDegrees=45
  twistMotion=Free
 }
 {bone1Name=dyn_backHair_l_01, bone2Name=base_backHair_l, swing1LimitDegrees=25, swing2LimitDegrees=25, twistLimitDegrees=0}
 {bone1Name=dyn_backHair_l_02, bone2Name=dyn_backHair_l_01, swing1LimitDegrees=25, swing2LimitDegrees=25, twistLimitDegrees=0}
 {
  bone1Name=base_backHair_r
  bone2Name=base_hair
  swing1LimitDegrees=45
  swing1Motion=Free
  swing2LimitDegrees=45
  swing2Motion=Free
  twistLimitDegrees=45
  twistMotion=Free
 }
 {bone1Name=dyn_backHair_r_01, bone2Name=base_backHair_r, swing1LimitDegrees=25, swing2LimitDegrees=25, twistLimitDegrees=0}
 {bone1Name=dyn_backHair_r_02, bone2Name=dyn_backHair_r_01, swing1LimitDegrees=25, swing2LimitDegrees=25, twistLimitDegrees=0}
 {
  bone1Name=base_hairPiece_l
  bone2Name=base_hair
  swing1LimitDegrees=45
  swing1Motion=Free
  swing2LimitDegrees=45
  swing2Motion=Free
  twistLimitDegrees=45
  twistMotion=Free
 }
 {bone1Name=dyn_hairPiece_l_01, bone2Name=base_hairPiece_l, swing1LimitDegrees=25, swing2LimitDegrees=45, twistLimitDegrees=0}
 {bone1Name=dyn_hairPiece_l_02, bone2Name=dyn_hairPiece_l_01, swing1LimitDegrees=25, swing2LimitDegrees=45, twistLimitDegrees=0}
 {bone1Name=dyn_hairPiece_l_03, bone2Name=dyn_hairPiece_l_02, swing1LimitDegrees=25, swing2LimitDegrees=45, twistLimitDegrees=0}
 {
  bone1Name=base_hairPiece_r
  bone2Name=base_hair
  swing1LimitDegrees=45
  swing1Motion=Free
  swing2LimitDegrees=45
  swing2Motion=Free
  twistLimitDegrees=45
  twistMotion=Free
 }
 {bone1Name=dyn_hairPiece_r_01, bone2Name=base_hairPiece_r, swing1LimitDegrees=25, swing2LimitDegrees=45, twistLimitDegrees=0}
 {bone1Name=dyn_hairPiece_r_02, bone2Name=dyn_hairPiece_r_01, swing1LimitDegrees=25, swing2LimitDegrees=45, twistLimitDegrees=0}
 {bone1Name=dyn_hairPiece_r_03, bone2Name=dyn_hairPiece_r_02, swing1LimitDegrees=25, swing2LimitDegrees=30, twistLimitDegrees=0}
 {
  bone1Name=clavicle_l
  bone2Name=spine_05
  swing1LimitDegrees=45
  swing1Motion=Free
  swing2LimitDegrees=45
  swing2Motion=Free
  twistLimitDegrees=45
  twistMotion=Free
 }
]
