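# Post-process style AnimBP for the CordSyrup face accessory, driving back hair and pigtail motion on the character.
# It copies the body mesh pose, then applies additive back-hair bone offsets weighted by head pitch, crouch and targeting alphas.
# Chained AnimDynamics nodes spring each pigtail strand before a final RigidBody pass simulated relative to neck_02.
# The EventGraph computes those alphas and state-based gravity and RBAN settings through a deprecated helper.
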AnimBlueprint /BRCosmetics/Characters/Player/Female/Medium/Bodies/F_MED_CordSyrup/Meshes/Parts/F_MED_CordSyrup_FaceAcc_AnimBP
target_skeleton=&/F_MED_CordSyrup_FaceAcc_Skeleton
graphs [
 AnimGraph {nodes [
   n1 = Get Head Up() -> (Head Up:float)
   n2 = "Transform (Modify) Bone - Bone: base_backHair_l"(
    ComponentPose=Local To Component(LocalPose=Use cached pose 'StartCache'())
    Alpha=@n1
    Translation="0,0,0"
    Rotation="0.008,-22.072,-0.015"
    Scale="1,1,1"
   ) {
    rotationMode=BMM_Additive
    rotationSpace=BCS_BoneSpace
    translationMode=BMM_Additive
    translationSpace=BCS_BoneSpace
    alphaBoolBlend {blendOption=Linear}
    boneToModify {boneName=base_backHair_l}
    rotation=<rot 0.008 -22.072 -0.015>
   }
   n3 = "Transform (Modify) Bone - Bone: base_backHair_r"(ComponentPose=@n2, Alpha=@n1, Translation="-3.874,-0.614,1.075", Rotation="20.341,-42.586,-106.888", Scale="1,1,1") {
    rotationMode=BMM_Additive
    rotationSpace=BCS_BoneSpace
    translationMode=BMM_Additive
    translationSpace=BCS_BoneSpace
    alphaBoolBlend {blendOption=Linear}
    boneToModify {boneName=base_backHair_r}
    rotation=<rot 20.341 -42.586 -106.888>
    translation=<-3.874 -0.614 1.075>
   }
   n4 = "Transform (Modify) Bone - Bone: base_backHair_c"(ComponentPose=@n3, Alpha=@n1, Translation="-2.305,0,1.704", Rotation="49.441,-58.291,-110.966", Scale="1,1,1") {
    rotationMode=BMM_Additive
    rotationSpace=BCS_BoneSpace
    translationMode=BMM_Additive
    translationSpace=BCS_BoneSpace
    alphaBoolBlend {blendOption=Linear}
    boneToModify {boneName=base_backHair_c}
    rotation=<rot 49.441 -58.291 -110.966>
    translation {x=-2.305, z=1.704}
   }
   n5 = "Transform (Modify) Bone - Bone: base_backHair_c"(
    ComponentPose=@n4
    Alpha=Get Crouch Head Up() -> (Crouch Head Up:float)
    Translation="2.315,-2.883,5.898"
    Rotation="-17.698,0.285,0.040"
    Scale="1,1,1"
   ) {
    rotationMode=BMM_Additive
    rotationSpace=BCS_BoneSpace
    translationMode=BMM_Additive
    translationSpace=BCS_BoneSpace
    alphaBoolBlend {blendOption=Linear}
    boneToModify {boneName=base_backHair_c}
    rotation=<rot -17.698 0.285 0.04>
    translation=<2.315 -2.883 5.898>
   }
   n6 = Get Crouch Targeting Up() -> (Crouch Targeting Up:double)
   n7 = "Transform (Modify) Bone - Bone: base_backHair_c"(ComponentPose=@n5, Alpha=@n6, Translation="-1.457,4.260,-4.192", Rotation="32.922,73.803,115.458", Scale="1,1,1") {
    rotationMode=BMM_Additive
    rotationSpace=BCS_BoneSpace
    translationMode=BMM_Additive
    translationSpace=BCS_BoneSpace
    alphaBoolBlend {blendOption=Linear}
    boneToModify {boneName=base_backHair_c}
    rotation=<rot 32.922 73.803 115.458>
    translation=<-1.457 4.26 -4.192>
   }
   n8 = "Transform (Modify) Bone - Bone: base_backHair_r"(ComponentPose=@n7, Alpha=@n6, Translation="2.278,2.270,2.485", Rotation="-5.711,-91.523,87.658", Scale="1,1,1") {
    rotationMode=BMM_Additive
    rotationSpace=BCS_BoneSpace
    translationMode=BMM_Additive
    translationSpace=BCS_BoneSpace
    alphaBoolBlend {blendOption=Linear}
    boneToModify {boneName=base_backHair_r}
    rotation=<rot -5.711 -91.523 87.658>
    translation=<2.278 2.27 2.485>
   }
   n9 = Anim Dynamics - Bone: dyn_pigtails_l!(ComponentPose=@n8, Alpha=1, GravityOverride:Vector="0,0,-980", ExternalForce="0,0,0") -> (Pose) {
    angularBiasOverride=1
    angularDampingOverride=1
    angularSpringConstant=20000
    bAngularSpring=true
    bOverrideAngularBias=true
    bOverrideAngularDamping=true
    bUseGravityOverride=true
    gravityScale=0.35
    simulationSpace=RootRelative
    alphaBoolBlend {blendOption=Linear}
    boundBone {boneName=dyn_pigtails_l}
    componentAppliedLinearAccClamp=<3000 3000 3000>
    componentLinearAccScale=<0.3 0.33 0.3>
    componentLinearVelScale=<0.3 0.3 0.3>
    gravityOverride {z=-980}
    physicsBodyDefinitions [{
      boundBone {boneName=dyn_pigtails_l}
      boxExtents=<25 45 45>
      constraintSetup {angularLimitsMax=<25 5 25>, angularLimitsMin=<-25 -5 -25>, angularTarget=<1 0 0>}
      localJointOffset=<-2 0 -5>
     }]
   }
   n10 = Anim Dynamics - Bone: dyn_pigtailStrand_l_01!(ComponentPose=@n9, Alpha=1, GravityOverride:Vector="0,0,-980", ExternalForce="0,0,0") -> (Pose) {
    angularBiasOverride=1
    angularDampingOverride=1
    angularSpringConstant=20000
    bAngularSpring=true
    bOverrideAngularBias=true
    bOverrideAngularDamping=true
    bUseGravityOverride=true
    gravityScale=0.35
    simulationSpace=RootRelative
    alphaBoolBlend {blendOption=Linear}
    boundBone {boneName=dyn_pigtailStrand_l_01}
    componentAppliedLinearAccClamp=<3000 3000 3000>
    componentLinearAccScale=<0.3 0.33 0.3>
    componentLinearVelScale=<0.3 0.3 0.3>
    gravityOverride {z=-980}
    physicsBodyDefinitions [{
      boundBone {boneName=dyn_pigtailStrand_l_01}
      boxExtents=<20 30 30>
      constraintSetup {angularLimitsMax=<15 15 5>, angularLimitsMin=<-15 -15 -15>, angularTarget=<1 0 0>}
      localJointOffset=<0 0 -5>
     }]
   }
   n11 = Anim Dynamics - Bone: dyn_pigtailStrand_l_02!(ComponentPose=@n10, Alpha=1, GravityOverride:Vector="0,0,-980", ExternalForce="0,0,0") -> (Pose) {
    angularBiasOverride=1
    angularDampingOverride=1
    angularSpringConstant=20000
    bAngularSpring=true
    bOverrideAngularBias=true
    bOverrideAngularDamping=true
    bUseGravityOverride=true
    gravityScale=0.35
    simulationSpace=RootRelative
    alphaBoolBlend {blendOption=Linear}
    boundBone {boneName=dyn_pigtailStrand_l_02}
    componentAppliedLinearAccClamp=<3000 3000 3000>
    componentLinearAccScale=<0.3 0.33 0.3>
    componentLinearVelScale=<0.3 0.3 0.3>
    gravityOverride {z=-980}
    physicsBodyDefinitions [{
      boundBone {boneName=dyn_pigtailStrand_l_02}
      boxExtents=<10 20 20>
      constraintSetup {angularLimitsMax=<15 15 5>, angularLimitsMin=<-15 -15 -15>, angularTarget=<1 0 0>}
      localJointOffset=<0 0 -5>
     }]
   }
   n12 = Anim Dynamics - Bone: dyn_pigtailStrand_l_03!(ComponentPose=@n11, Alpha=1, GravityOverride:Vector="0,0,-980", ExternalForce="0,0,0") -> (Pose) {
    angularBiasOverride=1
    angularDampingOverride=1
    angularSpringConstant=20000
    bAngularSpring=true
    bOverrideAngularBias=true
    bOverrideAngularDamping=true
    bUseGravityOverride=true
    gravityScale=0.35
    simulationSpace=RootRelative
    alphaBoolBlend {blendOption=Linear}
    boundBone {boneName=dyn_pigtailStrand_l_03}
    componentAppliedLinearAccClamp=<3000 3000 3000>
    componentLinearAccScale=<0.3 0.33 0.3>
    componentLinearVelScale=<0.3 0.3 0.3>
    gravityOverride {z=-980}
    physicsBodyDefinitions [{
      boundBone {boneName=dyn_pigtailStrand_l_03}
      boxExtents=<20 30 30>
      constraintSetup {angularLimitsMax=<15 15 5>, angularLimitsMin=<-15 -15 -15>, angularTarget=<1 0 0>}
      localJointOffset=<-5 0 0>
     }]
   }
   n13 = Anim Dynamics - Bone: dyn_pigtailStrand_l_04!(ComponentPose=@n12, Alpha=1, GravityOverride:Vector="0,0,-980", ExternalForce="0,0,0") -> (Pose) {
    angularBiasOverride=1
    angularDampingOverride=1
    angularSpringConstant=20000
    bAngularSpring=true
    bOverrideAngularBias=true
    bOverrideAngularDamping=true
    bUseGravityOverride=true
    gravityScale=0.35
    simulationSpace=RootRelative
    alphaBoolBlend {blendOption=Linear}
    boundBone {boneName=dyn_pigtailStrand_l_04}
    componentAppliedLinearAccClamp=<3000 3000 3000>
    componentLinearAccScale=<0.3 0.33 0.3>
    componentLinearVelScale=<0.3 0.3 0.3>
    gravityOverride {z=-980}
    physicsBodyDefinitions [{
      boundBone {boneName=dyn_pigtailStrand_l_04}
      boxExtents=<10 20 20>
      constraintSetup {angularLimitsMax=<15 15 5>, angularLimitsMin=<-15 -15 -15>, angularTarget=<1 0 0>}
      localJointOffset=<0 0 -5>
     }]
   }
   n14 = Anim Dynamics - Bone: dyn_pigtailStrand_l_05!(ComponentPose=@n13, Alpha=1, GravityOverride:Vector="0,0,-980", ExternalForce="0,0,0") -> (Pose) {
    angularBiasOverride=1
    angularDampingOverride=1
    angularSpringConstant=20000
    bAngularSpring=true
    bOverrideAngularBias=true
    bOverrideAngularDamping=true
    bUseGravityOverride=true
    gravityScale=0.35
    simulationSpace=RootRelative
    alphaBoolBlend {blendOption=Linear}
    boundBone {boneName=dyn_pigtailStrand_l_05}
    componentAppliedLinearAccClamp=<3000 3000 3000>
    componentLinearAccScale=<0.3 0.33 0.3>
    componentLinearVelScale=<0.3 0.3 0.3>
    gravityOverride {z=-980}
    physicsBodyDefinitions [{
      boundBone {boneName=dyn_pigtailStrand_l_05}
      boxExtents=<20 30 30>
      constraintSetup {angularLimitsMax=<15 15 5>, angularLimitsMin=<-15 -15 -15>, angularTarget=<1 0 0>}
      localJointOffset=<5 0 0>
     }]
   }
   n15 = Anim Dynamics - Bone: dyn_pigtailStrand_l_06!(ComponentPose=@n14, Alpha=1, GravityOverride:Vector="0,0,-980", ExternalForce="0,0,0") -> (Pose) {
    angularBiasOverride=1
    angularDampingOverride=1
    angularSpringConstant=20000
    bAngularSpring=true
    bOverrideAngularBias=true
    bOverrideAngularDamping=true
    bUseGravityOverride=true
    gravityScale=0.35
    simulationSpace=RootRelative
    alphaBoolBlend {blendOption=Linear}
    boundBone {boneName=dyn_pigtailStrand_l_06}
    componentAppliedLinearAccClamp=<3000 3000 3000>
    componentLinearAccScale=<0.3 0.33 0.3>
    componentLinearVelScale=<0.3 0.3 0.3>
    gravityOverride {z=-980}
    physicsBodyDefinitions [{
      boundBone {boneName=dyn_pigtailStrand_l_06}
      boxExtents=<10 20 20>
      constraintSetup {angularLimitsMax=<15 15 5>, angularLimitsMin=<-15 -15 -15>, angularTarget=<1 0 0>}
      localJointOffset=<5 0 0>
     }]
   }
   n16 = Anim Dynamics - Bone: dyn_pigtailStrand_l_07!(ComponentPose=@n15, Alpha=1, GravityOverride:Vector="0,0,-980", ExternalForce="0,0,0") -> (Pose) {
    angularBiasOverride=1
    angularDampingOverride=1
    angularSpringConstant=20000
    bAngularSpring=true
    bOverrideAngularBias=true
    bOverrideAngularDamping=true
    bUseGravityOverride=true
    gravityScale=0.35
    simulationSpace=RootRelative
    alphaBoolBlend {blendOption=Linear}
    boundBone {boneName=dyn_pigtailStrand_l_07}
    componentAppliedLinearAccClamp=<3000 3000 3000>
    componentLinearAccScale=<0.3 0.33 0.3>
    componentLinearVelScale=<0.3 0.3 0.3>
    gravityOverride {z=-980}
    physicsBodyDefinitions [{
      boundBone {boneName=dyn_pigtailStrand_l_07}
      boxExtents=<20 30 30>
      constraintSetup {angularLimitsMax=<15 15 5>, angularLimitsMin=<-15 -15 -15>, angularTarget=<1 0 0>}
      localJointOffset=<5 0 0>
     }]
   }
   n17 = Anim Dynamics - Bone: dyn_pigtailStrand_l_08!(ComponentPose=@n16, Alpha=1, GravityOverride:Vector="0,0,-980", ExternalForce="0,0,0") -> (Pose) {
    angularBiasOverride=1
    angularDampingOverride=1
    angularSpringConstant=20000
    bAngularSpring=true
    bOverrideAngularBias=true
    bOverrideAngularDamping=true
    bUseGravityOverride=true
    gravityScale=0.35
    simulationSpace=RootRelative
    alphaBoolBlend {blendOption=Linear}
    boundBone {boneName=dyn_pigtailStrand_l_08}
    componentAppliedLinearAccClamp=<3000 3000 3000>
    componentLinearAccScale=<0.3 0.33 0.3>
    componentLinearVelScale=<0.3 0.3 0.3>
    gravityOverride {z=-980}
    physicsBodyDefinitions [{
      boundBone {boneName=dyn_pigtailStrand_l_08}
      boxExtents=<10 20 20>
      constraintSetup {angularLimitsMax=<15 15 5>, angularLimitsMin=<-15 -15 -15>, angularTarget=<1 0 0>}
      localJointOffset=<5 0 0>
     }]
   }
   n18 = Anim Dynamics - Bone: dyn_pigtailStrand_l_09!(ComponentPose=@n17, Alpha=1, GravityOverride:Vector="0,0,-980", ExternalForce="0,0,0") -> (Pose) {
    angularBiasOverride=1
    angularDampingOverride=1
    angularSpringConstant=20000
    bAngularSpring=true
    bOverrideAngularBias=true
    bOverrideAngularDamping=true
    bUseGravityOverride=true
    gravityScale=0.35
    simulationSpace=RootRelative
    alphaBoolBlend {blendOption=Linear}
    boundBone {boneName=dyn_pigtailStrand_l_09}
    componentAppliedLinearAccClamp=<3000 3000 3000>
    componentLinearAccScale=<0.3 0.33 0.3>
    componentLinearVelScale=<0.3 0.3 0.3>
    gravityOverride {z=-980}
    physicsBodyDefinitions [{
      boundBone {boneName=dyn_pigtailStrand_l_09}
      boxExtents=<20 30 30>
      constraintSetup {angularLimitsMax=<15 15 5>, angularLimitsMin=<-15 -15 -15>, angularTarget=<1 0 0>}
      localJointOffset=<5 0 0>
     }]
   }
   n19 = Anim Dynamics - Bone: dyn_pigtailStrand_l_10!(ComponentPose=@n18, Alpha=1, GravityOverride:Vector="0,0,-980", ExternalForce="0,0,0") -> (Pose) {
    angularBiasOverride=1
    angularDampingOverride=1
    angularSpringConstant=20000
    bAngularSpring=true
    bOverrideAngularBias=true
    bOverrideAngularDamping=true
    bUseGravityOverride=true
    gravityScale=0.35
    simulationSpace=RootRelative
    alphaBoolBlend {blendOption=Linear}
    boundBone {boneName=dyn_pigtailStrand_l_10}
    componentAppliedLinearAccClamp=<3000 3000 3000>
    componentLinearAccScale=<0.3 0.33 0.3>
    componentLinearVelScale=<0.3 0.3 0.3>
    gravityOverride {z=-980}
    physicsBodyDefinitions [{
      boundBone {boneName=dyn_pigtailStrand_l_10}
      boxExtents=<10 20 20>
      constraintSetup {angularLimitsMax=<15 30 5>, angularLimitsMin=<-15 -30 -15>, angularTarget=<1 0 0>}
      localJointOffset=<0 0 -5>
     }]
   }
   n20 = Anim Dynamics - Bone: dyn_pigtailStrand_l_11!(ComponentPose=@n19, Alpha=1, GravityOverride:Vector="0,0,-980", ExternalForce="0,0,0") -> (Pose) {
    angularBiasOverride=1
    angularDampingOverride=1
    angularSpringConstant=20000
    bAngularSpring=true
    bOverrideAngularBias=true
    bOverrideAngularDamping=true
    bUseGravityOverride=true
    gravityScale=0.35
    simulationSpace=RootRelative
    alphaBoolBlend {blendOption=Linear}
    boundBone {boneName=dyn_pigtailStrand_l_11}
    componentAppliedLinearAccClamp=<3000 3000 3000>
    componentLinearAccScale=<0.3 0.33 0.3>
    componentLinearVelScale=<0.3 0.3 0.3>
    gravityOverride {z=-980}
    physicsBodyDefinitions [{
      boundBone {boneName=dyn_pigtailStrand_l_11}
      boxExtents=<20 30 30>
      constraintSetup {angularLimitsMax=<15 15 5>, angularLimitsMin=<-15 -15 -15>, angularTarget=<1 0 0>}
      localJointOffset=<0 0 -5>
     }]
   }
   n21 = Anim Dynamics - Bone: dyn_pigtailStrand_l_12!(ComponentPose=@n20, Alpha=1, GravityOverride:Vector="0,0,-980", ExternalForce="0,0,0") -> (Pose) {
    angularBiasOverride=1
    angularDampingOverride=1
    angularSpringConstant=20000
    bAngularSpring=true
    bOverrideAngularBias=true
    bOverrideAngularDamping=true
    bUseGravityOverride=true
    gravityScale=0.35
    simulationSpace=RootRelative
    alphaBoolBlend {blendOption=Linear}
    boundBone {boneName=dyn_pigtailStrand_l_12}
    componentAppliedLinearAccClamp=<3000 3000 3000>
    componentLinearAccScale=<0.3 0.33 0.3>
    componentLinearVelScale=<0.3 0.3 0.3>
    gravityOverride {z=-980}
    physicsBodyDefinitions [{
      boundBone {boneName=dyn_pigtailStrand_l_12}
      boxExtents=<10 20 20>
      constraintSetup {angularLimitsMax=<15 15 5>, angularLimitsMin=<-15 -15 -15>, angularTarget=<1 0 0>}
      localJointOffset=<5 0 0>
     }]
   }
   n22 = Anim Dynamics - Bone: dyn_pigtails_r!(ComponentPose=@n21, Alpha=1, GravityOverride:Vector="0,0,-980", ExternalForce="0,0,0") -> (Pose) {
    angularBiasOverride=1
    angularDampingOverride=1
    angularSpringConstant=20000
    bAngularSpring=true
    bOverrideAngularBias=true
    bOverrideAngularDamping=true
    bUseGravityOverride=true
    gravityScale=0.35
    simulationSpace=RootRelative
    alphaBoolBlend {blendOption=Linear}
    boundBone {boneName=dyn_pigtails_r}
    componentAppliedLinearAccClamp=<3000 3000 3000>
    componentLinearAccScale=<0.3 0.33 0.3>
    componentLinearVelScale=<0.3 0.3 0.3>
    gravityOverride {z=-980}
    physicsBodyDefinitions [{
      boundBone {boneName=dyn_pigtails_r}
      boxExtents=<30 60 60>
      constraintSetup {angularLimitsMax=<25 5 25>, angularLimitsMin=<-25 -5 -25>, angularTarget=<1 0 0>}
      localJointOffset=<0 0 -5>
     }]
   }
   n23 = Anim Dynamics - Bone: dyn_pigtailStrand_r_01!(ComponentPose=@n22, Alpha=1, GravityOverride:Vector="0,0,-980", ExternalForce="0,0,0") -> (Pose) {
    angularBiasOverride=1
    angularDampingOverride=1
    angularSpringConstant=20000
    bAngularSpring=true
    bOverrideAngularBias=true
    bOverrideAngularDamping=true
    bUseGravityOverride=true
    gravityScale=0.35
    simulationSpace=RootRelative
    alphaBoolBlend {blendOption=Linear}
    boundBone {boneName=dyn_pigtailStrand_r_01}
    componentAppliedLinearAccClamp=<3000 3000 3000>
    componentLinearAccScale=<0.3 0.33 0.3>
    componentLinearVelScale=<0.3 0.3 0.3>
    gravityOverride {z=-980}
    physicsBodyDefinitions [{
      boundBone {boneName=dyn_pigtailStrand_r_01}
      boxExtents=<20 30 30>
      constraintSetup {angularLimitsMax=<15 15 5>, angularLimitsMin=<-15 -15 -15>, angularTarget=<1 0 0>}
      localJointOffset=<0 0 -5>
     }]
   }
   n24 = Anim Dynamics - Bone: dyn_pigtailStrand_r_02!(ComponentPose=@n23, Alpha=1, GravityOverride:Vector="0,0,-980", ExternalForce="0,0,0") -> (Pose) {
    angularBiasOverride=1
    angularDampingOverride=1
    angularSpringConstant=20000
    bAngularSpring=true
    bOverrideAngularBias=true
    bOverrideAngularDamping=true
    bUseGravityOverride=true
    gravityScale=0.35
    simulationSpace=RootRelative
    alphaBoolBlend {blendOption=Linear}
    boundBone {boneName=dyn_pigtailStrand_r_02}
    componentAppliedLinearAccClamp=<3000 3000 3000>
    componentLinearAccScale=<0.3 0.33 0.3>
    componentLinearVelScale=<0.3 0.3 0.3>
    gravityOverride {z=-980}
    physicsBodyDefinitions [{
      boundBone {boneName=dyn_pigtailStrand_r_02}
      boxExtents=<10 20 20>
      constraintSetup {angularLimitsMax=<15 15 5>, angularLimitsMin=<-15 -15 -15>, angularTarget=<1 0 0>}
      localJointOffset=<0 0 -5>
     }]
   }
   n25 = Anim Dynamics - Bone: dyn_pigtailStrand_r_03!(ComponentPose=@n24, Alpha=1, GravityOverride:Vector="0,0,-980", ExternalForce="0,0,0") -> (Pose) {
    angularBiasOverride=1
    angularDampingOverride=1
    angularSpringConstant=20000
    bAngularSpring=true
    bOverrideAngularBias=true
    bOverrideAngularDamping=true
    bUseGravityOverride=true
    gravityScale=0.35
    simulationSpace=RootRelative
    alphaBoolBlend {blendOption=Linear}
    boundBone {boneName=dyn_pigtailStrand_r_03}
    componentAppliedLinearAccClamp=<3000 3000 3000>
    componentLinearAccScale=<0.3 0.33 0.3>
    componentLinearVelScale=<0.3 0.3 0.3>
    gravityOverride {z=-980}
    physicsBodyDefinitions [{
      boundBone {boneName=dyn_pigtailStrand_r_03}
      boxExtents=<20 30 30>
      constraintSetup {angularLimitsMax=<15 15 5>, angularLimitsMin=<-15 -15 -15>, angularTarget=<1 0 0>}
      localJointOffset=<-5 0 0>
     }]
   }
   n26 = Anim Dynamics - Bone: dyn_pigtailStrand_r_04!(ComponentPose=@n25, Alpha=1, GravityOverride:Vector="0,0,-980", ExternalForce="0,0,0") -> (Pose) {
    angularBiasOverride=1
    angularDampingOverride=1
    angularSpringConstant=20000
    bAngularSpring=true
    bOverrideAngularBias=true
    bOverrideAngularDamping=true
    bUseGravityOverride=true
    gravityScale=0.35
    simulationSpace=RootRelative
    alphaBoolBlend {blendOption=Linear}
    boundBone {boneName=dyn_pigtailStrand_r_04}
    componentAppliedLinearAccClamp=<3000 3000 3000>
    componentLinearAccScale=<0.3 0.33 0.3>
    componentLinearVelScale=<0.3 0.3 0.3>
    gravityOverride {z=-980}
    physicsBodyDefinitions [{
      boundBone {boneName=dyn_pigtailStrand_r_04}
      boxExtents=<10 20 20>
      constraintSetup {angularLimitsMax=<15 15 5>, angularLimitsMin=<-15 -15 -15>, angularTarget=<1 0 0>}
      localJointOffset=<0 0 -5>
     }]
   }
   n27 = Anim Dynamics - Bone: dyn_pigtailStrand_r_05!(ComponentPose=@n26, Alpha=1, GravityOverride:Vector="0,0,-980", ExternalForce="0,0,0") -> (Pose) {
    angularBiasOverride=1
    angularDampingOverride=1
    angularSpringConstant=20000
    bAngularSpring=true
    bOverrideAngularBias=true
    bOverrideAngularDamping=true
    bUseGravityOverride=true
    gravityScale=0.35
    simulationSpace=RootRelative
    alphaBoolBlend {blendOption=Linear}
    boundBone {boneName=dyn_pigtailStrand_r_05}
    componentAppliedLinearAccClamp=<3000 3000 3000>
    componentLinearAccScale=<0.3 0.33 0.3>
    componentLinearVelScale=<0.3 0.3 0.3>
    gravityOverride {z=-980}
    physicsBodyDefinitions [{
      boundBone {boneName=dyn_pigtailStrand_r_05}
      boxExtents=<20 30 30>
      constraintSetup {angularLimitsMax=<15 15 5>, angularLimitsMin=<-15 -15 -15>, angularTarget=<1 0 0>}
      localJointOffset=<5 0 0>
     }]
   }
   n28 = Anim Dynamics - Bone: dyn_pigtailStrand_r_06!(ComponentPose=@n27, Alpha=1, GravityOverride:Vector="0,0,-980", ExternalForce="0,0,0") -> (Pose) {
    angularBiasOverride=1
    angularDampingOverride=1
    angularSpringConstant=20000
    bAngularSpring=true
    bOverrideAngularBias=true
    bOverrideAngularDamping=true
    bUseGravityOverride=true
    gravityScale=0.35
    simulationSpace=RootRelative
    alphaBoolBlend {blendOption=Linear}
    boundBone {boneName=dyn_pigtailStrand_r_06}
    componentAppliedLinearAccClamp=<3000 3000 3000>
    componentLinearAccScale=<0.3 0.33 0.3>
    componentLinearVelScale=<0.3 0.3 0.3>
    gravityOverride {z=-980}
    physicsBodyDefinitions [{
      boundBone {boneName=dyn_pigtailStrand_r_06}
      boxExtents=<10 20 20>
      constraintSetup {angularLimitsMax=<15 15 5>, angularLimitsMin=<-15 -15 -15>, angularTarget=<1 0 0>}
      localJointOffset=<5 0 0>
     }]
   }
   n29 = Anim Dynamics - Bone: dyn_pigtailStrand_r_07!(ComponentPose=@n28, Alpha=1, GravityOverride:Vector="0,0,-980", ExternalForce="0,0,0") -> (Pose) {
    angularBiasOverride=1
    angularDampingOverride=1
    angularSpringConstant=20000
    bAngularSpring=true
    bOverrideAngularBias=true
    bOverrideAngularDamping=true
    bUseGravityOverride=true
    gravityScale=0.35
    simulationSpace=RootRelative
    alphaBoolBlend {blendOption=Linear}
    boundBone {boneName=dyn_pigtailStrand_r_07}
    componentAppliedLinearAccClamp=<3000 3000 3000>
    componentLinearAccScale=<0.3 0.33 0.3>
    componentLinearVelScale=<0.3 0.3 0.3>
    gravityOverride {z=-980}
    physicsBodyDefinitions [{
      boundBone {boneName=dyn_pigtailStrand_r_07}
      boxExtents=<20 30 30>
      constraintSetup {angularLimitsMax=<15 15 5>, angularLimitsMin=<-15 -15 -15>, angularTarget=<1 0 0>}
      localJointOffset=<5 0 0>
     }]
   }
   n30 = Anim Dynamics - Bone: dyn_pigtailStrand_r_08!(ComponentPose=@n29, Alpha=1, GravityOverride:Vector="0,0,-980", ExternalForce="0,0,0") -> (Pose) {
    angularBiasOverride=1
    angularDampingOverride=1
    angularSpringConstant=20000
    bAngularSpring=true
    bOverrideAngularBias=true
    bOverrideAngularDamping=true
    bUseGravityOverride=true
    gravityScale=0.35
    simulationSpace=RootRelative
    alphaBoolBlend {blendOption=Linear}
    boundBone {boneName=dyn_pigtailStrand_r_08}
    componentAppliedLinearAccClamp=<3000 3000 3000>
    componentLinearAccScale=<0.3 0.33 0.3>
    componentLinearVelScale=<0.3 0.3 0.3>
    gravityOverride {z=-980}
    physicsBodyDefinitions [{
      boundBone {boneName=dyn_pigtailStrand_r_08}
      boxExtents=<10 20 20>
      constraintSetup {angularLimitsMax=<15 15 5>, angularLimitsMin=<-15 -15 -15>, angularTarget=<1 0 0>}
      localJointOffset=<5 0 0>
     }]
   }
   n31 = Anim Dynamics - Bone: dyn_pigtailStrand_r_09!(ComponentPose=@n30, Alpha=1, GravityOverride:Vector="0,0,-980", ExternalForce="0,0,0") -> (Pose) {
    angularBiasOverride=1
    angularDampingOverride=1
    angularSpringConstant=20000
    bAngularSpring=true
    bOverrideAngularBias=true
    bOverrideAngularDamping=true
    bUseGravityOverride=true
    gravityScale=0.35
    simulationSpace=RootRelative
    alphaBoolBlend {blendOption=Linear}
    boundBone {boneName=dyn_pigtailStrand_r_09}
    componentAppliedLinearAccClamp=<3000 3000 3000>
    componentLinearAccScale=<0.3 0.33 0.3>
    componentLinearVelScale=<0.3 0.3 0.3>
    gravityOverride {z=-980}
    physicsBodyDefinitions [{
      boundBone {boneName=dyn_pigtailStrand_r_09}
      boxExtents=<20 30 30>
      constraintSetup {angularLimitsMax=<15 15 5>, angularLimitsMin=<-15 -15 -15>, angularTarget=<1 0 0>}
      localJointOffset=<5 0 0>
     }]
   }
   n32 = Anim Dynamics - Bone: dyn_pigtailStrand_r_10!(ComponentPose=@n31, Alpha=1, GravityOverride:Vector="0,0,-980", ExternalForce="0,0,0") -> (Pose) {
    angularBiasOverride=1
    angularDampingOverride=1
    angularSpringConstant=20000
    bAngularSpring=true
    bOverrideAngularBias=true
    bOverrideAngularDamping=true
    bUseGravityOverride=true
    gravityScale=0.35
    simulationSpace=RootRelative
    alphaBoolBlend {blendOption=Linear}
    boundBone {boneName=dyn_pigtailStrand_r_10}
    componentAppliedLinearAccClamp=<3000 3000 3000>
    componentLinearAccScale=<0.3 0.33 0.3>
    componentLinearVelScale=<0.3 0.3 0.3>
    gravityOverride {z=-980}
    physicsBodyDefinitions [{
      boundBone {boneName=dyn_pigtailStrand_r_10}
      boxExtents=<10 20 20>
      constraintSetup {angularLimitsMax=<15 30 5>, angularLimitsMin=<-15 -30 -15>, angularTarget=<1 0 0>}
      localJointOffset=<0 0 -5>
     }]
   }
   n33 = Anim Dynamics - Bone: dyn_pigtailStrand_r_11!(ComponentPose=@n32, Alpha=1, GravityOverride:Vector="0,0,-980", ExternalForce="0,0,0") -> (Pose) {
    angularBiasOverride=1
    angularDampingOverride=1
    angularSpringConstant=20000
    bAngularSpring=true
    bOverrideAngularBias=true
    bOverrideAngularDamping=true
    bUseGravityOverride=true
    gravityScale=0.35
    simulationSpace=RootRelative
    alphaBoolBlend {blendOption=Linear}
    boundBone {boneName=dyn_pigtailStrand_r_11}
    componentAppliedLinearAccClamp=<3000 3000 3000>
    componentLinearAccScale=<0.3 0.33 0.3>
    componentLinearVelScale=<0.3 0.3 0.3>
    gravityOverride {z=-980}
    physicsBodyDefinitions [{
      boundBone {boneName=dyn_pigtailStrand_r_11}
      boxExtents=<20 30 30>
      constraintSetup {angularLimitsMax=<15 15 5>, angularLimitsMin=<-15 -15 -15>, angularTarget=<1 0 0>}
      localJointOffset=<0 0 -5>
     }]
   }
   n34 = Anim Dynamics - Bone: dyn_pigtailStrand_r_12!(ComponentPose=@n33, Alpha=1, GravityOverride:Vector="0,0,-980", ExternalForce="0,0,0") -> (Pose) {
    angularBiasOverride=1
    angularDampingOverride=1
    angularSpringConstant=20000
    bAngularSpring=true
    bOverrideAngularBias=true
    bOverrideAngularDamping=true
    bUseGravityOverride=true
    gravityScale=0.35
    simulationSpace=RootRelative
    alphaBoolBlend {blendOption=Linear}
    boundBone {boneName=dyn_pigtailStrand_r_12}
    componentAppliedLinearAccClamp=<3000 3000 3000>
    componentLinearAccScale=<0.3 0.33 0.3>
    componentLinearVelScale=<0.3 0.3 0.3>
    gravityOverride {z=-980}
    physicsBodyDefinitions [{
      boundBone {boneName=dyn_pigtailStrand_r_12}
      boxExtents=<10 20 20>
      constraintSetup {angularLimitsMax=<15 15 5>, angularLimitsMin=<-15 -15 -15>, angularTarget=<1 0 0>}
      localJointOffset=<5 0 0>
     }]
   }
   n35 = Anim Dynamics - Bone: dyn_pigtailStrand_r_13!(ComponentPose=@n34, Alpha=1, GravityOverride:Vector="0,0,-980", ExternalForce="0,0,0") -> (Pose) {
    angularBiasOverride=1
    angularDampingOverride=1
    angularSpringConstant=20000
    bAngularSpring=true
    bOverrideAngularBias=true
    bOverrideAngularDamping=true
    bUseGravityOverride=true
    gravityScale=0.35
    simulationSpace=RootRelative
    alphaBoolBlend {blendOption=Linear}
    boundBone {boneName=dyn_pigtailStrand_r_13}
    componentAppliedLinearAccClamp=<3000 3000 3000>
    componentLinearAccScale=<0.3 0.33 0.3>
    componentLinearVelScale=<0.3 0.3 0.3>
    gravityOverride {z=-980}
    physicsBodyDefinitions [{
      boundBone {boneName=dyn_pigtailStrand_r_13}
      boxExtents=<20 30 30>
      constraintSetup {angularLimitsMax=<15 15 5>, angularLimitsMin=<-15 -15 -15>, angularTarget=<1 0 0>}
      localJointOffset=<0 0 -5>
     }]
   }
   n36 = Anim Dynamics - Bone: dyn_pigtailStrand_r_14!(ComponentPose=@n35, Alpha=1, GravityOverride:Vector="0,0,-980", ExternalForce="0,0,0") -> (Pose) {
    angularBiasOverride=1
    angularDampingOverride=1
    angularSpringConstant=20000
    bAngularSpring=true
    bOverrideAngularBias=true
    bOverrideAngularDamping=true
    bUseGravityOverride=true
    gravityScale=0.35
    simulationSpace=RootRelative
    alphaBoolBlend {blendOption=Linear}
    boundBone {boneName=dyn_pigtailStrand_r_14}
    componentAppliedLinearAccClamp=<3000 3000 3000>
    componentLinearAccScale=<0.3 0.33 0.3>
    componentLinearVelScale=<0.3 0.3 0.3>
    gravityOverride {z=-980}
    physicsBodyDefinitions [{
      boundBone {boneName=dyn_pigtailStrand_r_14}
      boxExtents=<10 20 20>
      constraintSetup {angularLimitsMax=<15 15 5>, angularLimitsMin=<-15 -15 -15>, angularTarget=<1 0 0>}
      localJointOffset=<5 0 0>
     }]
   }
   n37 = Anim Dynamics - Bone: dyn_pigtailOuterStrand_r_02!(ComponentPose=@n36, Alpha=1, GravityOverride:Vector="0,0,-980", ExternalForce="0,0,0") -> (Pose) {
    angularBiasOverride=1
    angularDampingOverride=1
    angularSpringConstant=20000
    bAngularSpring=true
    bOverrideAngularBias=true
    bOverrideAngularDamping=true
    bUseGravityOverride=true
    gravityScale=0.35
    simulationSpace=RootRelative
    alphaBoolBlend {blendOption=Linear}
    boundBone {boneName=dyn_pigtailOuterStrand_r_02}
    componentAppliedLinearAccClamp=<3000 3000 3000>
    componentLinearAccScale=<0.3 0.33 0.3>
    componentLinearVelScale=<0.3 0.3 0.3>
    gravityOverride {z=-980}
    physicsBodyDefinitions [{
      boundBone {boneName=dyn_pigtailOuterStrand_r_02}
      boxExtents=<10 20 20>
      constraintSetup {angularLimitsMax=<15 15 5>, angularLimitsMin=<-15 -15 -15>, angularTarget=<1 0 0>}
      localJointOffset=<0 0 -5>
     }]
   }
   n38 = Anim Dynamics - Bone: dyn_pigtailOuterStrand_r_01!(ComponentPose=@n37, Alpha=1, GravityOverride:Vector="0,0,-980", ExternalForce="0,0,0") -> (Pose) {
    angularBiasOverride=1
    angularDampingOverride=1
    angularSpringConstant=20000
    bAngularSpring=true
    bOverrideAngularBias=true
    bOverrideAngularDamping=true
    bUseGravityOverride=true
    gravityScale=0.35
    simulationSpace=RootRelative
    alphaBoolBlend {blendOption=Linear}
    boundBone {boneName=dyn_pigtailOuterStrand_r_01}
    componentAppliedLinearAccClamp=<3000 3000 3000>
    componentLinearAccScale=<0.3 0.33 0.3>
    componentLinearVelScale=<0.3 0.3 0.3>
    gravityOverride {z=-980}
    physicsBodyDefinitions [{
      boundBone {boneName=dyn_pigtailOuterStrand_r_01}
      boxExtents=<20 30 30>
      constraintSetup {angularLimitsMax=<15 15 5>, angularLimitsMin=<-15 -15 -15>, angularTarget=<1 0 0>}
      localJointOffset=<0 0 -5>
     }]
   }
   n39 = Anim Dynamics - Bone: dyn_pigtailOuterStrand_l_02!(ComponentPose=@n38, Alpha=1, GravityOverride:Vector="0,0,-980", ExternalForce="0,0,0") -> (Pose) {
    angularBiasOverride=1
    angularDampingOverride=1
    angularSpringConstant=20000
    bAngularSpring=true
    bOverrideAngularBias=true
    bOverrideAngularDamping=true
    bUseGravityOverride=true
    gravityScale=0.35
    simulationSpace=RootRelative
    alphaBoolBlend {blendOption=Linear}
    boundBone {boneName=dyn_pigtailOuterStrand_l_02}
    componentAppliedLinearAccClamp=<3000 3000 3000>
    componentLinearAccScale=<0.3 0.33 0.3>
    componentLinearVelScale=<0.3 0.3 0.3>
    gravityOverride {z=-980}
    physicsBodyDefinitions [{
      boundBone {boneName=dyn_pigtailOuterStrand_l_02}
      boxExtents=<10 20 20>
      constraintSetup {angularLimitsMax=<15 15 5>, angularLimitsMin=<-15 -15 -15>, angularTarget=<1 0 0>}
      localJointOffset=<0 0 -5>
     }]
   }
   n40 = Anim Dynamics - Bone: dyn_pigtailOuterStrand_l_01!(ComponentPose=@n39, Alpha=1, GravityOverride:Vector="0,0,-980", ExternalForce="0,0,0") -> (Pose) {
    angularBiasOverride=1
    angularDampingOverride=1
    angularSpringConstant=20000
    bAngularSpring=true
    bOverrideAngularBias=true
    bOverrideAngularDamping=true
    bUseGravityOverride=true
    gravityScale=0.35
    simulationSpace=RootRelative
    alphaBoolBlend {blendOption=Linear}
    boundBone {boneName=dyn_pigtailOuterStrand_l_01}
    componentAppliedLinearAccClamp=<3000 3000 3000>
    componentLinearAccScale=<0.3 0.33 0.3>
    componentLinearVelScale=<0.3 0.3 0.3>
    gravityOverride {z=-980}
    physicsBodyDefinitions [{
      boundBone {boneName=dyn_pigtailOuterStrand_l_01}
      boxExtents=<20 30 30>
      constraintSetup {angularLimitsMax=<15 15 5>, angularLimitsMin=<-15 -15 -15>, angularTarget=<1 0 0>}
      localJointOffset=<-5 0 0>
     }]
   }
   n41 = RigidBody!(ComponentPose=@n40, Alpha, OverrideWorldGravity:Vector="0,0,0", ExternalForce) -> (Pose) {
    bOverrideWorldGravity=true
    overridePhysicsAsset=&/F_MED_CordSyrup_FaceAcc_Physics
    simulationSpace=BaseBoneSpace
    alphaBoolBlend {blendOption=Linear}
    baseBoneRef {boneName=neck_02}
    componentAppliedLinearAccClamp=<3000 3000 3000>
    componentLinearAccScale=<0.3 0.3 0.3>
    componentLinearVelScale=<0.3 0.3 0.3>
    simSpaceSettings {
     maxAngularAcceleration=1000
     maxAngularVelocity=3
     maxLinearAcceleration=700
     maxLinearVelocity=500
     velocityScaleZ=0
     worldAlpha=0.5
     externalLinearDragV=<0.2 0.2>
    }
   }
   Output Pose(Result=Component To Local(ComponentPose=@n41)) {layerGroup=None}
   Save cached pose 'StartCache'(Pose=Copy Pose From Mesh(SourceMeshComponent=Get BodyMesh() -> (BodyMesh:Skeletal Mesh Component Object Reference)) {bUseAttachedParent=true}) {cacheName=StartCache}
  ]}
 EventGraph {nodes [
   Event BlueprintUpdateAnimation()
   > n1 = Sequence()
   > n2 = Gravity Override State Function DEPRECATED(
    CharacterPartAnimInstance=Self-Reference()
    GameTimeInSeconds=GetGameTimeInSeconds()
    GravityOverrideParams=Get Gravity Override Params()
    RBAN Params=Get RBAN Params()
   )
   > Set Gravity Override(Gravity Override=@n2)
   > n3 = Set RBAN Settings Out(RBAN Settings Out:RBANSetting Struct Structure=@n2.RBAN SettingsOut) -> (Output_Get:RBANSetting Struct Structure)
   > Set RBAN Settings Belt for Cloth(RBAN Settings Belt for Cloth:RBANSetting Struct Structure=@n3) -> (Output_Get:RBANSetting Struct Structure)
   > n4 = Branch(Condition=Get bIsCrouching())
   > Set Crouch Head Up(Crouch Head Up:float=0) -> (Output_Get:float)
   Set Crouch Head Up(execute=@n4.else, Crouch Head Up=Get Head Up())
   > n5 = Branch(Condition=Get bIsTargeting())
   > Set Crouch Targeting Up(Crouch Targeting Up:double=0) -> (Output_Get:double)
   Set Crouch Targeting Up(execute=@n5.else, Crouch Targeting Up=Get Crouch Head Up())
   Set Head Up(
    execute=@n1.then_1
    Head Up:float=GetAlphaFromReferenceTransformDelta(SkeletalMeshComponent=Get SkelMesh(), BoneName=spine_05, BaseName=head, InputRange=(X=0,Y=90), TwistAxis="0,0,1")
   ) -> (Output_Get:float)
   Event BlueprintInitializeAnimation()
   > Set SkelMesh(SkelMesh=GetOwnerSkeletalMeshComponent())
   Event OnCharacterCustomizationLoaded()
   > n6 = Cast To PlayerPawn_Athena(Object=TryGetPawnOwner())
   Break Sim Space Settings!(SimSpaceSettings:"Sim Space Settings Structure (by ref)"=Break RBANSetting Struct!(RBANSettingStruct:"RBANSetting Struct Structure (by ref)"=@n3) -> (
     ComponentLinearAccScale_28_#1:Vector=~
     ComponentLinearVelScale_31_#2:Vector=~
     SimSpaceSettings_34_#3:Sim Space Settings Structure
    ) {is=title K2Node_BreakStruct}.SimSpaceSettings_34_#3) -> (
    WorldAlpha:float=~
    VelocityScaleZ:float=~
    DampingAlpha:float=~
    MaxLinearVelocity:float=~
    MaxAngularVelocity:float=~
    MaxLinearAcceleration:float=~
    MaxAngularAcceleration:float=~
    ExternalLinearDragV:Vector=~
    ExternalLinearVelocity:Vector=~
    ExternalAngularVelocity:Vector=~
   ) {is=title K2Node_BreakStruct}
   Set BodyMesh(BodyMesh=GetSkeletalMeshForPartType(self=@n6))
  ]}
]
variables [
 BodyMesh {type=SkeletalMeshComponent}
 Crouch Head Up {type=number}
 Crouch Targeting Up {type=number}
 Crouch Walk Targeting Up {type=number}
 Gravity Override {type=Vector}
 Gravity Override Params {type=GravityOverrideParamsStruct}
 Head Up {type=number}
 RBAN Params {
  type=RBANParamsStruct
  default {
   dBNO {simSpaceSettings {
     maxAngularAcceleration=3
     maxAngularVelocity=5
     maxLinearAcceleration=500
     maxLinearVelocity=1500
     velocityScaleZ=1
     externalLinearDragV=<1 1 0>
    }}
   dBNO_Carried {simSpaceSettings {
     maxAngularAcceleration=3
     maxAngularVelocity=5
     maxLinearAcceleration=500
     maxLinearVelocity=1500
     velocityScaleZ=1
     externalLinearDragV=<1 1 0>
    }}
   emoteOrMelee {simSpaceSettings {
     maxAngularAcceleration=3
     maxAngularVelocity=5
     maxLinearAcceleration=500
     maxLinearVelocity=1500
     velocityScaleZ=1
     externalLinearDragV=<1 1 0>
    }}
   front_End {simSpaceSettings {
     maxAngularAcceleration=3
     maxAngularVelocity=5
     maxLinearAcceleration=500
     maxLinearVelocity=1500
     velocityScaleZ=1
     externalLinearDragV=<1 1 0>
    }}
   onGround_Crouch_Moving {
    componentLinearAccScale=<1 1 1>
    componentLinearVelScale=<1 1 1>
    simSpaceSettings {
     maxAngularAcceleration=3
     maxAngularVelocity=5
     maxLinearAcceleration=500
     maxLinearVelocity=1500
     velocityScaleZ=1
     externalLinearDragV=<1 1 0>
    }
   }
   onGround_Falling {simSpaceSettings {
     maxAngularAcceleration=3
     maxAngularVelocity=5
     maxLinearAcceleration=500
     maxLinearVelocity=1500
     velocityScaleZ=1
     externalLinearDragV=<1 1 0>
    }}
   onGround_Moving {
    componentLinearAccScale=<1 1 1>
    componentLinearVelScale=<1 1 1>
    simSpaceSettings {
     maxAngularAcceleration=5
     maxAngularVelocity=10
     maxLinearAcceleration=500
     maxLinearVelocity=1500
     velocityScaleZ=1
     externalLinearDragV=<1 1 0>
    }
   }
   onGround_Standing {
    componentLinearAccScale=<1 1 1>
    componentLinearVelScale=<1 1 1>
    simSpaceSettings {
     maxAngularAcceleration=3
     maxAngularVelocity=5
     maxLinearAcceleration=500
     maxLinearVelocity=1500
     velocityScaleZ=1
     externalLinearDragV=<1 1 0>
    }
   }
   rail_Grinding {simSpaceSettings {
     maxAngularAcceleration=3
     maxAngularVelocity=5
     maxLinearAcceleration=500
     maxLinearVelocity=1500
     velocityScaleZ=1
     worldAlpha=1
     externalLinearDragV=<1 1 0>
    }}
   riding_Moving {simSpaceSettings {
     maxAngularAcceleration=3
     maxAngularVelocity=5
     maxLinearAcceleration=500
     maxLinearVelocity=1500
     velocityScaleZ=1
     externalLinearDragV=<1 1 0>
    }}
   riding_Moving_Targeting {simSpaceSettings {
     maxAngularAcceleration=3
     maxAngularVelocity=5
     maxLinearAcceleration=500
     maxLinearVelocity=1500
     velocityScaleZ=1
     externalLinearDragV=<1 1 0>
    }}
   riding_Standing {simSpaceSettings {
     maxAngularAcceleration=3
     maxAngularVelocity=5
     maxLinearAcceleration=500
     maxLinearVelocity=1500
     velocityScaleZ=1
     externalLinearDragV=<1 1 0>
    }}
   skydive_DivingDown {
    componentLinearAccScale=<0 0 0>
    componentLinearVelScale=<0 0 0>
    simSpaceSettings {
     maxAngularAcceleration=3
     maxAngularVelocity=5
     maxLinearAcceleration=500
     maxLinearVelocity=1500
     velocityScaleZ=1
     externalLinearDragV=<1 1 0>
    }
   }
   skydive_FloatingDown {
    componentLinearAccScale=<1 1 1>
    componentLinearVelScale=<1 1 1>
    simSpaceSettings {
     maxAngularAcceleration=3
     maxAngularVelocity=5
     maxLinearAcceleration=500
     maxLinearVelocity=1500
     velocityScaleZ=1
     externalLinearDragV=<1 1 0>
    }
   }
   skydive_GoingUp {simSpaceSettings {
     maxAngularAcceleration=3
     maxAngularVelocity=5
     maxLinearAcceleration=500
     maxLinearVelocity=1500
     velocityScaleZ=1
     externalLinearDragV=<1 1 0>
    }}
   skydive_Parachuting {
    componentLinearAccScale=<1 1 1>
    componentLinearVelScale=<1 1 1>
    simSpaceSettings {
     maxAngularAcceleration=3
     maxAngularVelocity=5
     maxLinearAcceleration=500
     maxLinearVelocity=1500
     velocityScaleZ=1
     externalLinearDragV=<1 1 0>
    }
   }
   skydive_Surfing {simSpaceSettings {
     maxAngularAcceleration=3
     maxAngularVelocity=5
     maxLinearAcceleration=500
     maxLinearVelocity=1500
     velocityScaleZ=1
     externalLinearDragV=<1 1 0>
    }}
   swimming {simSpaceSettings {
     maxAngularAcceleration=3
     maxAngularVelocity=5
     maxLinearAcceleration=500
     maxLinearVelocity=1500
     velocityScaleZ=1
     externalLinearDragV=<1 1 0>
    }}
   tactical_Sprint {
    componentLinearAccScale=<1 1 1>
    componentLinearVelScale=<1 1 1>
    simSpaceSettings {
     maxAngularAcceleration=3
     maxAngularVelocity=5
     maxLinearAcceleration=500
     maxLinearVelocity=1500
     velocityScaleZ=1
     externalLinearDragV=<1 1 0>
    }
   }
   ziplining {simSpaceSettings {
     maxAngularAcceleration=0.5
     maxAngularVelocity=1
     maxLinearAcceleration=50
     maxLinearVelocity=200
     velocityScaleZ=1
     externalLinearDragV=<0.5 0.5 0>
    }}
  }
 }
 RBAN Settings Belt for Cloth {
  type=RBANSettingStruct
  default {
   componentLinearAccScale=<0 0 0>
   componentLinearVelScale=<0 0 0>
   simSpaceSettings {
    dampingAlpha=1
    maxAngularAcceleration=1000
    maxAngularVelocity=3
    maxLinearAcceleration=300
    maxLinearVelocity=700
    velocityScaleZ=0
    worldAlpha=1
    externalAngularVelocity=<0 0 0>
    externalLinearDragV=<0 0.5 0>
    externalLinearVelocity=<0 0 0>
   }
  }
 }
 RBAN Settings Out {
  type=RBANSettingStruct
  default {
   componentLinearAccScale=<0 0 0>
   componentLinearVelScale=<0 0 0>
   simSpaceSettings {
    dampingAlpha=1
    maxAngularAcceleration=1000
    maxAngularVelocity=3
    maxLinearAcceleration=300
    maxLinearVelocity=700
    velocityScaleZ=0
    worldAlpha=1
    externalAngularVelocity=<0 0 0>
    externalLinearDragV=<0 0.5 0>
    externalLinearVelocity=<0 0 0>
   }
  }
 }
 SkelMesh {type=SkeletalMeshComponent}
]
